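# One-shot stereo MetaSound UI sound in the Fort_UI sound class, named as the whoosh to a tab on the Compete screen.
# On play, it triggers the shared MSP_Audio_StereoRng_Waveplayer patch, fed by a wave asset array variable.
# The patch's left and right outputs drive the stereo output, and its finished trigger ends the source.
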MetaSoundSource /Compete/Assets/Audio/MSS_UI_CompeteScreen_WooshToTab
inputs [UE.Source.OnPlay {type=Trigger}]
interfaces [UE.OutputFormat.Stereo:1.0, UE.Source.OneShot:1.0, UE.Source:1.1]
outputs [
 "UE.OutputFormat.Stereo.Audio:0" {type=Audio}
 "UE.OutputFormat.Stereo.Audio:1" {type=Audio}
 UE.Source.OneShot.OnFinished {type=Trigger}
]
pages [Default {
  input_defaults {UE.Source.OnPlay=null}
  nodes [MSP_Audio_StereoRng_Waveplayer_0 = /FortMetasound/Patches/MSP_Audio_StereoRng_Waveplayer(Play=^UE.Source.OnPlay, In Array=VariableAccessor.WaveAsset:Array(Variable=InitVariable.WaveAsset:Array().Variable).Value) {also {defaults {Pitch.Max=0}}}]
  links [
   {dst="@UE.Source.OneShot.OnFinished", src=MSP_Audio_StereoRng_Waveplayer_0.On Finished}
   {dst="@UE.OutputFormat.Stereo.Audio:0", src=MSP_Audio_StereoRng_Waveplayer_0.Out.L}
   {dst="@UE.OutputFormat.Stereo.Audio:1", src=MSP_Audio_StereoRng_Waveplayer_0.Out.R}
  ]
 }]
properties {
 assetImportData=&self:AssetImportData
 numChannels=2
 outputFormat=Stereo
 soundClassObject=&Fort_UI
}
