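# Calendar month header widget with an uppercase month label and previous/next chevron buttons mapped to gamepad triggers.
# Clicks fire an element-interacted event, then the matching event dispatcher; SetText sets the label.
# Input type drives layout: gamepad hides chevrons and moves Next right of the label, as on mobile, which also enlarges icons; otherwise both buttons sit left.
# PlayButtonAnimation fades and slides the initially collapsed buttons in, or reverses to hide them.
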WidgetBlueprint /Compete/Calendar/WBP_CalendarWrapper
animations [Anim_ShowButtons {
  binding_count=6
  length_seconds=1.05
  object=&self:Anim_ShowButtons
  sequence {bindings [
    NextButton {
     id=#1
     possessed_class=WBP_UIKit_Button_Generic_C
     tracks [
      {
       type=MovieSceneFloatTrack
       display_name=Render Opacity
       sections [{type=MovieSceneFloatSection, end_frame=18, start_frame=0, channels [None {type=MovieSceneScriptingFloatChannel, keys [[0,0],[13,0],[18,1]]}]}]
      }
      {
       type=MovieSceneByteTrack
       display_name=Visibility
       sections [{type=MovieSceneByteSection, end_frame=0, start_frame=-1, channels [None {type=MovieSceneScriptingByteChannel, keys [[-1,1],[0,4]]}]}]
      }
     ]
    }
    GridSlot_2 {
     id=#2
     parent_id=#3
     possessed_class=GridSlot
     tracks [
      {
       type=MovieSceneDoubleVectorTrack
       display_name=Nudge
       sections [{
         type=MovieSceneDoubleVectorSection
         end_frame=20
         start_frame=0
         channels [
          Vector.X {type=MovieSceneScriptingDoubleChannel, keys [[8,28],[20,0]]}
          Vector.Y {type=MovieSceneScriptingDoubleChannel, const [0,1]}
         ]
        }]
      }
      {
       type=MovieSceneMarginTrack
       display_name=Padding
       sections [{
         type=MovieSceneMarginSection
         end_frame=20
         start_frame=-1
         channels [
          Left {type=MovieSceneScriptingFloatChannel, keys [[-1,-56],[20,0]]}
          Top {type=MovieSceneScriptingFloatChannel}
          Right {type=MovieSceneScriptingFloatChannel}
          Bottom {type=MovieSceneScriptingFloatChannel}
         ]
        }]
      }
     ]
    }
    PreviousButton {
     id=#3
     possessed_class=WBP_UIKit_Button_Generic_C
     tracks [
      {
       type=MovieSceneFloatTrack
       display_name=Render Opacity
       sections [{type=MovieSceneFloatSection, end_frame=18, start_frame=0, channels [None {type=MovieSceneScriptingFloatChannel, keys [[0,0],[13,0],[18,1]]}]}]
      }
      {
       type=MovieSceneByteTrack
       display_name=Visibility
       sections [{type=MovieSceneByteSection, end_frame=0, start_frame=-1, channels [None {type=MovieSceneScriptingByteChannel, keys [[-1,1],[0,4]]}]}]
      }
     ]
    }
    Text_Month {id=#4, possessed_class=CommonTextBlock, tracks []}
    GridSlot_1 {
     id=#5
     parent_id=#4
     possessed_class=GridSlot
     tracks [{
       type=MovieSceneMarginTrack
       display_name=Padding
       sections [{
         type=MovieSceneMarginSection
         end_frame=20
         start_frame=-1
         channels [
          Left {type=MovieSceneScriptingFloatChannel, keys [[-1,0],[20,18]]}
          Top {type=MovieSceneScriptingFloatChannel, const [0,1]}
          Right {type=MovieSceneScriptingFloatChannel, keys [[-1,0],[20,18]]}
          Bottom {type=MovieSceneScriptingFloatChannel, const [0,1]}
         ]
        }]
      }]
    }
    GridSlot_3 {
     id=#6
     parent_id=#1
     possessed_class=GridSlot
     tracks [
      {
       type=MovieSceneDoubleVectorTrack
       display_name=Nudge
       sections [{
         type=MovieSceneDoubleVectorSection
         end_frame=20
         start_frame=0
         channels [
          Vector.X {type=MovieSceneScriptingDoubleChannel, keys [[0,-28],[8,-28],[20,0]]}
          Vector.Y {type=MovieSceneScriptingDoubleChannel}
         ]
        }]
      }
      {
       type=MovieSceneMarginTrack
       display_name=Padding
       sections [{
         type=MovieSceneMarginSection
         end_frame=12
         start_frame=0
         channels [
          Left {type=MovieSceneScriptingFloatChannel, const [0,2]}
          Top {type=MovieSceneScriptingFloatChannel}
          Right {type=MovieSceneScriptingFloatChannel, keys [[0,-56],[11,0]]}
          Bottom {type=MovieSceneScriptingFloatChannel}
         ]
        }]
      }
     ]
    }
   ]}
 }]
graphs [
 EventGraph {nodes [
   n1 = "On Clicked (PreviousButton)"!() -> (OutputDelegate, then, Button:Common Button Base Object Reference) {is=title K2Node_ComponentBoundEvent}
   > Fire Event Element Interacted(Element=@n1.Button)
   > Call On Clicked Previous Button!(execute:Exec, self:WBP Calendar Wrapper Object Reference) -> (then:Exec) {is=title K2Node_CallDelegate}
   n2 = "On Clicked (NextButton)"!() -> (OutputDelegate, then, Button:Common Button Base Object Reference) {is=title K2Node_ComponentBoundEvent}
   > Fire Event Element Interacted(Element=@n2.Button)
   > Call On Clicked Next Button!(execute:Exec, self:WBP Calendar Wrapper Object Reference) -> (then:Exec) {is=title K2Node_CallDelegate}
   Event PreConstruct()
   > n3 = Sequence!(execute) -> (then_0, then_1, then_2:Exec)
   > SetRenderTransform!(
    execute
    self:Widget Object Reference=Get Block_Icon!(self:WBP UIKit Button Generic Base Object Reference=Get PreviousButton() -> (PreviousButton:WBP UIKit Button Generic Object Reference)) -> (Block_Icon:WBP UIKit Block Image Object Reference)
    InTransform_Translation:Vector 2D Structure=(X=0,Y=0)
    InTransform_Scale:Vector 2D Structure=(X=1,Y=1)
    InTransform_Shear:Vector 2D Structure=(X=0,Y=0)
    InTransform_Angle:float=180
   ) -> (then)
   n4 = Get CommonInputSubsystem()
   > SetBlockVisibility!(
    execute
    self:WBP UIKit Block Base Object Reference=[Get Block_Icon(self=Get NextButton() -> (NextButton:WBP UIKit Button Generic Object Reference)), Get Block_Icon(self=Get PreviousButton())]
    Visible:bool=Select!(MouseAndKeyboard:bool=true, Gamepad:bool=false, Touch:bool=true, Count:bool=true, Index:ECommonInputType Enum=GetCurrentInputType(self=@n4)) -> (ReturnValue:bool)
   ) -> (then)
   UpdateButtonsByInputType!(execute=@n3.then_1, self, InputType:ECommonInputType Enum=GetCurrentInputType(self=Get CommonInputSubsystem())) -> (then)
   > Bind Event to On Input Method Changed(self=Get CommonInputSubsystem(), Delegate=Create Event())
   > SetCompactUIStyle() {is=title K2Node_CallFunction}
  ]}
 GetNextButtonDefaultColumn {nodes [
   GetNextButtonDefaultColumn() {is=title K2Node_FunctionEntry}
   > Return Node!(execute, Column:int=SelectInt(A=3, B=1, bPickA=IsMobileGame()))
  ]}
 OnClickedNextButton {nodes [OnClickedNextButton() {is=title K2Node_FunctionEntry}]}
 OnClickedPreviousButton {nodes [OnClickedPreviousButton() {is=title K2Node_FunctionEntry}]}
 PlayButtonAnimation {nodes [
   n1 = PlayButtonAnimation!() -> (then, IsShow:bool=~) {is=title K2Node_FunctionEntry}
   > PlayAnimation(
    InAnimation=Get Anim_ShowButtons() -> (Anim_ShowButtons:Widget Animation Object Reference)
    PlayMode=Select!(Option 0:EUMGSequencePlayMode Enum=Reverse, Option 1:EUMGSequencePlayMode Enum, Index:bool=@n1) -> (ReturnValue:EUMGSequencePlayMode Enum)
    PlaybackSpeed=3
   )
  ]}
 SetCompactUIStyle {nodes [
   SetCompactUIStyle() {is=title K2Node_FunctionEntry}
   > Branch(Condition=IsMobileGame())
   > SetIconSize!(execute, self:WBP UIKit Button Generic Base Object Reference=[Get NextButton(), Get PreviousButton()], Icon:Vector 2D Structure=(X=80,Y=80)) -> (then)
  ]}
 SetText {nodes [
   n1 = SetText!() -> (then, InText:Text) {is=title K2Node_FunctionEntry}
   > "SetText (Text)"(self=Get Text_Month() -> (Text_Month:Common Text Object Reference), InText=@n1)
  ]}
 UpdateButtonsByInputType {nodes [
   n1 = UpdateButtonsByInputType!() -> (then, InputType:ECommonInputType Enum) {is=title K2Node_FunctionEntry}
   > SetColumn!(
    execute
    self:Grid Slot Object Reference=SlotAsGridSlot(Widget=Get NextButton())
    InColumn:int=SelectInt(A=3, B=GetNextButtonDefaultColumn!() -> (Column:int=0) {is=title K2Node_CallFunction}, bPickA="Equal (Enum)"(A=@n1, B=Gamepad))
   ) -> (then)
   > SetBlockVisibility(
    self=[Get Block_Icon(self=Get PreviousButton()), Get Block_Icon(self=Get NextButton())]
    Visible=Select!(MouseAndKeyboard:bool=true, Gamepad:bool=false, Touch:bool=true, Count:bool=true, Index:ECommonInputType Enum=@n1) -> (ReturnValue:bool)
   )
  ]}
]
variables [
 OnClickedNextButton {}
 OnClickedPreviousButton {}
]
widgets {
 info {inheritedWidgetCount=0, parentClass=&CommonUserWidget, rootWidgetClass=&GridPanel, widgetCount=4}
 widgets [
  Root {widgetClassPath=&GridPanel, widget=Root}
  Text_Month {
   widgetClassPath=&CommonTextBlock
   bIsVariable=true
   parent=Root
   widget=Text_Month
   properties {
    justification=Center
    text=Month
    textTransformPolicy=ToUpper
    font {
     bForceMonospaced=false
     bMaterialIsStencil=false
     fontMaterial=null
     fontObject=&HeadingNow
     letterSpacing=-40
     monospacedWidth=1
     size=32.25
     skewAmount=0
     typefaceFontName=86 Bold Italic
     outlineSettings {
      bApplyOutlineToDropShadows=false
      bMiteredCorners=false
      bSeparateFillAlpha=false
      outlineMaterial=null
      outlineSize=0
      outlineColor=<rgba 0 0 0 1>
     }
    }
    shadowOffset=<0 0>
   }
   slot {class=GridSlot, properties {column=2, verticalAlignment=VAlign_Center}}
  }
  PreviousButton {
   widgetClassPath=&WBP_UIKit_Button_Generic
   bIsVariable=true
   parent=Root
   widget=PreviousButton
   properties {
    bDisplayInputActionWhenNotInteractable=true
    bHideInputActionWithKeyboard=true
    bIsFocusable=false
    iconMaterial=/FortUILibrary/Textures/MaterialInstances/MI_UIKit_Icon_Chevron
    mobileMinHeight=80
    mobileMinWidth=80
    priority=100
    promptHorizontalAlignment=HAlign_Center
    useBackgroundBlock=false
    useTextBlock=false
    visibility=Collapsed
    buttonPadding=<ltrb 8 0 8 0>
    promptNudge=<0 0>
    triggeringInputAction {dataTable=&InputActionDataTable, rowName=GenericLeftTrigger}
   }
   slot {class=GridSlot, properties {horizontalAlignment=HAlign_Right}}
  }
  NextButton {
   widgetClassPath=&WBP_UIKit_Button_Generic
   bIsVariable=true
   parent=Root
   widget=NextButton
   properties {
    bDisplayInputActionWhenNotInteractable=true
    bHideInputActionWithKeyboard=true
    bIsFocusable=false
    iconMaterial=&MI_UIKit_Icon_Chevron
    mobileMinHeight=80
    mobileMinWidth=80
    priority=100
    promptHorizontalAlignment=HAlign_Center
    useBackgroundBlock=false
    useTextBlock=false
    visibility=Collapsed
    buttonPadding=<ltrb 8 0 8 0>
    promptNudge=<0 0>
    triggeringInputAction {dataTable=&InputActionDataTable, rowName=GenericRightTrigger}
   }
   slot {class=GridSlot, properties {column=1, horizontalAlignment=HAlign_Left}}
  }
 ]
}
