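# WBP_FilterButton is a UIKit generic button subclass for the Compete tournament calendar that summarizes the active filters.
# PreConstruct rearranges inherited grid children to add a status indicator beside the prompt, sets the background material, and adjusts sizing on mobile.
# Update Filter Info reads the calendar filter view model, labels the button with the first enabled region or "None", and writes the filter count into the indicator.
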WidgetBlueprint /Compete/Calendar/WBP_FilterButton
graphs [
 EventGraph {nodes [
   n1 = Event PreConstruct()
   > Parent: PreConstruct(IsDesignTime=@n1.IsDesignTime)
   > Set Button Design() {is=title K2Node_CallFunction}
   > Branch(Condition=IsMobileGame())
   > SetIconSize!(execute, self:WBP UIKit Button Generic Base Object Reference, Icon:Vector 2D Structure=(X=50,Y=50)) -> (then)
   > SetButtonMinSize!(execute, self:WBP UIKit Button Generic Base Object Reference, Min Width:int=0, Min Height:int=88) -> (then)
   > SetGridSlotPadding!(
    execute
    Widget:Widget Object Reference=Get Main() -> (Main:Grid Panel Object Reference)
    Padding_Left:float=15
    Padding_Top:float=0
    Padding_Right:float=15
    Padding_Bottom:float=0
   ) -> (then)
  ]}
 Set Button Design {nodes [
   Set Button Design() {is=title K2Node_FunctionEntry}
   > AddChildToGrid!(
    execute
    self:Grid Panel Object Reference=Get Main()
    Content:Widget Object Reference=Get Prompt() -> (Prompt:Grid Panel Object Reference)
    InRow:int=1
    InColumn:int=0
   ) -> (then, ReturnValue:Grid Slot Object Reference)
   > SetPadding!(
    execute
    self:Grid Slot Object Reference=SlotAsGridSlot(Widget=Get Prompt())
    InPadding_Left:float=0
    InPadding_Top:float=0
    InPadding_Right:float=10
    InPadding_Bottom:float=0
   ) -> (then)
   > AddChildToGrid(self=Get Main(), Content=Get StatusIndicator() -> (StatusIndicator:WBP UIKit Status Indicator Object Reference), InRow=1, InColumn=1)
   > UpdateStatusIndicator(self:WBP UIKit Button Generic Base Object Reference)
   > Set Bang State!(execute, self:Fort Bang Wrapper NUI Object Reference=Get StatusIndicator(), bEnabled:bool=true, Count:int=0) -> (then) {is=title K2Node_CallFunction}
   > Set Column!(
    execute
    Column:int=2
    self:Grid Slot Object Reference=SlotAsGridSlot(Widget=Get Block_Icon() -> (Block_Icon:WBP UIKit Block Image Object Reference))
   ) -> (then, Output_Get:int=0)
   > SetVerticalAlignment(
    self:Grid Slot Object Reference=SlotAsGridSlot(Widget=Get Block_Text() -> (Block_Text:WBP UIKit Block Text Object Reference))
    InVerticalAlignment=VAlign_Center
   )
   > SetBackgroundMaterial!(
    execute
    self:WBP UIKit Block Background Base Object Reference=Get Block_Background() -> (Block_Background:WBP UIKit Block Background Base Object Reference)
    Material Instance:"Material Instance Object Reference (by ref)"=Get Material() -> (Material:Material Instance Object Reference)
   ) -> (then)
  ]}
 Update Filter Info {nodes [
   Update Filter Info() {is=title K2Node_FunctionEntry}
   > n1 = Sequence()
   > SetText!(execute, self:WBP UIKit Button Generic Base Object Reference, Text:Text="LOCTEXT(\"None\")") -> (then)
   n2 = Get CalendarFilterVM!(self:Fort Compete UIManager Object Reference=Get FortCompeteUIManager() -> (ReturnValue:Fort Compete UIManager Object Reference) {is=title K2Node_GetSubsystem}) -> (CalendarFilterVM:Tournaments Calendar Filter VM Object Reference)
   > n3 = GetSectionRegionValues!(execute, self:Tournaments Calendar Filter VM Object Reference=@n2) -> (then, ReturnValue:Map of ERegionType Enums to Booleans=None)
   > n4 = Keys!(execute, TargetMap:Map of ERegionType Enums to Booleans=@n3) -> (then, Keys:Array of ERegionType Enums)
   > n5 = For Each Loop with Break!(Exec:Exec, Array:Array of ERegionType Enums=@n4, Break:Exec=@n7.then) -> (LoopBody:Exec, Array Element:ERegionType Enum, Array Index:int, Completed:Exec)
   > Branch(Condition=Find(TargetMap:Map of ERegionType Enums to Booleans=@n3, Key:"ERegionType Enum (by ref)"=@n5) -> (Value:bool))
   > n6 = ConvertRegionTypeToId!(
    execute
    InRegionType:ERegionType Enum=@n5
    None:%=Get FortCompeteRegionSubsystem() -> (ReturnValue:Fort Compete Region Subsystem Object Reference) {is=title K2Node_GetSubsystem}
   ) -> (then, ReturnValue:String)
   > n7 = SetText(Text="To Text (String)"(InString=@n6))
   n8 = GetFilterCount!(execute=@n1.then_1, self:Tournaments Calendar Filter VM Object Reference=@n2) -> (then, ReturnValue:int=0)
   n9 = Get StatusIndicator()
   > Is Valid(InputObject=@n9)
   > "SetText (Text)"(
    self=Get Indicator_Text!(self:WBP UIKit Status Indicator Base Object Reference=@n9) -> (Indicator_Text:Common Text Object Reference)
    InText="To Text (Integer)"(Value=@n8)
   )
  ]}
]
variables []
widgets {
 info {
  inheritedWidgetCount=0
  parentClass=/FortUILibrary/UIKit/Buttons/WBP_UIKit_Button_Generic.WBP_UIKit_Button_Generic_C
  rootWidgetClass=null
  widgetCount=0
 }
 widgets []
}
