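# WBP_TournamentTile is the compete calendar list tile showing a tournament's poster, date, region, type tag, live status, rank icon and bookmarks.
# It takes a calendar tile view model from the list item, triggers UpdateData and exposes setters for each element.
# SetEligibility converts requirement checks into a lock switcher state and reason text, re-evaluated on input-method changes.
# Clicking fires analytics and sets details-screen parameters; selection animates tile size.
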WidgetBlueprint /Compete/TournamentTile/WBP_TournamentTile
animations [Anim_Selected {
  binding_count=1
  length_seconds=1
  object=&self:Anim_Selected
  sequence {bindings [SizeBox_Anim {
     id=#1
     possessed_class=SizeBox
     tracks [
      {
       type=MovieSceneFloatTrack
       display_name=Height Override
       sections [
        {type=MovieSceneFloatSection, end_frame=20, start_frame=0, channels [None {type=MovieSceneScriptingFloatChannel, keys [[0,512],[20,568]]}]}
        {type=MovieSceneFloatSection, end_frame=20, start_frame=0, channels [None {type=MovieSceneScriptingFloatChannel, keys [[0,380],[20,420]]}]}
       ]
      }
      {
       type=MovieSceneFloatTrack
       display_name=Width Override
       sections [
        {type=MovieSceneFloatSection, end_frame=20, start_frame=0, channels [None {type=MovieSceneScriptingFloatChannel, keys [[0,512],[20,568]]}]}
        {type=MovieSceneFloatSection, end_frame=20, start_frame=0, channels [None {type=MovieSceneScriptingFloatChannel, keys [[0,380],[20,420]]}]}
       ]
      }
     ]
    }]}
 }]
graphs [
 DebugWidget {nodes [
   DebugWidget() {is=title K2Node_FunctionEntry}
   n1 = GetObjectName(Object=Self-Reference())
   n2 = Enum to String!(Enumerator:EFortCompeteTournamentWindowEligibility Enum=Get TournamentEligibility!(self:Calendar Tournament Tile VM Object Reference=Get FortPoblanoTournamentsCalendarTournamentTileVM() -> (FortPoblanoTournamentsCalendarTournamentTileVM:Calendar Tournament Tile VM Object Reference)) -> (TournamentEligibility:EFortCompeteTournamentWindowEligibility Enum=Unknown)) -> (ReturnValue:String)
   n3 = Get DEBUG_TEXT() -> (DEBUG_TEXT:Common Text Object Reference)
   > "SetText (Text)"(self=@n3, InText="To Text (String)"(InString=Append!(A=@n1, B="\r\n", C:String=@n2) -> (ReturnValue)))
   > SetVisibility(self=@n3, InVisibility=HitTestInvisible)
  ]}
 EventGraph {nodes [
   n1 = OnInputChanged!() -> (OutputDelegate, then, bNewInputType:ECommonInputType Enum) {is=title K2Node_CustomEvent}
   > n2 = UpdateData(self:Calendar Tournament Tile VM Object Reference=Get FortPoblanoTournamentsCalendarTournamentTileVM())
   SetEligibility!(
    execute=[@n15.then, @n6.then, @n2.then]
    self
    Eligibility:EFortCompeteTournamentWindowEligibility Enum=Get TournamentEligibility(self=Get FortPoblanoTournamentsCalendarTournamentTileVM()) -> (TournamentEligibility=Unknown)
   ) -> (then)
   n3 = Event OnListItemObjectSet()
   > n4 = Cast To FortPoblanoTournamentsCalendarTournamentTileVM!(execute, Object=@n3.ListItemObject) -> (then, CastFailed, AsCalendar Tournament Tile VM:Calendar Tournament Tile VM Object Reference) {is=title K2Node_DynamicCast}
   > n5 = Set FortPoblanoTournamentsCalendarTournamentTileVM(FortPoblanoTournamentsCalendarTournamentTileVM:Calendar Tournament Tile VM Object Reference=@n4) -> (Output_Get:Calendar Tournament Tile VM Object Reference)
   > UpdateData(self=@n5)
   > n6 = SetRequirementRankTile()
   Event On Clicked()
   n7 = Make Flow Analytics Additional Info!() -> (FlowAnalyticsAdditionalInfo:Flow Analytics Additional Info Structure) {is=title K2Node_MakeStruct}
   n8 = Get FortPoblanoTournamentsCalendarTournamentTileVM()
   n9 = Get TournamentID!(self:Calendar Tournament Tile VM Object Reference=@n8) -> (TournamentID:String)
   > Fire Event Element Interacted(
    Element=Self-Reference()
    AdditionalInfo="Add String (Non-Localized)"!(
     Info:"Flow Analytics Additional Info Structure (by ref)"="Add String (Non-Localized)"(Info=@n7, Key=TournamentID, Value=@n9)
     Key:String=TournamentWindowID
     Value:String=Get TournamentWindowId!(self:Calendar Tournament Tile VM Object Reference=@n8) -> (TournamentWindowId:String)
    ) -> (ReturnValue:Flow Analytics Additional Info Structure)
   )
   > SetCanResetCalendarFilterVM!(
    execute
    self:Fort Compete UIManager Object Reference=Get FortCompeteUIManager() -> (ReturnValue:Fort Compete UIManager Object Reference) {is=title K2Node_GetSubsystem}
    bInCanResetCalendarFilter:bool=false
   ) -> (then)
   n10 = Get FortPoblanoTournamentsCalendarTournamentTileVM()
   > SetTournamentDetailsScreenParams!(
    execute
    self:Tournaments Carousel VM Object Reference=Get FortPoblanoTournamentsCarouselVM() -> (FortPoblanoTournamentsCarouselVM:Tournaments Carousel VM Object Reference)
    InTournamentDetailsScreenParams:"Tournament Details Screen Params Structure (by ref)"=Make Tournament Details Screen Params!(
     ContextPC:Player Controller Object Reference=GetOwningPlayer()
     TournamentId:String=Get TournamentID(self=@n10)
     TournamentWindowId:String=Get TournamentWindowId(self=@n10)
    ) -> (TournamentDetailsScreenParams:Tournament Details Screen Params Structure) {is=title K2Node_MakeStruct}
   ) -> (then)
   n11 = Event PreConstruct()
   > Parent: PreConstruct(IsDesignTime=@n11.IsDesignTime)
   n12 = Get Tag_TournamentType() -> (Tag_TournamentType:WBP UIKit Tag Object Reference)
   > Set!(
    execute
    Size_Value:float=1
    Size_SizeRule:ESlateSizeRule Enum=Fill
    self:Stack Box Slot Object Reference=SlotAsStackBoxSlot!(Widget:Widget Object Reference=Get Text_Label!(self:WBP UIKit Tag Base Object Reference=@n12) -> (Text_Label:Common Text Object Reference)) -> (ReturnValue:Stack Box Slot Object Reference)
   ) -> (then, Output_Get:Slate Child Size Structure)
   > n13 = Sequence()
   > RegisterModularBlocks()
   > SetTextTransformPolicy!(execute, self:Text Object Reference=Get Text_Label(self=Get Tag_TournamentType()), InTransformPolicy:ETextTransformPolicy Enum=None) -> (then)
   > n14 = GetDynamicMaterial(self=Get Image_PosterFront() -> (Image_PosterFront:Epic CMSImage Object Reference))
   > Set MID_Poster(MID_Poster:Material Instance Dynamic Object Reference=@n14) -> (Output_Get:Material Instance Dynamic Object Reference)
   > ShowLockStatus!(execute, self, ShowLock:E_LockStatus Enum=NewEnumerator0) -> (then)
   > Delay(Duration=1)
   > n15 = Assign On Input Method Changed!(execute:Exec, self:Common Input Object Reference=Get CommonInputSubsystem(), Delegate:"Delegate (by ref)"=@n1) -> (then:Exec) {is=title K2Node_AssignDelegate}
   SetMobileStyle(execute=@n13.then_1) {is=title K2Node_CallFunction}
   Event On Selected()
   > Parent: On Selected() {is=title K2Node_CallParentFunction}
   n16 = float / float(
    A=1
    B=Break UIKit Block Timing!(UIKitBlockTiming:"UIKit Block Timing Structure (by ref)"=GetBlockTiming!() -> (BlockTiming:UIKit Block Timing Structure)) -> (Selecting:float=~) {is=title K2Node_BreakStruct}
   )
   n17 = Get Anim_Selected() -> (Anim_Selected:Widget Animation Object Reference)
   > PlayAnimation(InAnimation=@n17, PlaybackSpeed=@n16)
   Event On Deselected()
   > Parent: On Deselected() {is=title K2Node_CallParentFunction}
   > PlayAnimation(InAnimation=@n17, PlayMode=Reverse, PlaybackSpeed=@n16)
  ]}
 GetRankRequirementStats {nodes [
   GetRankRequirementStats() {is=title K2Node_FunctionEntry}
   n1 = Get!(self:Calendar Tournament Tile VM Object Reference=Get FortPoblanoTournamentsCalendarTournamentTileVM()) -> (RankInfo_ConditionalModes:Array of Tournament Conditional Mode Requirement Structures)
   > n2 = Branch(Condition=Is Empty!(TargetArray:Array of Tournament Conditional Mode Requirement Structures=@n1) -> (ReturnValue:bool))
   > Return Node!(execute, Level:int=0, ReturnValue:String)
   n3 = "Get (a ref)"!(Array:Array of Tournament Conditional Mode Requirement Structures=@n1, Dimension 1=0) -> (Output_Type:String, Output_Mode:String, Output_Level:int)
   Return Node(execute=@n2.else, Level=@n3.Output_Level, ReturnValue=@n3.Output_Mode)
  ]}
 RegisterModularBlocks {nodes [
   RegisterModularBlocks() {is=title K2Node_FunctionEntry}
   > RegisterModularBlock!(
    execute
    self:UIKit Modular Button Object Reference
    Block:UIKit Block Object Reference=Get Block_Outline() -> (Block_Outline:WBP UIKit Block Outline Object Reference)
   ) -> (then)
  ]}
 SelectTournamentLocksText {nodes [
   SelectTournamentLocksText() {is=title K2Node_FunctionEntry}
   n1 = Break Fort Tournament Lock Status!(FortTournamentLockStatus:"Fort Tournament Lock Status Structure (by ref)"=Get AllTournamentLocks!(self:Calendar Tournament Tile VM Object Reference=Get FortPoblanoTournamentsCalendarTournamentTileVM()) -> (AllTournamentLocks:Fort Tournament Lock Status Structure)) -> (Region:EFortTournamentLockStatus Enum, Account:EFortTournamentLockStatus Enum, Team:EFortTournamentLockStatus Enum) {is=title K2Node_BreakStruct}
   > n2 = Branch(Condition="Equal (Enum)"(A:EFortTournamentLockStatus Enum=@n1.Account, B:EFortTournamentLockStatus Enum=Failed))
   > Return Node!(execute, LockText:Text="LOCTABLE(\"&/ST_RequirementsTexts\", \"TounamentLocksAccountRequired\")")
   n3 = Branch(execute=@n2.else, Condition="Equal (Enum)"(A=@n1.Team, B=Failed))
   > Return Node(LockText="LOCTABLE(\"&ST_RequirementsTexts\", \"TounamentLocksTeamRequired\")")
   n4 = Branch(execute=@n3.else, Condition="Equal (Enum)"(A=@n1, B=Failed))
   > Return Node(LockText="LOCTABLE(\"&ST_RequirementsTexts\", \"TounamentLocksRegionRequired\")")
   Return Node(execute=@n4.else, LockText="INVTEXT(\"[PH] Mystery Requirement\")")
  ]}
 SetBookmarkStatus {nodes [
   n1 = SetBookmarkStatus!() -> (then, IsBookmarked:bool=~) {is=title K2Node_FunctionEntry}
   > n2 = Branch(Condition=@n1)
   n3 = Get BookmarkStatus() -> (BookmarkStatus:Grid Panel Object Reference)
   > SetVisibility(self=@n3, InVisibility=SelfHitTestInvisible)
   SetVisibility(execute=@n2.else, self=@n3, InVisibility=Collapsed)
  ]}
 SetBookmarksAmount {nodes [
   n1 = SetBookmarksAmount!() -> (then, InText:Text) {is=title K2Node_FunctionEntry}
   > "SetText (Text)"(self=Get CommonText_BookmarksAmount() -> (CommonText_BookmarksAmount:Common Text Object Reference), InText=@n1)
   > SetVisibility(
    self=Get GridPanel_Bookmarks() -> (GridPanel_Bookmarks:Grid Panel Object Reference)
    InVisibility=Select(Option 0=SelfHitTestInvisible, Option 1=Collapsed, Index=TextIsEmpty(InText=@n1))
   )
  ]}
 SetDate {nodes [
   n1 = SetDate!() -> (then, Text:Text) {is=title K2Node_FunctionEntry}
   > "SetText (Text)"(self=Get CommonText_Date() -> (CommonText_Date:Common Text Object Reference), InText=@n1)
  ]}
 SetEligibility {nodes [
   SetEligibility!() -> (then, Eligibility:EFortCompeteTournamentWindowEligibility Enum=Unknown) {is=title K2Node_FunctionEntry}
   > n1 = Branch(Condition=NofNEligibilityRequirementEnabled!(self:Fort Compete UIManager Object Reference=Get FortCompeteUIManager()) -> (ReturnValue:bool))
   n2 = Get FortPoblanoTournamentsCalendarTournamentTileVM()
   n3 = Get EligibilityRequirementCount!(self:Calendar Tournament Tile VM Object Reference=@n2) -> (EligibilityRequirementCount:int=0)
   n4 = Get EligibilityRequirementScore!(self:Calendar Tournament Tile VM Object Reference=@n2) -> (EligibilityRequirementScore:int=0)
   > n5 = Branch(Condition="Equal (Integer)"(A=@n3, B=@n4))
   > ShowLockStatus(ShowLock=NewEnumerator2)
   Set RequirementText(
    execute=@n5.else
    RequirementText:Text=Format Text!(Format="LOCTEXT(\"{Score}/{Count} Requirements met\")", Score:int=@n4, Count:int=@n3) -> (Result)
   ) -> (Output_Get:Text)
   > ShowLockStatus(ShowLock=NewEnumerator1)
   n6 = Get RequirementText() -> (RequirementText:Text)
   > n7 = Branch(Condition=TextIsEmpty(InText=@n6))
   "SetText (Text)"(
    execute=@n7.else
    self=Get CommonText_EligibilityRequirement() -> (CommonText_EligibilityRequirement:Common Text Object Reference)
    InText=@n6
   )
   n8 = Branch(
    execute=@n1.else
    Condition=Get bMeetsTournamentSystemRequirements!(self:Calendar Tournament Tile VM Object Reference=Get FortPoblanoTournamentsCalendarTournamentTileVM()) -> (bMeetsTournamentSystemRequirements:bool)
   )
   > n9 = Branch(Condition=Get bMeetsPermanentTournamentBanRequirement!(self:Calendar Tournament Tile VM Object Reference=Get FortPoblanoTournamentsCalendarTournamentTileVM()) -> (bMeetsPermanentTournamentBanRequirement:bool))
   n10 = Get RankRequirementCheckResult!(self:Calendar Tournament Tile VM Object Reference=Get FortPoblanoTournamentsCalendarTournamentTileVM()) -> (RankRequirementCheckResult:EFortCompeteHabaneroRankRequirementCheckResult Enum=Invalid)
   > n11 = Branch(Condition=OR Boolean(
     A="Equal (Enum)"(A:EFortCompeteHabaneroRankRequirementCheckResult Enum=@n10, B:EFortCompeteHabaneroRankRequirementCheckResult Enum=Valid)
     B="Equal (Enum)"(A=@n10, B=Invalid)
    ))
   > n12 = Branch(Condition=Get bMeetsRequiresQualifyingRound!(self:Calendar Tournament Tile VM Object Reference=Get FortPoblanoTournamentsCalendarTournamentTileVM()) -> (bMeetsRequiresQualifyingRound:bool))
   > n13 = Branch(Condition=Get bMeetsAgeRequirement!(self:Calendar Tournament Tile VM Object Reference=Get FortPoblanoTournamentsCalendarTournamentTileVM()) -> (bMeetsAgeRequirement:bool))
   > n14 = Branch(Condition=Get bMeetsAccountLevelRequirement!(self:Calendar Tournament Tile VM Object Reference=Get FortPoblanoTournamentsCalendarTournamentTileVM()) -> (bMeetsAccountLevelRequirement:bool))
   > n15 = Branch(Condition=Get bMeetsCheckTournamentLocksRequirement!(self:Calendar Tournament Tile VM Object Reference=Get FortPoblanoTournamentsCalendarTournamentTileVM()) -> (bMeetsCheckTournamentLocksRequirement:bool))
   > n16 = Branch(Condition=Get bMeetsTimeAliveRequirement!(self:Calendar Tournament Tile VM Object Reference=Get FortPoblanoTournamentsCalendarTournamentTileVM()) -> (bMeetsTimeAliveRequirement:bool))
   > n17 = Branch(Condition=Get bMeetsInputTypeRequirement!(self:Calendar Tournament Tile VM Object Reference=Get FortPoblanoTournamentsCalendarTournamentTileVM()) -> (bMeetsInputTypeRequirement:bool))
   > n18 = Branch(Condition=IsTournamentHistoryRequirementEnabled!(self:Fort Compete UIManager Object Reference=Get FortCompeteUIManager()) -> (ReturnValue:bool))
   > n19 = Branch(Condition=Get bMeetsEventsNumberHistoryRequirement!(self:Calendar Tournament Tile VM Object Reference=Get FortPoblanoTournamentsCalendarTournamentTileVM()) -> (bMeetsEventsNumberHistoryRequirement:bool))
   ShowLockStatus(execute=[@n18.else, @n19.then], ShowLock=NewEnumerator2)
   n20 = Set RequirementText(execute=@n19.else, RequirementText="LOCTEXT(\"Past Event Count Not Met\")")
   ShowLockStatus(
    execute=[@n23.then, @n20.then, @n24.then, @n27.then, @n26.then, @n28.then, @n29.then, @n31.then, @n33.then, @n32.then]
    ShowLock=NewEnumerator1
   )
   n21 = Get RequirementText()
   > n22 = Branch(Condition=TextIsEmpty(InText=@n21))
   "SetText (Text)"(execute=@n22.else, self=Get CommonText_EligibilityRequirement(), InText=@n21)
   n23 = Set RequirementText(execute=@n17.else, RequirementText="LOCTABLE(\"&ST_RequirementsTexts\", \"InputTypeRequired\")")
   n24 = Set RequirementText(execute=@n16.else, RequirementText="LOCTEXT(\"Time Alive Not Met\")")
   n25 = SelectTournamentLocksText!(execute=@n15.else, self) -> (then, LockText:Text) {is=title K2Node_CallFunction}
   > n26 = Set RequirementText(RequirementText=@n25)
   n27 = Set RequirementText(execute=@n14.else, RequirementText="LOCTABLE(\"&ST_RequirementsTexts\", \"AccountLevelRequired\")")
   n28 = Set RequirementText(execute=@n13.else, RequirementText="LOCTABLE(\"&ST_RequirementsTexts\", \"AgeRequired\")")
   n29 = Set RequirementText(execute=@n12.else, RequirementText="LOCTABLE(\"&ST_RequirementsTexts\", \"QualifyingTokensRequired\")")
   n30 = MakeLiteralText(Value="LOCTABLE(\"&ST_RequirementsTexts\", \"HighRankRequired\")")
   n31 = Set RequirementText(
    execute=@n11.else
    RequirementText=Select!(
     Invalid:Text
     Valid:Text
     PlayerIsUnhabaneroed:Text=@n30
     PlayerCurrentRankBelowMinimumRequirement:Text=@n30
     PlayerHighestRankBelowMinimumRequirement:Text=@n30
     PlayerCurrentRankAboveMaximumRequirement:Text=MakeLiteralText(Value="LOCTABLE(\"&ST_RequirementsTexts\", \"LowerRankRequired\")")
     PlayerIsIneligible:Text=@n30
     PlayerTrackRankBelowMinimumRequirement:Text
     PlayerTrackRankAboveMaximumRequirement:Text
     Index:EFortCompeteHabaneroRankRequirementCheckResult Enum=@n10
    ) -> (ReturnValue:Text)
   )
   n32 = Set RequirementText(execute=@n9.else, RequirementText="LOCTABLE(\"&ST_RequirementsTexts\", \"TournamentBan\")")
   n33 = Set RequirementText(execute=@n8.else, RequirementText="LOCTABLE(\"&ST_RequirementsTexts\", \"SystemRequirements\")")
  ]}
 SetMobileStyle {nodes [
   SetMobileStyle() {is=title K2Node_FunctionEntry}
   n1 = IsTabletAspectRatio!() -> (ReturnValue:bool)
   > Branch(Condition=AND Boolean(A=IsUsingCompactUI(), B=NOT Boolean(A=@n1)))
   n2 = Get SizeBox_Anim() -> (SizeBox_Anim:Size Box Object Reference)
   > SetWidthOverride(self=@n2, InWidthOverride=380)
   > SetHeightOverride(self=@n2, InHeightOverride=380)
   > SetMobileFontSizeMultiplier!(
    execute
    self:Common Text Object Reference=Get CommonText_TournamentName() -> (CommonText_TournamentName:Common Text Object Reference)
    InMobileFontSizeMultiplier:float=1
   ) -> (then)
  ]}
 SetPosterColor {nodes [
   n1 = SetPosterColor!() -> (then, Colors_Main:Linear Color Structure, Colors_Sub:Linear Color Structure) {is=title K2Node_FunctionEntry}
   > SetVectorParameterValue(self=Get MID_Poster() -> (MID_Poster:Material Instance Dynamic Object Reference), ParameterName=GradientColor, Value=@n1.Colors_Sub)
  ]}
 SetRegion {nodes [
   n1 = SetRegion!() -> (then, Region:ERegionType Enum=None) {is=title K2Node_FunctionEntry}
   n2 = Find(
    TargetMap:Map of ERegionType Enums to Texts=Get Regions() -> (Regions:Map of ERegionType Enums to Texts=None)
    Key:"ERegionType Enum (by ref)"=@n1
   ) -> (Value:Text)
   > Branch(Condition=@n2.ReturnValue)
   > Branch(Condition=NOT Boolean(A=TextIsEmpty(InText=@n2)))
   > "SetText (Text)"(self=Get CommonText_Region() -> (CommonText_Region:Common Text Object Reference), InText=@n2)
  ]}
 SetRequirementRankTile {nodes [
   SetRequirementRankTile() {is=title K2Node_FunctionEntry}
   n1 = GetRankRequirementStats!(self) -> (Level:int=0, ReturnValue:String) {is=title K2Node_CallFunction}
   n2 = Get RankInfo() -> (RankInfo:Fort Habanero Display Data Object Reference)
   n3 = Break Fort Habanero Mode List!(FortHabaneroModeList:"Fort Habanero Mode List Structure (by ref)"=Find(
     TargetMap:Map of Strings to Fort Habanero Mode List Structures=Get EntryData!(self:Fort Habanero Display Data Object Reference=@n2) -> (EntryData:Map of Strings to Fort Habanero Mode List Structures)
     Key:"String (by ref)"=@n1.ReturnValue
    ) -> (Value:Fort Habanero Mode List Structure)) -> (TierList:Array of Fort Habanero Tier Structures, bShowTopLeaderboardIcon:bool=~, TopTierTopLeaderboardPositions:Fort Habanero Tier Structure) {is=title K2Node_BreakStruct}
   n4 = Is Valid Index!(TargetArray:Array of Fort Habanero Tier Structures=@n3, IndexToTest:int=@n1) -> (ReturnValue:bool)
   > SetVisibility(
    self=Get Image_RankIcon() -> (Image_RankIcon:Image Object Reference)
    InVisibility=Select!(Option 0:ESlateVisibility Enum=Collapsed, Option 1:ESlateVisibility Enum=SelfHitTestInvisible, Index:bool=@n4) -> (ReturnValue:ESlateVisibility Enum)
   )
   > Branch(Condition=@n4)
   > SetBrushFromSoftTexture!(
    execute
    self:Image Object Reference=Get Image_RankIcon()
    SoftTexture:Texture 2D Soft Object Reference=Break Fort Habanero Tier!(FortHabaneroTier:"Fort Habanero Tier Structure (by ref)"="Get (a copy)"(Array:Array of Fort Habanero Tier Structures=@n3, Dimension 1=@n1) -> (Output:Fort Habanero Tier Structure)) -> (
     ForegroundIcon:Material Instance Soft Object Reference
     ForegroundIconSmall:Material Instance Soft Object Reference
     BackgroundIcon:Material Instance Soft Object Reference
     BackgroundIconSmall:Material Instance Soft Object Reference
     WhiteIcon:Texture 2D Soft Object Reference
     SimpleIconSmall:Texture 2D Soft Object Reference
     Name:Text
     TierType:Fort Habanero Tier Type Data Object Reference
     AdditionalLeftIcon:Material Instance Soft Object Reference
     AdditionalRightIcon:Material Instance Soft Object Reference
    ) {is=title K2Node_BreakStruct}.SimpleIconSmall
    bMatchSize:bool=false
   ) -> (then)
  ]}
 SetRoundDateTime {nodes [
   n1 = SetRoundDateTime!() -> (then, InText:Text) {is=title K2Node_FunctionEntry}
   > n2 = Branch(Condition=TextIsEmpty(InText=@n1))
   > SetVisibility(self=Get CommonText_RoundDateTime() -> (CommonText_RoundDateTime:Common Text Object Reference), InVisibility=Collapsed)
   "SetText (Text)"(execute=@n2.else, self=Get CommonText_RoundDateTime(), InText=@n1)
   > SetVisibility(self=Get CommonText_RoundDateTime(), InVisibility=SelfHitTestInvisible)
  ]}
 SetTagIcon {nodes [
   n1 = SetTagIcon!() -> (then, TournamentType:ETournamentType Enum=None) {is=title K2Node_FunctionEntry}
   n2 = Find(
    TargetMap:Map of ETournamentType Enums to Material Instance Object References=Get TagIcons() -> (TagIcons:Map of ETournamentType Enums to Material Instance Object References=None)
    Key:"ETournamentType Enum (by ref)"=@n1
   ) -> (Value:Material Instance Object Reference)
   > Branch(Condition=@n2.ReturnValue)
   > Is Valid(InputObject=@n2)
   > SetIconMaterial!(execute, self:WBP UIKit Tag Base Object Reference=Get Tag_TournamentType(), IconMaterial:Material Instance Object Reference=@n2) -> (then)
  ]}
 SetTagName {nodes [
   n1 = SetTagName!() -> (then, Game Type:EFortGameType Enum=BR) {is=title K2Node_FunctionEntry}
   n2 = Select!(
    BR:Text="LOCTABLE(\"/Compete/StringTables/GameModeStringTable\", \"BRKey\")"
    ZeroBuild:Text="LOCTEXT(\"Zero Build\")"
    Creative:Text
    CreativeLTM:Text
    Playground:Text
    STW:Text
    BRArena:Text="LOCTABLE(\"&GameModeStringTable\", \"BRKey\")"
    BRLTM:Text
    Social:Text
    VKEdit:Text
    VKPlay:Text
    DelMar:Text
    Festival:Text
    FestivalMainStage:Text="LOCTABLE(\"&GameModeStringTable\", \"FestivalKey\")"
    FestivalJamStage:Text="LOCTABLE(\"&GameModeStringTable\", \"FestivalKey\")"
    Juno:Text
    FestivalBattleStage:Text="LOCTABLE(\"&GameModeStringTable\", \"FestivalKey\")"
    BlastBerry:Text="LOCTABLE(\"&GameModeStringTable\", \"ReloadKey\")"
    Figment:Text="LOCTABLE(\"&GameModeStringTable\", \"FigmentKey\")"
    Feral:Text
    ForbiddenFruit:Text="LOCTEXT(\"Blitz\")"
    Delulu:Text
    WickedSmoke:Text
    FestivalConcordGrape:Text
    ShootersFTUE:Text
    RadCow:Text
    Index:EFortGameType Enum=@n1
   ) -> (ReturnValue:Text)
   > n3 = Branch(Condition=NOT Boolean(A=TextIsEmpty(InText=@n2)))
   n4 = Get Tag_TournamentType()
   > SetTagText!(execute, self:WBP UIKit Tag Base Object Reference=@n4, Text:Text=@n2) -> (then)
   > SetVisibility(self=@n4, InVisibility=Visible)
   SetVisibility(execute=@n3.else, self=@n4, InVisibility=Collapsed)
  ]}
 SetTeamSize {nodes [
   n1 = SetTeamSize!() -> (then, TeamSize:EFortCompeteMatchType Enum=Unknown) {is=title K2Node_FunctionEntry}
   n2 = Find(
    TargetMap:Map of EFortCompeteMatchType Enums to Texts=Get TeamSizes() -> (TeamSizes:Map of EFortCompeteMatchType Enums to Texts=Unknown)
    Key:"EFortCompeteMatchType Enum (by ref)"=@n1
   ) -> (Value:Text)
   > Branch(Condition=@n2.ReturnValue)
   > Branch(Condition=NOT Boolean(A=TextIsEmpty(InText=@n2)))
   > SetLabelText!(execute, self:WBP UIKit Tag Base Object Reference=Get Tag_TournamentType(), Text:Text=@n2) -> (then)
  ]}
 SetTournamentId {nodes [
   n1 = SetTournamentId!() -> (then, Id:String) {is=title K2Node_FunctionEntry}
   > Set TournamentId(TournamentId:String=@n1) -> (Output_Get:String)
   > UpdateDebugText()
  ]}
 SetTournamentState {nodes [
   SetTournamentState!() -> (then, TournamentState:EFortCompeteTournamentWindowState Enum=Unknown) {is=title K2Node_FunctionEntry}
   > n1 = Set bIsLive(bIsLive:bool=Contains Item(
     TargetArray:Array of EFortCompeteTournamentWindowState Enums=Get LiveStates() -> (LiveStates:Array of EFortCompeteTournamentWindowState Enums=Unknown)
     ItemToFind:"EFortCompeteTournamentWindowState Enum (by ref)"=Get TournamentState() -> (TournamentState:EFortCompeteTournamentWindowState Enum=Unknown)
    )) -> (Output_Get:bool)
   > n2 = Set bIsEnded(bIsEnded:bool=Contains Item(
     TargetArray=Get EndedStates() -> (EndedStates:Array of EFortCompeteTournamentWindowState Enums=Unknown)
     ItemToFind=Get TournamentState() -> (TournamentState=Unknown)
    )) -> (Output_Get:bool)
   n3 = OR Boolean(A=@n2, B=@n1)
   n4 = Get Highlight_LiveStatus() -> (Highlight_LiveStatus:WBP UIKit Highlight Object Reference)
   > SetVisibility(
    self=@n4
    InVisibility=Select!(Option 0:ESlateVisibility Enum=Collapsed, Option 1:ESlateVisibility Enum=SelfHitTestInvisible, Index:bool=@n3) -> (ReturnValue:ESlateVisibility Enum)
   )
   > Branch(Condition=@n3)
   n5 = Get bIsLive() -> (bIsLive:bool)
   > SetText!(
    execute
    self:WBP UIKit Highlight Base Object Reference=@n4
    Text:Text=Select(Option 0:Text=Get EndedStatusText() -> (EndedStatusText:Text), Option 1:Text=Get LiveStatusText() -> (LiveStatusText:Text), Index=@n5) -> (ReturnValue:Text)
   ) -> (then)
   > SetType!(
    execute
    self:WBP UIKit Highlight Base Object Reference=@n4
    Type:E_UI_HighlightType Enum=Select(Option 0:E_UI_HighlightType Enum=NewEnumerator4, Option 1:E_UI_HighlightType Enum=NewEnumerator1, Index=@n5) -> (ReturnValue:E_UI_HighlightType Enum)
   ) -> (then)
  ]}
 SetTournamentTitle {nodes [
   n1 = SetTournamentTitle!() -> (then, TournamentTitle:Text) {is=title K2Node_FunctionEntry}
   > "SetText (Text)"(self=Get CommonText_TournamentName(), InText=@n1)
  ]}
 SetTournamentWindowId {nodes [
   n1 = SetTournamentWindowId!() -> (then, Id:String) {is=title K2Node_FunctionEntry}
   > Set TournamentWindowId(TournamentWindowId:String=@n1) -> (Output_Get:String)
   > UpdateDebugText()
  ]}
 ShouldFocusOnSelected {nodes [
   ShouldFocusOnSelected() {is=title K2Node_FunctionEntry}
   n1 = Get CommonInputSubsystem()
   > Return Node!(execute, Result:bool="Equal (Enum)"(A:ECommonInputType Enum=GetCurrentInputType(self=@n1), B:ECommonInputType Enum=Touch))
  ]}
 ShowLockStatus {nodes [
   n1 = ShowLockStatus!() -> (then, ShowLock:E_LockStatus Enum=NewEnumerator0) {is=title K2Node_FunctionEntry}
   n2 = Get LockStatusSwitcher() -> (LockStatusSwitcher:Common Activatable Widget Switcher Object Reference)
   > SetVisibility(
    self=@n2
    InVisibility=Select!(
     NewEnumerator0:ESlateVisibility Enum=HitTestInvisible
     NewEnumerator1:ESlateVisibility Enum=HitTestInvisible
     NewEnumerator2:ESlateVisibility Enum=Collapsed
     Index:E_LockStatus Enum=@n1
    ) -> (ReturnValue:ESlateVisibility Enum)
   )
   > SetActiveWidgetIndex(self=@n2, Index=Select(NewEnumerator0:int=0, NewEnumerator1:int=1, NewEnumerator2:int=2, Index=@n1) -> (ReturnValue:int))
  ]}
 UpdateDebugText {nodes [
   UpdateDebugText() {is=title K2Node_FunctionEntry}
   n1 = Get bShowDebugInfo!(self:Fort Compete UIManager Object Reference=Get FortCompeteUIManager()) -> (bShowDebugInfo:bool)
   > SetVisibility(
    self=Get DEBUG_TEXT()
    InVisibility=Select!(Option 0:ESlateVisibility Enum=Collapsed, Option 1:ESlateVisibility Enum=SelfHitTestInvisible, Index:bool=@n1) -> (ReturnValue:ESlateVisibility Enum)
   )
   > n2 = Branch(Condition=@n1)
   n3 = Get TournamentId() -> (TournamentId:String)
   n4 = IsEmpty(InString=@n3)
   n5 = Get TournamentWindowId!() -> (TournamentWindowId)
   n6 = IsEmpty(InString=@n5)
   > "SetText (Text)"(
    self=Get DEBUG_TEXT()
    InText=Format Text!(
     Format="INVTEXT(\"EventId:\\r\\n\\'{tournamentId}\\'\\r\\nEventWindowId:\\r\\n\\'{windowId}\\'\")"
     tournamentId:String=Select(Option 0:String=@n3, Option 1:String="<None>", Index=@n4) -> (ReturnValue:String)
     windowId:String=Select(Option 0=@n5, Option 1="<None>", Index=@n6)
    ) -> (Result)
   )
   "SetText (Text)"(execute=@n2.else, self=Get DEBUG_TEXT(), InText="LOCTEXT(\"Debug: Failed to get current leaderboard category.\")")
  ]}
 UpdatePosterImage {nodes [
   n1 = UpdatePosterImage!() -> (then, InPosterTexture:Texture 2D Object Reference) {is=title K2Node_FunctionEntry}
   > SetTextureParameterValue(self=Get MID_Poster(), ParameterName=Character, Value=@n1)
  ]}
]
variables [
 EndedStatusText {type=string, default=ENDED}
 LiveStatusText {type=string, default=LIVE}
 MID_Poster {type=MaterialInstanceDynamic}
 PlayerRank {type=E_RankType, default=Bronze}
 RankInfo {type=FortHabaneroDisplayData, default=/Game/UI/Habanero/HabaneroDisplayData}
 Regions {default {aSIA=ASIA, bR=BR, eU=EU, mE=ME, nAC=NAC, nAE=NAE, nAW=NAW, oCE=OCE}}
 TagIcons {default {
   cashCup=/FortUILibrary/Textures/MaterialInstances/MI_UIKit_Icon_CashCup
   fNCS=/FortUILibrary/Textures/MaterialInstances/MI_UIKit_Icon_FNCS
   none=/FortUILibrary/Textures/MaterialInstances/MI_UIKit_Icon_OtherTournaments
   other=&MI_UIKit_Icon_OtherTournaments
   rankedCup=/FortUILibrary/Textures/MaterialInstances/MI_UIKit_Icon_RankedCup
   shopCup=/FortUILibrary/Textures/MaterialInstances/MI_UIKit_Icon_ItemShopCup
   victoryCup=/FortUILibrary/Textures/MaterialInstances/MI_UIKit_Icon_VictoryCup
  }}
 TeamSizes {default {duos=Duos, solo=Solo, squads=Squads, trios=Trios}}
 TournamentId {type=string, default=""}
 TournamentWindowId {type=string, default=""}
]
widgets {
 info {
  inheritedWidgetCount=1
  parentClass=&/WBP_TournamentTile_Base.WBP_TournamentTile_Base_C
  rootWidgetClass=&GridPanel
  widgetCount=36
 }
 widgets [
  Root {widgetClassPath=&GridPanel, widget=Root, properties {columnFill [1], rowFill [1]}}
  SizeBox_Anim {
   widgetClassPath=&SizeBox
   bIsVariable=true
   parent=Root
   widget=SizeBox_Anim
   properties {bOverride_HeightOverride=true, bOverride_WidthOverride=true, heightOverride=512, widthOverride=512}
   slot {class=GridSlot, properties {horizontalAlignment=HAlign_Center, verticalAlignment=VAlign_Bottom}}
  }
  Content {
   widgetClassPath=&GridPanel
   parent=SizeBox_Anim
   widget=Content
   properties {columnFill [1], rowFill [1]}
   slot {class=SizeBoxSlot}
  }
  LoadGuard_Image {
   widgetClassPath=/Script/CommonUI.CommonLoadGuard
   bIsVariable=true
   parent=Content
   widget=LoadGuard_Image
   properties {loadingText="", textStyle=&TextStyle-Base_C}
   slot {class=GridSlot}
  }
  Image_PosterFront {
   widgetClassPath=/Script/EpicCMSUIFramework.EpicCMSImage
   bIsVariable=true
   parent=LoadGuard_Image
   widget=Image_PosterFront
   properties {
    brush {
     bIsDynamicallyLoaded=false
     drawAs=Image
     imageType=FullColor
     mirroring=NoMirror
     resourceName=None
     resourceObject=/Compete/MaterialInstances/MI_UI_CompeteTile
     tiling=NoTile
     imageSize=<800 800>
     margin=<ltrb 0 0 0 0>
     outlineSettings {
      bUseBrushTransparency=false
      roundingType=HalfHeightRadius
      width=0
      color {colorUseRule=UseColor_Specified, specifiedColor=<rgba 0 0 0 0>}
      cornerRadii=<0 0 0 0>
     }
     tintColor {colorUseRule=UseColor_Specified, specifiedColor=<rgba 1 1 1 1>}
     uVRegion {bIsValid=false, max=<0 0>, min=<0 0>}
    }
    loadingBackgroundBrush {
     bIsDynamicallyLoaded=false
     drawAs=Image
     imageType=NoImage
     mirroring=NoMirror
     resourceName=None
     resourceObject=null
     tiling=NoTile
     imageSize=<32 32>
     margin=<ltrb 0 0 0 0>
     outlineSettings {
      bUseBrushTransparency=false
      roundingType=HalfHeightRadius
      width=0
      color {colorUseRule=UseColor_Specified, specifiedColor=<rgba 0 0 0 0>}
      cornerRadii=<0 0 0 0>
     }
     tintColor {colorUseRule=UseColor_Specified, specifiedColor=<rgba 0 0.0182 0.144 1>}
     uVRegion {bIsValid=false, max=<0 0>, min=<0 0>}
    }
   }
   slot {class=LoadGuardSlot}
  }
  Header {
   widgetClassPath=&GridPanel
   parent=Content
   widget=Header
   properties {columnFill [1,0], rowFill [1]}
   slot {class=GridSlot, properties {layer=1, verticalAlignment=VAlign_Top, nudge=<0 -38>}}
  }
  CommonText_Date {
   widgetClassPath=&CommonTextBlock
   bIsVariable=true
   parent=Header
   widget=CommonText_Date
   properties {
    mobileFontSizeMultiplier=1.13
    scrollStyle=&TextScrollStyle-Base_C
    text="Monday, April 15TH"
    textTransformPolicy=ToUpper
    font {
     bForceMonospaced=false
     bMaterialIsStencil=false
     fontMaterial=null
     fontObject=&Roboto
     letterSpacing=0
     monospacedWidth=1
     size=15.75
     skewAmount=0
     typefaceFontName=75 Medium
     outlineSettings {
      bApplyOutlineToDropShadows=false
      bMiteredCorners=false
      bSeparateFillAlpha=false
      outlineMaterial=null
      outlineSize=0
      outlineColor=<rgba 0 0 0 1>
     }
    }
    shadowOffset=<0 0>
   }
   slot {class=GridSlot, properties {horizontalAlignment=HAlign_Left, verticalAlignment=VAlign_Center, padding=<ltrb 0 0 16 0>}}
  }
  GridPanel_Bookmarks {
   widgetClassPath=&GridPanel
   bIsVariable=true
   parent=Header
   widget=GridPanel_Bookmarks
   properties {columnFill [0,1], rowFill [0]}
   slot {class=GridSlot, properties {column=1, horizontalAlignment=HAlign_Right, verticalAlignment=VAlign_Center, padding=<ltrb 0 0 14 0>}}
  }
  Icon {
   widgetClassPath=&Image
   parent=GridPanel_Bookmarks
   widget=Icon
   properties {
    brush {
     bIsDynamicallyLoaded=false
     drawAs=Image
     imageType=FullColor
     mirroring=NoMirror
     resourceName=None
     resourceObject=/Compete/MaterialInstances/MI_UI_Compete_Bookmark
     tiling=NoTile
     imageSize=<15 21>
     margin=<ltrb 0 0 0 0>
     outlineSettings {
      bUseBrushTransparency=false
      roundingType=HalfHeightRadius
      width=0
      color {colorUseRule=UseColor_Specified, specifiedColor=<rgba 0 0 0 0>}
      cornerRadii=<0 0 0 0>
     }
     tintColor {colorUseRule=UseColor_Specified, specifiedColor=<rgba 1 1 1 1>}
     uVRegion {bIsValid=false, max=<0 0>, min=<0 0>}
    }
    colorAndOpacity=<rgba 1 1 1 0.8>
   }
   slot {class=GridSlot, properties {horizontalAlignment=HAlign_Left, verticalAlignment=VAlign_Center}}
  }
  CommonText_BookmarksAmount {
   widgetClassPath=&CommonTextBlock
   bIsVariable=true
   parent=GridPanel_Bookmarks
   widget=CommonText_BookmarksAmount
   properties {
    mobileFontSizeMultiplier=1.2
    scrollStyle=&TextScrollStyle-Base_C
    text=25.2K
    colorAndOpacity {colorUseRule=UseColor_Specified, specifiedColor=<rgba 1 1 1 0.8>}
    font {
     bForceMonospaced=false
     bMaterialIsStencil=false
     fontMaterial=null
     fontObject=&HeadingNow
     letterSpacing=10
     monospacedWidth=1
     size=15.75
     skewAmount=0
     typefaceFontName=74 Regular
     outlineSettings {
      bApplyOutlineToDropShadows=false
      bMiteredCorners=false
      bSeparateFillAlpha=false
      outlineMaterial=null
      outlineSize=0
      outlineColor=<rgba 0 0 0 1>
     }
    }
    shadowOffset=<0 0>
   }
   slot {class=GridSlot, properties {column=1, verticalAlignment=VAlign_Center, padding=<ltrb 10 0 0 0>}}
  }
  Subheader {
   widgetClassPath=&GridPanel
   parent=Content
   widget=Subheader
   properties {columnFill [0,1,0], rowFill [0]}
   slot {class=GridSlot, properties {layer=1, verticalAlignment=VAlign_Top, padding=<ltrb 16 12 16 12>}}
  }
  Border_RankIcon {
   widgetClassPath=&Border
   bIsVariable=true
   parent=Subheader
   widget=Border_RankIcon
   properties {
    horizontalAlignment=HAlign_Center
    verticalAlignment=VAlign_Center
    background {
     bIsDynamicallyLoaded=false
     drawAs=Image
     imageType=FullColor
     mirroring=NoMirror
     resourceName=None
     resourceObject=/Compete/TournamentTile/MaterialInstances/MI_Compete_BackgroundRounded
     tiling=NoTile
     imageSize=<32 32>
     margin=<ltrb 0 0 0 0>
     outlineSettings {
      bUseBrushTransparency=false
      roundingType=HalfHeightRadius
      width=0
      color {colorUseRule=UseColor_Specified, specifiedColor=<rgba 0 0 0 0>}
      cornerRadii=<0 0 0 0>
     }
     tintColor {colorUseRule=UseColor_Specified, specifiedColor=<rgba 1 1 1 1>}
     uVRegion {bIsValid=false, max=<0 0>, min=<0 0>}
    }
    padding=<ltrb 4 4 4 4>
   }
   slot {class=GridSlot, properties {horizontalAlignment=HAlign_Left, verticalAlignment=VAlign_Center, padding=<ltrb 0 0 6 0>}}
  }
  Image_RankIcon {
   widgetClassPath=&Image
   bIsVariable=true
   parent=Border_RankIcon
   widget=Image_RankIcon
   properties {brush {
     bIsDynamicallyLoaded=false
     drawAs=Image
     imageType=FullColor
     mirroring=NoMirror
     resourceName=None
     resourceObject=/FortUILibrary/UIKit/Throbber/MI_UIKit_Monochrome_Throbber
     tiling=NoTile
     imageSize=<32 32>
     margin=<ltrb 0 0 0 0>
     outlineSettings {
      bUseBrushTransparency=false
      roundingType=HalfHeightRadius
      width=0
      color {colorUseRule=UseColor_Specified, specifiedColor=<rgba 0 0 0 0>}
      cornerRadii=<0 0 0 0>
     }
     tintColor {colorUseRule=UseColor_Specified, specifiedColor=<rgba 1 1 1 1>}
     uVRegion {bIsValid=false, max=<0 0>, min=<0 0>}
    }}
   slot {class=BorderSlot, properties {horizontalAlignment=HAlign_Center, verticalAlignment=VAlign_Center, padding=<ltrb 4 4 4 4>}}
  }
  CommonText_Region {
   widgetClassPath=&CommonTextBlock
   bIsVariable=true
   parent=Subheader
   widget=CommonText_Region
   properties {
    mobileFontSizeMultiplier=1.1
    scrollStyle=&TextScrollStyle-Base_C
    text=NAE
    font {
     bForceMonospaced=false
     bMaterialIsStencil=false
     fontMaterial=null
     fontObject=&NotoSans
     letterSpacing=0
     monospacedWidth=1
     size=24
     skewAmount=0
     typefaceFontName=NotoSans-Regular
     outlineSettings {
      bApplyOutlineToDropShadows=false
      bMiteredCorners=false
      bSeparateFillAlpha=false
      outlineMaterial=null
      outlineSize=0
      outlineColor=<rgba 0 0 0 1>
     }
    }
    shadowOffset=<0 0>
   }
   slot {class=GridSlot, properties {column=1, verticalAlignment=VAlign_Center, padding=<ltrb 0 8 16 8>}}
  }
  Highlight_LiveStatus {
   widgetClassPath=/FortUILibrary/UIKit/Highlight/WBP_UIKit_Highlight
   bIsVariable=true
   parent=Subheader
   widget=Highlight_LiveStatus
   properties {highlightText=LIVE, highlightType=Social}
   slot {class=GridSlot, properties {column=2, horizontalAlignment=HAlign_Right, layer=1, verticalAlignment=VAlign_Center}}
  }
  LockStatusSwitcher {
   widgetClassPath=&CommonActivatableWidgetSwitcher
   bIsVariable=true
   parent=Content
   widget=LockStatusSwitcher
   properties {transitionDuration=1}
   slot {class=GridSlot, properties {horizontalAlignment=HAlign_Center, layer=1, verticalAlignment=VAlign_Bottom, padding=<ltrb 0 0 0 250>}}
  }
  ThrobberImg {
   widgetClassPath=&Image
   bIsVariable=true
   parent=LockStatusSwitcher
   widget=ThrobberImg
   properties {
    visibility=SelfHitTestInvisible
    brush {
     bIsDynamicallyLoaded=false
     drawAs=Image
     imageType=FullColor
     mirroring=NoMirror
     resourceName=None
     resourceObject=&MI_UIKit_Monochrome_Throbber
     tiling=NoTile
     imageSize=<70 70>
     margin=<ltrb 0 0 0 0>
     outlineSettings {
      bUseBrushTransparency=false
      roundingType=HalfHeightRadius
      width=0
      color {colorUseRule=UseColor_Specified, specifiedColor=<rgba 0 0 0 0>}
      cornerRadii=<0 0 0 0>
     }
     tintColor {colorUseRule=UseColor_Specified, specifiedColor=<rgba 1 1 1 1>}
     uVRegion {bIsValid=false, max=<0 0>, min=<0 0>}
    }
   }
   slot {class=WidgetSwitcherSlot, properties {padding=<ltrb 0 0 0 48>}}
  }
  EligibilityStatus {
   widgetClassPath=&GridPanel
   bIsVariable=true
   parent=LockStatusSwitcher
   widget=EligibilityStatus
   properties {columnFill [1], rowFill [1]}
   slot {class=WidgetSwitcherSlot, properties {padding=<ltrb 16 0 16 0>}}
  }
  Border_LockIcon {
   widgetClassPath=&Border
   parent=EligibilityStatus
   widget=Border_LockIcon
   properties {
    horizontalAlignment=HAlign_Center
    verticalAlignment=VAlign_Center
    background {
     bIsDynamicallyLoaded=false
     drawAs=Image
     imageType=FullColor
     mirroring=NoMirror
     resourceName=None
     resourceObject=&MI_Compete_BackgroundRounded
     tiling=NoTile
     imageSize=<72 72>
     margin=<ltrb 0 0 0 0>
     outlineSettings {
      bUseBrushTransparency=false
      roundingType=HalfHeightRadius
      width=0
      color {colorUseRule=UseColor_Specified, specifiedColor=<rgba 0 0 0 0>}
      cornerRadii=<0 0 0 0>
     }
     tintColor {colorUseRule=UseColor_Specified, specifiedColor=<rgba 1 1 1 1>}
     uVRegion {bIsValid=false, max=<0 0>, min=<0 0>}
    }
    padding=<ltrb 10 10 10 10>
   }
   slot {class=GridSlot, properties {horizontalAlignment=HAlign_Center, verticalAlignment=VAlign_Center}}
  }
  Image_LockIcon {
   widgetClassPath=&Image
   parent=Border_LockIcon
   widget=Image_LockIcon
   properties {brush {
     bIsDynamicallyLoaded=false
     drawAs=Image
     imageType=FullColor
     mirroring=NoMirror
     resourceName=None
     resourceObject=/FortUILibrary/Textures/MaterialInstances/MI_UIKit_Icon_Locked
     tiling=NoTile
     imageSize=<52 52>
     margin=<ltrb 0 0 0 0>
     outlineSettings {
      bUseBrushTransparency=false
      roundingType=HalfHeightRadius
      width=0
      color {colorUseRule=UseColor_Specified, specifiedColor=<rgba 0 0 0 0>}
      cornerRadii=<0 0 0 0>
     }
     tintColor {colorUseRule=UseColor_Specified, specifiedColor=<rgba 1 1 1 1>}
     uVRegion {bIsValid=false, max=<0 0>, min=<0 0>}
    }}
   slot {class=BorderSlot, properties {horizontalAlignment=HAlign_Center, verticalAlignment=VAlign_Center, padding=<ltrb 10 10 10 10>}}
  }
  Border_EligibilityRequirement {
   widgetClassPath=&Border
   parent=EligibilityStatus
   widget=Border_EligibilityRequirement
   properties {
    horizontalAlignment=HAlign_Center
    verticalAlignment=VAlign_Center
    background {
     bIsDynamicallyLoaded=false
     drawAs=Image
     imageType=FullColor
     mirroring=NoMirror
     resourceName=None
     resourceObject=&MI_Compete_BackgroundRounded
     tiling=NoTile
     imageSize=<32 32>
     margin=<ltrb 0 0 0 0>
     outlineSettings {
      bUseBrushTransparency=false
      roundingType=HalfHeightRadius
      width=0
      color {colorUseRule=UseColor_Specified, specifiedColor=<rgba 0 0 0 0>}
      cornerRadii=<0 0 0 0>
     }
     tintColor {colorUseRule=UseColor_Specified, specifiedColor=<rgba 1 1 1 1>}
     uVRegion {bIsValid=false, max=<0 0>, min=<0 0>}
    }
    padding=<ltrb 12 7 12 7>
   }
   slot {class=GridSlot, properties {row=1, padding=<ltrb 0 8 0 0>}}
  }
  CommonText_EligibilityRequirement {
   widgetClassPath=&CommonTextBlock
   bIsVariable=true
   parent=Border_EligibilityRequirement
   widget=CommonText_EligibilityRequirement
   properties {
    bAutoCollapseWithEmptyText=true
    scrollStyle=&TextScrollStyle-Base_C
    text=System Requirements Not Met
    font {
     bForceMonospaced=false
     bMaterialIsStencil=false
     fontMaterial=null
     fontObject=&HeadingNow
     letterSpacing=10
     monospacedWidth=1
     size=15.75
     skewAmount=0
     typefaceFontName=74 Regular
     outlineSettings {
      bApplyOutlineToDropShadows=false
      bMiteredCorners=false
      bSeparateFillAlpha=false
      outlineMaterial=null
      outlineSize=0
      outlineColor=<rgba 0 0 0 1>
     }
    }
    shadowOffset=<0 0>
   }
   slot {class=BorderSlot, properties {horizontalAlignment=HAlign_Center, verticalAlignment=VAlign_Center, padding=<ltrb 12 7 12 7>}}
  }
  PassedImg {
   widgetClassPath=&Image
   bIsVariable=true
   parent=LockStatusSwitcher
   widget=PassedImg
   properties {
    visibility=Collapsed
    brush {
     bIsDynamicallyLoaded=false
     drawAs=Image
     imageType=FullColor
     mirroring=NoMirror
     resourceName=None
     resourceObject=/FortUILibrary/Textures/Icons/T_UI_Icon_UIKit_Check
     tiling=NoTile
     imageSize=<64 64>
     margin=<ltrb 0 0 0 0>
     outlineSettings {
      bUseBrushTransparency=false
      roundingType=HalfHeightRadius
      width=0
      color {colorUseRule=UseColor_Specified, specifiedColor=<rgba 0 0 0 0>}
      cornerRadii=<0 0 0 0>
     }
     tintColor {colorUseRule=UseColor_Specified, specifiedColor=<rgba 1 1 1 1>}
     uVRegion {bIsValid=false, max=<0 0>, min=<0 0>}
    }
   }
   slot {class=WidgetSwitcherSlot, properties {padding=<ltrb 0 0 0 48>}}
  }
  VerticalBox_TournamentInfo {
   widgetClassPath=&VerticalBox
   bIsVariable=true
   parent=Content
   widget=VerticalBox_TournamentInfo
   slot {class=GridSlot, properties {layer=1, verticalAlignment=VAlign_Bottom, padding=<ltrb 16 24 16 24>}}
  }
  BookmarkStatus {
   widgetClassPath=&GridPanel
   bIsVariable=true
   parent=VerticalBox_TournamentInfo
   widget=BookmarkStatus
   properties {visibility=Collapsed, columnFill [0,1], rowFill [0]}
   slot {class=VerticalBoxSlot, properties {verticalAlignment=VAlign_Top, padding=<ltrb 0 0 0 8>}}
  }
  CupBackground {
   widgetClassPath=&Image
   parent=BookmarkStatus
   widget=CupBackground
   properties {brush {
     bIsDynamicallyLoaded=false
     drawAs=Image
     imageType=FullColor
     mirroring=NoMirror
     resourceName=None
     resourceObject=/Compete/MaterialInstances/MI_UI_Compete_Flag
     tiling=NoTile
     imageSize=<26 37>
     margin=<ltrb 0 0 0 0>
     outlineSettings {
      bUseBrushTransparency=false
      roundingType=HalfHeightRadius
      width=0
      color {colorUseRule=UseColor_Specified, specifiedColor=<rgba 0 0 0 0>}
      cornerRadii=<0 0 0 0>
     }
     tintColor {colorUseRule=UseColor_Specified, specifiedColor=<rgba 1 1 1 1>}
     uVRegion {bIsValid=false, max=<0 0>, min=<0 0>}
    }}
   slot {class=GridSlot, properties {horizontalAlignment=HAlign_Center, verticalAlignment=VAlign_Center}}
  }
  Image_Cup {
   widgetClassPath=&Image
   parent=BookmarkStatus
   widget=Image_Cup
   properties {brush {
     bIsDynamicallyLoaded=false
     drawAs=Image
     imageType=FullColor
     mirroring=NoMirror
     resourceName=None
     resourceObject=/FortUILibrary/Textures/MaterialInstances/MI_UIKit_Icon_Compete
     tiling=NoTile
     imageSize=<22 24>
     margin=<ltrb 0 0 0 0>
     outlineSettings {
      bUseBrushTransparency=false
      roundingType=HalfHeightRadius
      width=0
      color {colorUseRule=UseColor_Specified, specifiedColor=<rgba 0 0 0 0>}
      cornerRadii=<0 0 0 0>
     }
     tintColor {colorUseRule=UseColor_Specified, specifiedColor=<rgba 0.013 0.013 0.013 1>}
     uVRegion {bIsValid=false, max=<0 0>, min=<0 0>}
    }}
   slot {class=GridSlot, properties {horizontalAlignment=HAlign_Center, layer=1, verticalAlignment=VAlign_Center, padding=<ltrb 0 0 0 2>}}
  }
  CommonText_Bookmark {
   widgetClassPath=&CommonTextBlock
   bIsVariable=true
   parent=BookmarkStatus
   widget=CommonText_Bookmark
   properties {
    mobileFontSizeMultiplier=1.16
    scrollStyle=&TextScrollStyle-Base_C
    text="Bookmarked!"
    colorAndOpacity {colorUseRule=UseColor_Specified, specifiedColor=<rgba 0.0395 1 0.2307 1>}
    font {
     bForceMonospaced=false
     bMaterialIsStencil=false
     fontMaterial=null
     fontObject=&HeadingNow
     letterSpacing=0
     monospacedWidth=1
     size=15.75
     skewAmount=0
     typefaceFontName=75 Medium Italic
     outlineSettings {
      bApplyOutlineToDropShadows=false
      bMiteredCorners=false
      bSeparateFillAlpha=false
      outlineMaterial=null
      outlineSize=0
      outlineColor=<rgba 0 0 0 1>
     }
    }
    shadowOffset=<0 0>
   }
   slot {class=GridSlot, properties {column=1, verticalAlignment=VAlign_Center, padding=<ltrb 12 0 0 0>}}
  }
  CommonText_RoundDateTime {
   widgetClassPath=&CommonTextBlock
   bIsVariable=true
   parent=VerticalBox_TournamentInfo
   widget=CommonText_RoundDateTime
   properties {
    scrollStyle=&TextScrollStyle-Base_C
    text=6:00 PM - 8:00 PM GMT-3
    colorAndOpacity {colorUseRule=UseColor_Specified, specifiedColor=<rgba 0.5776 0.5776 0.5776 1>}
    font {
     bForceMonospaced=false
     bMaterialIsStencil=false
     fontMaterial=null
     fontObject=&NotoSans
     letterSpacing=0
     monospacedWidth=1
     size=24
     skewAmount=0
     typefaceFontName=NotoSans-Regular
     outlineSettings {
      bApplyOutlineToDropShadows=false
      bMiteredCorners=false
      bSeparateFillAlpha=false
      outlineMaterial=null
      outlineSize=0
      outlineColor=<rgba 0 0 0 1>
     }
    }
    shadowOffset=<0 0>
   }
   slot {class=VerticalBoxSlot}
  }
  SizeBox_0 {
   widgetClassPath=&SizeBox
   parent=VerticalBox_TournamentInfo
   widget=SizeBox_0
   properties {bOverride_MaxDesiredHeight=true, bOverride_WidthOverride=true, maxDesiredHeight=110, widthOverride=480}
   slot {class=VerticalBoxSlot, properties {horizontalAlignment=HAlign_Left}}
  }
  CommonText_TournamentName {
   widgetClassPath=&CommonTextBlock
   bIsVariable=true
   parent=SizeBox_0
   widget=CommonText_TournamentName
   properties {
    applyLineHeightToBottomLine=false
    autoWrapText=true
    bIsScrollingEnabled=false
    clipping=OnDemand
    lineHeightPercentage=0.75
    mobileFontSizeMultiplier=1.2
    scrollStyle=&TextScrollStyle-Base_C
    text="Bronze,  fh fgh fgh fg hfgh fg hfgSilver && Golsadsdas asd ask  12 12 12"
    textOverflowPolicy=MultilineEllipsis
    textTransformPolicy=ToUpper
    font {
     bForceMonospaced=false
     bMaterialIsStencil=false
     fontMaterial=/FortUILibrary/UIKit/Buttons/MaterialInstances/ItemShopTile/MI_UIKit_ItemShopTile_TitleText
     fontObject=&HeadingNow
     letterSpacing=-10
     monospacedWidth=1
     size=25.5
     skewAmount=0
     typefaceFontName=86 Bold Italic
     outlineSettings {
      bApplyOutlineToDropShadows=false
      bMiteredCorners=false
      bSeparateFillAlpha=false
      outlineMaterial=null
      outlineSize=0
      outlineColor=<rgba 0 0 0 1>
     }
    }
    shadowOffset=<0 0>
   }
   slot {class=SizeBoxSlot, properties {verticalAlignment=VAlign_Center}}
  }
  Tag_TournamentType {
   widgetClassPath=/FortUILibrary/UIKit/Tag/WBP_UIKit_Tag
   bIsVariable=true
   parent=VerticalBox_TournamentInfo
   widget=Tag_TournamentType
   properties {
    iconLeftAligned=true
    iconMaterial=/FortUILibrary/Textures/MaterialInstances/MI_UIKit_Icon_JunoStyle
    labelText=Solo
    tagText=Battle Royale
    tagType=Primary Tag
    useIcon=true
    iconPadding=<ltrb 0 0 4 0>
   }
   slot {class=VerticalBoxSlot, properties {verticalAlignment=VAlign_Bottom}}
  }
  Block_Outline {
   widgetClassPath=/FortUILibrary/UIKit/Buttons/Blocks/WBP_UIKit_Block_Outline
   bIsVariable=true
   parent=Content
   widget=Block_Outline
   properties {bottomLeftCornerRadius=27, bottomRightCornerRadius=27, radiusInPixels=true, topLeftCornerRadius=27, topRightCornerRadius=27}
   slot {class=GridSlot, properties {layer=30}}
  }
  Image_OffscreenOverlay {
   widgetClassPath=&Image
   bIsVariable=true
   parent=Content
   widget=Image_OffscreenOverlay
   properties {
    visibility=SelfHitTestInvisible
    brush {
     bIsDynamicallyLoaded=false
     drawAs=Image
     imageType=NoImage
     mirroring=NoMirror
     resourceName=None
     resourceObject=/FortUILibrary/Systems/ItemShopTile/Materials/MI_UI_MPShopTile_OffscreenOverlay
     tiling=NoTile
     imageSize=<32 32>
     margin=<ltrb 0 0 0 0>
     outlineSettings {
      bUseBrushTransparency=false
      roundingType=HalfHeightRadius
      width=0
      color {colorUseRule=UseColor_Specified, specifiedColor=<rgba 0 0 0 0>}
      cornerRadii=<0 0 0 0>
     }
     tintColor {colorUseRule=UseColor_Specified, specifiedColor=<rgba 1 1 1 1>}
     uVRegion {bIsValid=false, max=<0 0>, min=<0 0>}
    }
   }
   slot {class=GridSlot, properties {layer=40}}
  }
  DEBUG_TEXT {
   widgetClassPath=&CommonTextBlock
   bIsVariable=true
   parent=Content
   widget=DEBUG_TEXT
   properties {
    autoWrapText=true
    justification=Center
    text="[PH] Debug text"
    visibility=Collapsed
    wrappingPolicy=AllowPerCharacterWrapping
    font {
     bForceMonospaced=false
     bMaterialIsStencil=false
     fontMaterial=null
     fontObject=&Roboto
     letterSpacing=0
     monospacedWidth=1
     size=18
     skewAmount=0
     typefaceFontName=Regular
     outlineSettings {
      bApplyOutlineToDropShadows=false
      bMiteredCorners=false
      bSeparateFillAlpha=false
      outlineMaterial=null
      outlineSize=4
      outlineColor=<rgba 0 0 0 1>
     }
    }
    shadowOffset=<0 0>
   }
   slot {class=GridSlot, properties {layer=2, verticalAlignment=VAlign_Top, nudge=<0 50>}}
  }
  InputActionWidget {widgetClassPath=&CommonActionWidget, bInherited=true, widget=null}
 ]
}
