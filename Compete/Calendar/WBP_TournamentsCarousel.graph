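# Horizontal carousel of tournament tiles for the Compete calendar, built on a polymorphic list view with a hidden scrollbar.
# Calendar tile viewmodels map to selectable tournament tiles; other items become non-selectable empty tiles.
# Each generated tournament tile receives the carousel viewmodel.
# When details-screen params change with a valid player controller, it opens the details screen via the Compete UI manager, then resets those params to an empty struct.
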WidgetBlueprint /Compete/Calendar/WBP_TournamentsCarousel
graphs [
 EventGraph {nodes [
   n1 = OnEntryGenerated!() -> (OutputDelegate, then, Widget:User Widget Object Reference) {is=title K2Node_CustomEvent}
   > n2 = Cast To WBP_TournamentTile!(execute, Object=@n1.Widget) -> (then, CastFailed, AsWBP Tournament Tile:WBP Tournament Tile Object Reference) {is=title K2Node_DynamicCast}
   > SetFortPoblanoTournamentsCarouselVM!(
    execute
    self:WBP Tournament Tile Object Reference=@n2
    Viewmodel:Tournaments Carousel VM Object Reference=Get FortPoblanoTournamentsCarouselVM() -> (FortPoblanoTournamentsCarouselVM:Tournaments Carousel VM Object Reference)
   ) -> (then)
   Event Construct()
   > Bind Event to On Entry Generated(self:List View Base Object Reference=Get ListView() -> (ListView:Common List View Object Reference), Delegate=@n1) {is=title K2Node_AddDelegate}
   Event Destruct()
   > Unbind Event from On Entry Generated(self:List View Base Object Reference=Get ListView(), Delegate=@n1) {is=title K2Node_RemoveDelegate}
   Event PreConstruct()
  ]}
 GetDesiredEntryClassForItem {nodes [
   n1 = GetDesiredEntryClassForItem!() -> (then, Item:Object Reference) {is=title K2Node_FunctionEntry}
   > n2 = Cast To FortPoblanoTournamentsCalendarTournamentTileVM!(execute, Object=@n1) -> (then, CastFailed, AsCalendar Tournament Tile VM:Calendar Tournament Tile VM Object Reference) {is=title K2Node_DynamicCast}
   > Return Node!(execute, ReturnValue:User Widget Class Reference=/Compete/TournamentTile/WBP_TournamentTile.WBP_TournamentTile_C)
   Return Node(execute=@n2.CastFailed, ReturnValue=&/WBP_EmptyTile_Entry.WBP_EmptyTile_Entry_C)
  ]}
 GetItemSelectableOrNavigable {nodes [
   n1 = GetItemSelectableOrNavigable!() -> (then, Item:Object Reference) {is=title K2Node_FunctionEntry}
   > n2 = Cast To FortPoblanoTournamentsCalendarTournamentTileVM(Object=@n1)
   > Return Node(ReturnValue:bool=true)
   Return Node(execute=@n2.CastFailed, ReturnValue=false)
  ]}
 OnTournamentDetailsScreenParamsChanged {nodes [
   OnTournamentDetailsScreenParamsChanged!() -> (then, DetailsScreenParams:Tournament Details Screen Params Structure) {is=title K2Node_FunctionEntry}
   > Branch(Condition=IsValid(Object=Get!() -> (
      DetailsScreenParams_ContextPC:Player Controller Object Reference
      DetailsScreenParams_TournamentId:String
      DetailsScreenParams_TournamentWindowId:String
     )))
   > ShowTournamentDetailsScreen!(
    execute
    self:Fort Compete UIManager Object Reference=Get FortCompeteUIManager() -> (ReturnValue:Fort Compete UIManager Object Reference) {is=title K2Node_GetSubsystem}
    Params:"Tournament Details Screen Params Structure (by ref)"=Get DetailsScreenParams() -> (DetailsScreenParams:Tournament Details Screen Params Structure)
   ) -> (then)
   > SetTournamentDetailsScreenParams!(
    execute
    self:Tournaments Carousel VM Object Reference=Get FortPoblanoTournamentsCarouselVM()
    InTournamentDetailsScreenParams:"Tournament Details Screen Params Structure (by ref)"=Make Tournament Details Screen Params!(ContextPC:Player Controller Object Reference, TournamentId:String, TournamentWindowId:String) -> (TournamentDetailsScreenParams:Tournament Details Screen Params Structure) {is=title K2Node_MakeStruct}
   ) -> (then)
  ]}
]
variables []
widgets {
 info {
  inheritedWidgetCount=1
  parentClass=/Script/FortniteUI.FortPolymorphicListView
  rootWidgetClass=&CommonListView
  widgetCount=1
 }
 widgets [ListView {
   widgetClassPath=&CommonListView
   bInherited=true
   bIsVariable=true
   widget=ListView
   properties {
    bClearScrollVelocityOnSelection=false
    bEnableFixedLineOffset=true
    bEnableRightClickScrolling=false
    bInEnableTouchAnimatedScrolling=true
    bReturnFocusToSelection=true
    clipping=Inherit
    consumeMouseWheel=Never
    entryWidgetClass=&WBP_TournamentTile_C
    fixedLineScrollOffset=0.2
    horizontalEntrySpacing=32
    orientation=Orient_Horizontal
    scrollIntoViewAlignment=TopOrLeft
    wheelScrollMultiplier=0
    scrollBarStyle {
     thickness=0
     draggedThumbImage {
      bIsDynamicallyLoaded=false
      drawAs=NoDrawType
      imageType=FullColor
      mirroring=NoMirror
      resourceName=../../../Engine/Content/Slate/Common/Scrollbar_Thumb.png
      resourceObject=null
      tiling=NoTile
      imageSize=<0 0>
      margin=<ltrb 0.25 0.25 0.25 0.25>
      outlineSettings {
       bUseBrushTransparency=false
       roundingType=HalfHeightRadius
       width=0
       color {colorUseRule=UseColor_Specified, specifiedColor=<rgba 0 0 0 0>}
       cornerRadii=<0 0 0 0>
      }
      tintColor {colorUseRule=UseColor_Specified, specifiedColor=<rgba 1 1 1 1>}
      uVRegion {bIsValid=false, max=<0 0>, min=<0 0>}
     }
     horizontalBackgroundImage {
      bIsDynamicallyLoaded=false
      drawAs=NoDrawType
      imageType=NoImage
      mirroring=NoMirror
      resourceName=None
      resourceObject=null
      tiling=NoTile
      imageSize=<0 0>
      margin=<ltrb 0 0 0 0>
      outlineSettings {
       bUseBrushTransparency=false
       roundingType=HalfHeightRadius
       width=0
       color {colorUseRule=UseColor_Specified, specifiedColor=<rgba 0 0 0 0>}
       cornerRadii=<0 0 0 0>
      }
      tintColor {colorUseRule=UseColor_Specified, specifiedColor=<rgba 1 1 1 1>}
      uVRegion {bIsValid=false, max=<0 0>, min=<0 0>}
     }
     horizontalBottomSlotImage {
      bIsDynamicallyLoaded=false
      drawAs=NoDrawType
      imageType=FullColor
      mirroring=NoMirror
      resourceName=../../../Engine/Content/Slate/Common/Scrollbar_Background_Horizontal.png
      resourceObject=null
      tiling=NoTile
      imageSize=<8 8>
      margin=<ltrb 0 0 0 0>
      outlineSettings {
       bUseBrushTransparency=false
       roundingType=HalfHeightRadius
       width=0
       color {colorUseRule=UseColor_Specified, specifiedColor=<rgba 0 0 0 0>}
       cornerRadii=<0 0 0 0>
      }
      tintColor {colorUseRule=UseColor_Specified, specifiedColor=<rgba 1 1 1 1>}
      uVRegion {bIsValid=false, max=<0 0>, min=<0 0>}
     }
     horizontalTopSlotImage {
      bIsDynamicallyLoaded=false
      drawAs=NoDrawType
      imageType=FullColor
      mirroring=NoMirror
      resourceName=../../../Engine/Content/Slate/Common/Scrollbar_Background_Horizontal.png
      resourceObject=null
      tiling=NoTile
      imageSize=<8 8>
      margin=<ltrb 0 0 0 0>
      outlineSettings {
       bUseBrushTransparency=false
       roundingType=HalfHeightRadius
       width=0
       color {colorUseRule=UseColor_Specified, specifiedColor=<rgba 0 0 0 0>}
       cornerRadii=<0 0 0 0>
      }
      tintColor {colorUseRule=UseColor_Specified, specifiedColor=<rgba 1 1 1 1>}
      uVRegion {bIsValid=false, max=<0 0>, min=<0 0>}
     }
     hoveredThumbImage {
      bIsDynamicallyLoaded=false
      drawAs=NoDrawType
      imageType=FullColor
      mirroring=NoMirror
      resourceName=../../../Engine/Content/Slate/Common/Scrollbar_Thumb.png
      resourceObject=null
      tiling=NoTile
      imageSize=<0 0>
      margin=<ltrb 0.25 0.25 0.25 0.25>
      outlineSettings {
       bUseBrushTransparency=false
       roundingType=HalfHeightRadius
       width=0
       color {colorUseRule=UseColor_Specified, specifiedColor=<rgba 0 0 0 0>}
       cornerRadii=<0 0 0 0>
      }
      tintColor {colorUseRule=UseColor_Specified, specifiedColor=<rgba 1 1 1 1>}
      uVRegion {bIsValid=false, max=<0 0>, min=<0 0>}
     }
     normalThumbImage {
      bIsDynamicallyLoaded=false
      drawAs=NoDrawType
      imageType=FullColor
      mirroring=NoMirror
      resourceName=../../../Engine/Content/Slate/Common/Scrollbar_Thumb.png
      resourceObject=null
      tiling=NoTile
      imageSize=<0 0>
      margin=<ltrb 0.25 0.25 0.25 0.25>
      outlineSettings {
       bUseBrushTransparency=false
       roundingType=HalfHeightRadius
       width=0
       color {colorUseRule=UseColor_Specified, specifiedColor=<rgba 0 0 0 0>}
       cornerRadii=<0 0 0 0>
      }
      tintColor {colorUseRule=UseColor_Specified, specifiedColor=<rgba 1 1 1 1>}
      uVRegion {bIsValid=false, max=<0 0>, min=<0 0>}
     }
     verticalBackgroundImage {
      bIsDynamicallyLoaded=false
      drawAs=NoDrawType
      imageType=NoImage
      mirroring=NoMirror
      resourceName=None
      resourceObject=null
      tiling=NoTile
      imageSize=<0 0>
      margin=<ltrb 0 0 0 0>
      outlineSettings {
       bUseBrushTransparency=false
       roundingType=HalfHeightRadius
       width=0
       color {colorUseRule=UseColor_Specified, specifiedColor=<rgba 0 0 0 0>}
       cornerRadii=<0 0 0 0>
      }
      tintColor {colorUseRule=UseColor_Specified, specifiedColor=<rgba 1 1 1 1>}
      uVRegion {bIsValid=false, max=<0 0>, min=<0 0>}
     }
     verticalBottomSlotImage {
      bIsDynamicallyLoaded=false
      drawAs=NoDrawType
      imageType=FullColor
      mirroring=NoMirror
      resourceName=../../../Engine/Content/Slate/Common/Scrollbar_Background_Vertical.png
      resourceObject=null
      tiling=NoTile
      imageSize=<8 8>
      margin=<ltrb 0 0 0 0>
      outlineSettings {
       bUseBrushTransparency=false
       roundingType=HalfHeightRadius
       width=0
       color {colorUseRule=UseColor_Specified, specifiedColor=<rgba 0 0 0 0>}
       cornerRadii=<0 0 0 0>
      }
      tintColor {colorUseRule=UseColor_Specified, specifiedColor=<rgba 1 1 1 1>}
      uVRegion {bIsValid=false, max=<0 0>, min=<0 0>}
     }
     verticalTopSlotImage {
      bIsDynamicallyLoaded=false
      drawAs=NoDrawType
      imageType=FullColor
      mirroring=NoMirror
      resourceName=../../../Engine/Content/Slate/Common/Scrollbar_Background_Vertical.png
      resourceObject=null
      tiling=NoTile
      imageSize=<8 8>
      margin=<ltrb 0 0 0 0>
      outlineSettings {
       bUseBrushTransparency=false
       roundingType=HalfHeightRadius
       width=0
       color {colorUseRule=UseColor_Specified, specifiedColor=<rgba 0 0 0 0>}
       cornerRadii=<0 0 0 0>
      }
      tintColor {colorUseRule=UseColor_Specified, specifiedColor=<rgba 1 1 1 1>}
      uVRegion {bIsValid=false, max=<0 0>, min=<0 0>}
     }
    }
   }
  }]
}
