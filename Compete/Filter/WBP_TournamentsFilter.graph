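# Activatable dialog for filtering the tournaments calendar by experience, team size, region, type, status, eligibility and bookmarks.
# On activation it loads each section's toggles from the calendar filter view model; Apply writes them back, flags it dirty and closes, while Cancel just closes.
# Region collapses while the local party has multiple players, updating on party changes.
# Gamepad input refreshes focus to the Experience section, and narrow-aspect mobile screens get a taller dialog.
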WidgetBlueprint /Compete/Filter/WBP_TournamentsFilter
graphs [
 BP_GetDesiredFocusTarget {nodes [
   Get Desired Focus Target()
   n1 = Get Section_Experience() -> (Section_Experience:WBP Tournaments Filter Section Object Reference)
   > Return Node!(
    execute
    ReturnValue:Widget Object Reference="Get (a copy)"(Array:Array of WBP Tournament Filter Sort Toggle Button Object References=Get FilterSectionEntries!(self:WBP Tournaments Filter Section Object Reference=@n1) -> (FilterSectionEntries:Array of WBP Tournament Filter Sort Toggle Button Object References)) -> (Output:WBP Tournament Filter Sort Toggle Button Object Reference)
   )
  ]}
 EventGraph {nodes [
   "On Clicked (Button_ApplyFilters)"!() -> (OutputDelegate, then, Button:Common Button Base Object Reference) {is=title K2Node_ComponentBoundEvent}
   > n1 = Sequence!(execute) -> (then_0, then_1, then_2:Exec, then_3:Exec, then_4:Exec, then_5:Exec, then_6:Exec, then_7:Exec, then_8:Exec)
   n2 = Get FortPoblanoTournamentsCalendarFilterVM() -> (FortPoblanoTournamentsCalendarFilterVM:Tournaments Calendar Filter VM Object Reference)
   SetAllSectionTournamentType!(
    execute=@n1.then_3
    self:Tournaments Calendar Filter VM Object Reference=@n2
    State:Array of Booleans=Get ToggleValues!(self:WBP Tournaments Filter Section Object Reference=Get Section_TournamentType() -> (Section_TournamentType:WBP Tournaments Filter Section Object Reference)) -> (ToggleValues:Array of Booleans)
   ) -> (then)
   SetAllSectionTournamentStatus!(
    execute=@n1.then_4
    self:Tournaments Calendar Filter VM Object Reference=@n2
    State:Array of Booleans=Get ToggleValues(self=Get Section_TournamentStatus() -> (Section_TournamentStatus:WBP Tournaments Filter Section Object Reference))
   ) -> (then)
   SetAllSectionExperience!(
    execute=@n1.then_0
    self:Tournaments Calendar Filter VM Object Reference=@n2
    States:Array of Booleans=Get ToggleValues(self=Get Section_Experience())
   ) -> (then)
   SetAllSectionTeamSize!(
    execute=@n1.then_1
    self:Tournaments Calendar Filter VM Object Reference=@n2
    State:Array of Booleans=Get ToggleValues(self=Get Section_TeamSize() -> (Section_TeamSize:WBP Tournaments Filter Section Object Reference))
   ) -> (then)
   SetAllSectionRegion!(
    execute=@n1.then_2
    self:Tournaments Calendar Filter VM Object Reference=@n2
    State:Array of Booleans=Get ToggleValues(self=Get Section_Region() -> (Section_Region:WBP Tournaments Filter Section Object Reference))
   ) -> (then)
   DeactivateWidget(execute=[@n3.then, @n1.then_8])
   SetbIsDirtyFilter!(execute=@n1.then_7, self:Tournaments Calendar Filter VM Object Reference=@n2, bInIsDirtyFilter:bool=true) -> (then)
   SetAllSectionBookmarks!(
    execute=@n1.then_6
    self:Tournaments Calendar Filter VM Object Reference=@n2
    State:Array of Booleans=Get ToggleValues(self=Get Section_Bookmarks() -> (Section_Bookmarks:WBP Tournaments Filter Section Object Reference))
   ) -> (then)
   SetAllSectionElegibility!(
    execute=@n1.then_5
    self:Tournaments Calendar Filter VM Object Reference=@n2
    State:Array of Booleans=Get ToggleValues(self=Get Section_Eligibility() -> (Section_Eligibility:WBP Tournaments Filter Section Object Reference))
   ) -> (then)
   n3 = "On Clicked (Button_Cancel)"!() -> (OutputDelegate, then, Button:Common Button Base Object Reference) {is=title K2Node_ComponentBoundEvent}
   n4 = OnInputMethodChanged!() -> (OutputDelegate, then, bNewInputType:ECommonInputType Enum) {is=title K2Node_CustomEvent}
   > n5 = Switch on ECommonInputType!(execute, Selection:ECommonInputType Enum=@n4.bNewInputType) -> (MouseAndKeyboard:Exec, Gamepad:Exec, Touch:Exec, Count:Exec) {is=title K2Node_SwitchEnum}
   RequestRefreshFocus(execute=@n5.Gamepad, self:Common Activatable Widget Object Reference)
   n6 = OnPlayerStateChanged_Event!() -> (OutputDelegate, then, PlayerInfo:"Fort Team Member Info Structure (by ref)") {is=title K2Node_CustomEvent}
   n7 = GetNumPlayersInLocalParty!(self:Fort Party Context Object Reference=Get FortPartyContext!() -> (ReturnValue:Fort Party Context Object Reference) {is=title K2Node_GetBlueprintContextEx}) -> (ReturnValue:int=0)
   SetVisibility(
    execute=[@n16.then, @n6.then]
    self=Get Section_Region()
    InVisibility=Select(Option 0=SelfHitTestInvisible, Option 1=Collapsed, Index="integer > integer"(A=@n7, B=1) {is=title K2Node_CallFunction})
   )
   Event Construct()
   > SetCompactUIStyle() {is=title K2Node_CallFunction}
   Event On Activated()
   n8 = Get CommonInputSubsystem()
   > OnInputMethodChanged!(execute, self, bNewInputType:ECommonInputType Enum=GetCurrentInputType(self=@n8)) -> (then) {is=title K2Node_CallFunction}
   > Assign On Input Method Changed!(execute:Exec, self:Common Input Object Reference=@n8, Delegate:"Delegate (by ref)"=@n4) -> (then:Exec) {is=title K2Node_AssignDelegate}
   > n9 = Sequence!(execute) -> (then_0, then_1, then_2, then_3, then_4, then_5, then_6, then_7)
   n10 = Get FortPoblanoTournamentsCalendarFilterVM()
   n11 = GetAllSectionTournamentType!(execute=@n9.then_3, self:Tournaments Calendar Filter VM Object Reference=@n10) -> (then, ReturnValue:Array of Booleans)
   > UpdateToggleValues!(execute, self:WBP Tournaments Filter Section Object Reference=Get Section_TournamentType(), Values:Array of Booleans=@n11) -> (then)
   n12 = GetAllSectionExperience!(execute=@n9.then_0, self:Tournaments Calendar Filter VM Object Reference=@n10) -> (then, ReturnValue:Array of Booleans)
   > UpdateToggleValues(self=Get Section_Experience(), Values=@n12)
   n13 = GetAllSectionTeamSize!(execute=@n9.then_1, self:Tournaments Calendar Filter VM Object Reference=@n10) -> (then, ReturnValue:Array of Booleans)
   > UpdateToggleValues(self=Get Section_TeamSize(), Values=@n13)
   n14 = GetAllSectionRegion!(execute=@n9.then_2, self:Tournaments Calendar Filter VM Object Reference=@n10) -> (then, ReturnValue:Array of Booleans)
   > UpdateToggleValues(self=Get Section_Region(), Values=@n14)
   n15 = GetAllSectionTournamentStatus!(execute=@n9.then_4, self:Tournaments Calendar Filter VM Object Reference=@n10) -> (then, ReturnValue:Array of Booleans)
   > UpdateToggleValues(self=Get Section_TournamentStatus(), Values=@n15)
   n16 = Assign On Player State Changed!(execute:Exec=@n9.then_7, self:Fort Party Context Object Reference=Get FortPartyContext(), Delegate:"Delegate (by ref)"=@n6) -> (then:Exec) {is=title K2Node_AssignDelegate}
   n17 = GetAllSectionElegibility!(execute=@n9.then_5, self:Tournaments Calendar Filter VM Object Reference=@n10) -> (then, ReturnValue:Array of Booleans)
   > UpdateToggleValues(self=Get Section_Eligibility(), Values=@n17)
   n18 = GetAllSectionBookmarks!(execute=@n9.then_6, self:Tournaments Calendar Filter VM Object Reference=@n10) -> (then, ReturnValue:Array of Booleans)
   > UpdateToggleValues(self=Get Section_Bookmarks(), Values=@n18)
   Event On Deactivated()
   > Unbind Event from On Player State Changed(self:Fort Party Context Object Reference=Get FortPartyContext(), Delegate=@n6) {is=title K2Node_RemoveDelegate}
   > Unbind Event from On Input Method Changed(self=Get CommonInputSubsystem(), Delegate=Create Event())
  ]}
 GetRegionName {nodes [
   n1 = GetRegionName!() -> (then, Text:Text) {is=title K2Node_FunctionEntry}
   n2 = "To String (Text)"(InText=@n1)
   > n3 = Branch(Condition=Contains(SearchIn=@n2, Substring="("))
   > Return Node!(execute, FormatedText:Text="To Text (String)"(InString=Split(SourceString=@n2, InStr=" (", SearchDir=FromStart)))
   Return Node(execute=@n3.else, FormatedText=Get Text!() -> (Text))
  ]}
 SetCompactUIStyle {nodes [
   SetCompactUIStyle() {is=title K2Node_FunctionEntry}
   > Branch(Condition=IsMobileGame())
   n1 = GetViewportSize!() -> (ReturnValue_X:double, ReturnValue_Y:double)
   > n2 = Branch(Condition="float >= float"(A=float / float(A=@n1, B=@n1.ReturnValue_Y), B=1.45))
   SetSizeBox!(
    execute=@n2.else
    self:WBP UIKit Dialog Internal Object Reference=Get Dialog_Internal() -> (Dialog_Internal:WBP UIKit Dialog Internal Object Reference)
    Data_WidthOverride_107_#1:S UI Bool Float Structure=(Enable_3_#2=False,Value_4_#3=0)
    Data_HeightOverride_108_#4_Enable_3_#2:bool=true
    Data_HeightOverride_108_#4_Value_4_#3:double=1260
    Data_MinDesiredWidth_109_#5:S UI Bool Float Structure=(Enable_3_#2=False,Value_4_#3=0)
    Data_MinDesiredHeight_110_#6:S UI Bool Float Structure=(Enable_3_#2=False,Value_4_#3=0)
    Data_MaxDesiredWidth_111_#7:S UI Bool Float Structure=(Enable_3_#2=False,Value_4_#3=0)
    Data_MaxDesiredHeight_112_#8:S UI Bool Float Structure=(Enable_3_#2=False,Value_4_#3=0)
    Data_MinAspectRatio_113_#9:S UI Bool Float Structure=(Enable_3_#2=False,Value_4_#3=0)
    Data_MaxAspectRatio_114_#10:S UI Bool Float Structure=(Enable_3_#2=False,Value_4_#3=0)
   ) -> (then)
  ]}
]
variables []
widgets {
 info {
  inheritedWidgetCount=0
  namedSlotCount=1
  parentClass=&CommonActivatableWidget
  rootWidgetClass=/FortUILibrary/Systems/Dialog/WBP_UIKit_Dialog_Internal.WBP_UIKit_Dialog_Internal_C
  widgetCount=17
 }
 widgets [
  Dialog_Internal {
   widgetClassPath=/FortUILibrary/Systems/Dialog/WBP_UIKit_Dialog_Internal
   bIsVariable=true
   widget=Dialog_Internal
   properties {
    useTitle=false
    buttons {
     each Button Horizontal Fill=true
     each Button Min Width=0
     each Button Size=Medium
     focuse Button=0
     space Between Buttons=30
     cTA Buttons Types [Regular,Quiet]
     group Placement {
      horizontal Alignment=HAlign_Fill
      vertical Alignment=VAlign_Bottom
      grid {column=1, column Span=0, layer=0, row=2, row Span=0, nudge=<0 0>}
      padding=<ltrb 0 0 0 0>
     }
    }
    dialogSIze {
     height Override {enable=true, value=886}
     max Aspect Ratio {enable=false, value=0}
     max Desired Height {enable=true, value=800}
     max Desired Width {enable=false, value=0}
     min Aspect Ratio {enable=false, value=0}
     min Desired Height {enable=false, value=0}
     min Desired Width {enable=false, value=0}
     width Override {enable=true, value=1511}
    }
    titleProperties {
     auto Wrap Text=true
     justification=Left
     line Height Percentage=1
     mobile Font Size Multiplier=1.3
     transform Policy=ToUpper
     wrap Text At=0
     color {colorUseRule=UseColor_Specified, specifiedColor=<rgba 1 1 1 1>}
     font {
      bForceMonospaced=false
      bMaterialIsStencil=false
      fontMaterial=null
      fontObject=&HeadingNow
      letterSpacing=1
      monospacedWidth=1
      size=18
      skewAmount=0
      typefaceFontName=86 Bold
      outlineSettings {
       bApplyOutlineToDropShadows=false
       bMiteredCorners=false
       bSeparateFillAlpha=false
       outlineMaterial=null
       outlineSize=0
       outlineColor=<rgba 0 0 0 1>
      }
     }
     placement {
      horizontal Alignment=HAlign_Fill
      vertical Alignment=VAlign_Top
      grid {column=1, column Span=0, layer=0, row=0, row Span=0, nudge=<0 0>}
      padding=<ltrb 80 80 80 32>
     }
    }
   }
  }
  Overlay_MainContent {widgetClassPath=&Overlay, namedSlotHost=Dialog_Internal, widget=Overlay_MainContent}
  GridPanel_Main {
   widgetClassPath=&GridPanel
   parent=Overlay_MainContent
   widget=GridPanel_Main
   properties {columnFill [1], rowFill [0,1]}
   slot {class=OverlaySlot, properties {horizontalAlignment=HAlign_Fill, padding=<ltrb 62 80 50 43>}}
  }
  Text_TournamentFilter {
   widgetClassPath=&CommonTextBlock
   bIsVariable=true
   parent=GridPanel_Main
   widget=Text_TournamentFilter
   properties {
    bAutoCollapseWithEmptyText=true
    clipping=OnDemand
    scrollStyle=/FortUILibrary/Styles/Typography/Foundation/TextScrollStyle-NoFade.TextScrollStyle-NoFade_C
    text=Filter Tournaments
    textOverflowPolicy=Ellipsis
    textTransformPolicy=ToUpper
    font {
     bForceMonospaced=false
     bMaterialIsStencil=false
     fontMaterial=null
     fontObject=&HeadingNow
     letterSpacing=50
     monospacedWidth=1
     size=24
     skewAmount=0
     typefaceFontName=86 Bold Italic
     outlineSettings {
      bApplyOutlineToDropShadows=false
      bMiteredCorners=false
      bSeparateFillAlpha=false
      outlineMaterial=null
      outlineSize=0
      outlineColor=<rgba 0 0 0 1>
     }
    }
    shadowOffset=<0 0>
   }
   slot {class=GridSlot, properties {padding=<ltrb 18 0 0 0>}}
  }
  ScrollBox_FilterSelector {
   widgetClassPath=&ScrollBox
   parent=GridPanel_Main
   widget=ScrollBox_FilterSelector
   properties {
    allowOverscroll=false
    bAnimateWheelScrolling=true
    navigation=&self:WidgetTree.ScrollBox_FilterSelector.WidgetNavigation_0
    navigationDestination=Center
    navigationScrollPadding=100
    scrollWhenFocusChanges=AnimatedScroll
    wheelScrollMultiplier=3
    scrollbarPadding=<ltrb 0 0 0 0>
    scrollbarThickness=<12 12>
    widgetBarStyle {
     thickness=10
     draggedThumbImage {
      bIsDynamicallyLoaded=false
      drawAs=RoundedBox
      imageType=FullColor
      mirroring=NoMirror
      resourceName=None
      resourceObject=null
      tiling=NoTile
      imageSize=<12 12>
      margin=<ltrb 0.5 0.5 0.5 0.5>
      outlineSettings {
       bUseBrushTransparency=false
       roundingType=FixedRadius
       width=2
       color {colorUseRule=UseColor_Specified, specifiedColor=<rgba 0 0 0 0>}
       cornerRadii=<6 6 6 6>
      }
      tintColor {colorUseRule=UseColor_Specified, specifiedColor=<rgba 1 1 1 0.75>}
      uVRegion {bIsValid=false, max=<0 0>, min=<0 0>}
     }
     horizontalBackgroundImage {
      bIsDynamicallyLoaded=false
      drawAs=NoDrawType
      imageType=NoImage
      mirroring=NoMirror
      resourceName=None
      resourceObject=null
      tiling=NoTile
      imageSize=<0 0>
      margin=<ltrb 0 0 0 0>
      outlineSettings {
       bUseBrushTransparency=false
       roundingType=HalfHeightRadius
       width=0
       color {colorUseRule=UseColor_Specified, specifiedColor=<rgba 0 0 0 0>}
       cornerRadii=<0 0 0 0>
      }
      tintColor {colorUseRule=UseColor_Specified, specifiedColor=<rgba 1 1 1 1>}
      uVRegion {bIsValid=false, max=<0 0>, min=<0 0>}
     }
     horizontalBottomSlotImage {
      bIsDynamicallyLoaded=false
      drawAs=Image
      imageType=FullColor
      mirroring=NoMirror
      resourceName=../../../Engine/Content/Slate/Common/Scrollbar_Background_Horizontal.png
      resourceObject=null
      tiling=NoTile
      imageSize=<8 8>
      margin=<ltrb 0 0 0 0>
      outlineSettings {
       bUseBrushTransparency=false
       roundingType=HalfHeightRadius
       width=0
       color {colorUseRule=UseColor_Specified, specifiedColor=<rgba 0 0 0 0>}
       cornerRadii=<0 0 0 0>
      }
      tintColor {colorUseRule=UseColor_Specified, specifiedColor=<rgba 1 1 1 1>}
      uVRegion {bIsValid=false, max=<0 0>, min=<0 0>}
     }
     horizontalTopSlotImage {
      bIsDynamicallyLoaded=false
      drawAs=Image
      imageType=FullColor
      mirroring=NoMirror
      resourceName=../../../Engine/Content/Slate/Common/Scrollbar_Background_Horizontal.png
      resourceObject=null
      tiling=NoTile
      imageSize=<8 8>
      margin=<ltrb 0 0 0 0>
      outlineSettings {
       bUseBrushTransparency=false
       roundingType=HalfHeightRadius
       width=0
       color {colorUseRule=UseColor_Specified, specifiedColor=<rgba 0 0 0 0>}
       cornerRadii=<0 0 0 0>
      }
      tintColor {colorUseRule=UseColor_Specified, specifiedColor=<rgba 1 1 1 1>}
      uVRegion {bIsValid=false, max=<0 0>, min=<0 0>}
     }
     hoveredThumbImage {
      bIsDynamicallyLoaded=false
      drawAs=RoundedBox
      imageType=FullColor
      mirroring=NoMirror
      resourceName=None
      resourceObject=null
      tiling=NoTile
      imageSize=<12 12>
      margin=<ltrb 0.5 0.5 0.5 0.5>
      outlineSettings {
       bUseBrushTransparency=false
       roundingType=FixedRadius
       width=2
       color {colorUseRule=UseColor_Specified, specifiedColor=<rgba 0 0 0 0>}
       cornerRadii=<6 6 6 6>
      }
      tintColor {colorUseRule=UseColor_Specified, specifiedColor=<rgba 1 1 1 0.75>}
      uVRegion {bIsValid=false, max=<0 0>, min=<0 0>}
     }
     normalThumbImage {
      bIsDynamicallyLoaded=false
      drawAs=RoundedBox
      imageType=FullColor
      mirroring=NoMirror
      resourceName=None
      resourceObject=null
      tiling=NoTile
      imageSize=<12 12>
      margin=<ltrb 0.5 0.5 0.5 0.5>
      outlineSettings {
       bUseBrushTransparency=false
       roundingType=FixedRadius
       width=2
       color {colorUseRule=UseColor_Specified, specifiedColor=<rgba 0 0 0 0>}
       cornerRadii=<6 6 6 6>
      }
      tintColor {colorUseRule=UseColor_Specified, specifiedColor=<rgba 1 1 1 0.75>}
      uVRegion {bIsValid=false, max=<0 0>, min=<0 0>}
     }
     verticalBackgroundImage {
      bIsDynamicallyLoaded=false
      drawAs=Box
      imageType=FullColor
      mirroring=NoMirror
      resourceName=None
      resourceObject=/FortUILibrary/UIKit/ScrollBar/MaterialInstances/MI_UIKit_ScrollbarBackground
      tiling=NoTile
      imageSize=<12 12>
      margin=<ltrb 0.5 0.5 0.5 0.5>
      outlineSettings {
       bUseBrushTransparency=false
       roundingType=HalfHeightRadius
       width=0
       color {colorUseRule=UseColor_Specified, specifiedColor=<rgba 0 0 0 0>}
       cornerRadii=<0 0 0 0>
      }
      tintColor {colorUseRule=UseColor_Specified, specifiedColor=<rgba 1 1 1 0.2>}
      uVRegion {bIsValid=false, max=<0 0>, min=<0 0>}
     }
     verticalBottomSlotImage {
      bIsDynamicallyLoaded=false
      drawAs=NoDrawType
      imageType=FullColor
      mirroring=NoMirror
      resourceName=../../../Engine/Content/Slate/Common/Scrollbar_Background_Vertical.png
      resourceObject=null
      tiling=NoTile
      imageSize=<10 10>
      margin=<ltrb 0.5 0.5 0.5 0.5>
      outlineSettings {
       bUseBrushTransparency=false
       roundingType=HalfHeightRadius
       width=0
       color {colorUseRule=UseColor_Specified, specifiedColor=<rgba 0 0 0 0>}
       cornerRadii=<0 0 0 0>
      }
      tintColor {colorUseRule=UseColor_Specified, specifiedColor=<rgba 1 1 1 1>}
      uVRegion {bIsValid=false, max=<0 0>, min=<0 0>}
     }
     verticalTopSlotImage {
      bIsDynamicallyLoaded=false
      drawAs=NoDrawType
      imageType=FullColor
      mirroring=NoMirror
      resourceName=../../../Engine/Content/Slate/Common/Scrollbar_Background_Vertical.png
      resourceObject=null
      tiling=NoTile
      imageSize=<8 8>
      margin=<ltrb 0 0 0 0>
      outlineSettings {
       bUseBrushTransparency=false
       roundingType=HalfHeightRadius
       width=9
       color {colorUseRule=UseColor_Specified, specifiedColor=<rgba 0 0 0 0>}
       cornerRadii=<0.5 0.5 0.5 0.5>
      }
      tintColor {colorUseRule=UseColor_Specified, specifiedColor=<rgba 1 1 1 1>}
      uVRegion {bIsValid=false, max=<0 0>, min=<0 0>}
     }
    }
    widgetStyle {
     barThickness=12
     bottomShadowBrush {
      bIsDynamicallyLoaded=false
      drawAs=Box
      imageType=FullColor
      mirroring=NoMirror
      resourceName=../../../Engine/Content/Slate/Common/ScrollBoxShadowBottom.png
      resourceObject=null
      tiling=NoTile
      imageSize=<16 8>
      margin=<ltrb 0 0 0 0>
      outlineSettings {
       bUseBrushTransparency=false
       roundingType=HalfHeightRadius
       width=0
       color {colorUseRule=UseColor_Specified, specifiedColor=<rgba 0 0 0 0>}
       cornerRadii=<0 0 0 0>
      }
      tintColor {colorUseRule=UseColor_Specified, specifiedColor=<rgba 1 1 1 0>}
      uVRegion {bIsValid=false, max=<0 0>, min=<0 0>}
     }
     horizontalScrolledContentPadding=<ltrb 0 0 1 0>
     leftShadowBrush {
      bIsDynamicallyLoaded=false
      drawAs=Box
      imageType=FullColor
      mirroring=NoMirror
      resourceName=../../../Engine/Content/Slate/Common/ScrollBoxShadowLeft.png
      resourceObject=null
      tiling=NoTile
      imageSize=<8 16>
      margin=<ltrb 1 0.5 0 0.5>
      outlineSettings {
       bUseBrushTransparency=false
       roundingType=HalfHeightRadius
       width=0
       color {colorUseRule=UseColor_Specified, specifiedColor=<rgba 0 0 0 0>}
       cornerRadii=<0 0 0 0>
      }
      tintColor {colorUseRule=UseColor_Specified, specifiedColor=<rgba 1 1 1 1>}
      uVRegion {bIsValid=false, max=<0 0>, min=<0 0>}
     }
     rightShadowBrush {
      bIsDynamicallyLoaded=false
      drawAs=Box
      imageType=FullColor
      mirroring=NoMirror
      resourceName=../../../Engine/Content/Slate/Common/ScrollBoxShadowRight.png
      resourceObject=null
      tiling=NoTile
      imageSize=<8 16>
      margin=<ltrb 0 0.5 1 0.5>
      outlineSettings {
       bUseBrushTransparency=false
       roundingType=HalfHeightRadius
       width=0
       color {colorUseRule=UseColor_Specified, specifiedColor=<rgba 0 0 0 0>}
       cornerRadii=<0 0 0 0>
      }
      tintColor {colorUseRule=UseColor_Specified, specifiedColor=<rgba 1 1 1 0>}
      uVRegion {bIsValid=false, max=<0 0>, min=<0 0>}
     }
     topShadowBrush {
      bIsDynamicallyLoaded=false
      drawAs=NoDrawType
      imageType=FullColor
      mirroring=NoMirror
      resourceName=../../../Engine/Content/Slate/Common/ScrollBoxShadowTop.png
      resourceObject=null
      tiling=NoTile
      imageSize=<1 90>
      margin=<ltrb 0 0 0 0>
      outlineSettings {
       bUseBrushTransparency=false
       roundingType=HalfHeightRadius
       width=0
       color {colorUseRule=UseColor_Specified, specifiedColor=<rgba 0 0 0 0>}
       cornerRadii=<0 0 0 0>
      }
      tintColor {colorUseRule=UseColor_Specified, specifiedColor=<rgba 1 1 1 0>}
      uVRegion {bIsValid=false, max=<0 0>, min=<0 0>}
     }
     verticalScrolledContentPadding=<ltrb 0 0 0 1>
    }
   }
   slot {class=GridSlot, properties {row=1, padding=<ltrb 0 28 0 64>}}
  }
  VerticalBox_FilterSelector {widgetClassPath=&VerticalBox, parent=ScrollBox_FilterSelector, widget=VerticalBox_FilterSelector, slot {class=ScrollBoxSlot}}
  Section_Experience {
   widgetClassPath=&/WBP_TournamentsFilter_Section
   bIsVariable=true
   parent=VerticalBox_FilterSelector
   widget=Section_Experience
   slot {class=VerticalBoxSlot}
  }
  Section_TeamSize {
   widgetClassPath=&WBP_TournamentsFilter_Section
   bIsVariable=true
   parent=VerticalBox_FilterSelector
   widget=Section_TeamSize
   slot {class=VerticalBoxSlot}
  }
  Section_Region {
   widgetClassPath=&WBP_TournamentsFilter_Section
   bIsVariable=true
   parent=VerticalBox_FilterSelector
   widget=Section_Region
   properties {bRadioType=true}
   slot {class=VerticalBoxSlot}
  }
  Section_TournamentType {
   widgetClassPath=&WBP_TournamentsFilter_Section
   bIsVariable=true
   parent=VerticalBox_FilterSelector
   widget=Section_TournamentType
   properties {isShowIcons=true}
   slot {class=VerticalBoxSlot}
  }
  Section_TournamentStatus {
   widgetClassPath=&WBP_TournamentsFilter_Section
   bIsVariable=true
   parent=VerticalBox_FilterSelector
   widget=Section_TournamentStatus
   slot {class=VerticalBoxSlot}
  }
  Section_Eligibility {
   widgetClassPath=&WBP_TournamentsFilter_Section
   bIsVariable=true
   parent=VerticalBox_FilterSelector
   widget=Section_Eligibility
   properties {bRadioType=true}
   slot {class=VerticalBoxSlot}
  }
  Section_Bookmarks {
   widgetClassPath=&WBP_TournamentsFilter_Section
   bIsVariable=true
   parent=VerticalBox_FilterSelector
   widget=Section_Bookmarks
   properties {bRadioType=true}
   slot {class=VerticalBoxSlot}
  }
  GridPanel_Buttons {
   widgetClassPath=&GridPanel
   parent=GridPanel_Main
   widget=GridPanel_Buttons
   properties {columnFill [1,0,1]}
   slot {class=GridSlot, properties {row=2, padding=<ltrb 18 0 16 0>}}
  }
  Button_ApplyFilters {
   widgetClassPath=&WBP_UIKit_Button_Regular
   bIsVariable=true
   parent=GridPanel_Buttons
   widget=Button_ApplyFilters
   properties {
    bIsFocusable=false
    minWidth=596
    text=Apply Filters
    triggeringInputAction {dataTable=&InputActionDataTable, rowName=Accept}
   }
   slot {class=GridSlot}
  }
  Spacer_309 {
   widgetClassPath=&Spacer
   parent=GridPanel_Buttons
   widget=Spacer_309
   properties {size=<24 1>}
   slot {class=GridSlot, properties {column=1}}
  }
  Button_Cancel {
   widgetClassPath=&WBP_UIKit_Button_Regular
   bIsVariable=true
   parent=GridPanel_Buttons
   widget=Button_Cancel
   properties {
    bIsFocusable=false
    buttonType=Quiet
    minWidth=596
    text=Cancel
    triggeringInputAction {dataTable=&InputActionDataTable, rowName=Cancel}
   }
   slot {class=GridSlot, properties {column=2}}
  }
 ]
}
