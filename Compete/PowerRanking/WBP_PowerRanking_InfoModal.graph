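# Activatable Power Rankings info modal that explains how players gain PR, using the UIKit dialog with an image and a Close button.
# PreConstruct sets the localized title and rich text, then adds a flash-material Image to the dialog grid in the content image's cell, with a layer derived from that image.
# The button is the focus target and deactivates the modal on click.
# While active, input method changes trigger a focus refresh.
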WidgetBlueprint /Compete/PowerRanking/WBP_PowerRanking_InfoModal
graphs [
 BP_GetDesiredFocusTarget {nodes [
   Get Desired Focus Target()
   > Return Node!(
    execute
    ReturnValue:Widget Object Reference=Get WBP_UIKit_Button_Regular() -> (WBP_UIKit_Button_Regular:WBP UIKit Button Regular Object Reference)
   )
  ]}
 EventGraph {nodes [
   "On Clicked (WBP_UIKit_Button_Regular)"!() -> (OutputDelegate, then, Button:Common Button Base Object Reference) {is=title K2Node_ComponentBoundEvent}
   > DeactivateWidget()
   n1 = InputMethodChanged!() -> (OutputDelegate, then, bNewInputType:ECommonInputType Enum) {is=title K2Node_CustomEvent}
   > RequestRefreshFocus(self:Common Activatable Widget Object Reference)
   Event PreConstruct()
   n2 = Get WBP_UIKit_Dialog_Internal() -> (WBP_UIKit_Dialog_Internal:WBP UIKit Dialog Internal Object Reference)
   > SetDialogTitle!(execute, self:WBP UIKit Dialog Internal Object Reference=@n2, Title:Text="LOCTEXT(\"Power Rankings (PR)\")") -> (then)
   > SetDialogText!(
    execute
    self:WBP UIKit Dialog Internal Object Reference=@n2
    Text:Text=Format Text!(
     Format="LOCTEXT(\"Get your Power Ranking!\\r\\n<text color=\\\"{color}\\\"> • Play in tournaments to start your journey.</> \\r\\n\\r\\nCompete in Tournaments to improve\\r\\n<text color=\\\"{color}\\\"> • The higher tier the tournament, the more PR you gain!</>\\r\\n\\r\\nFollow your Friends and Favorites\\r\\n<text color=\\\"{color}\\\"> • Leaderboards update as tournaments finish.</>\")"
     color:Text=Format Text!(Format="INVTEXT(\"c8c8c8cc\")") -> (Result)
    ) -> (Result)
   ) -> (then)
   > n3 = Construct Object from Class(Class=&Image) -> (ReturnValue:Image Object Reference)
   > SetBrushFromMaterial(self=@n3, Material=/Compete/PowerRanking/Materials/M_UI_PowerRanking_Flash)
   n4 = Get WBP_UIKit_Dialog_Internal()
   n5 = SlotAsGridSlot(Widget=Get ContentImage!(self:WBP UIKit Dialog Internal Object Reference=@n4) -> (ContentImage:Image Object Reference))
   > n6 = AddChildToGrid!(
    execute
    self:Grid Panel Object Reference=Get Main!(self:WBP UIKit Dialog Internal Object Reference=@n4) -> (Main:Grid Panel Object Reference)
    Content:Widget Object Reference=@n3
    InRow:int=Get Row!(self:Grid Slot Object Reference=@n5) -> (Row:int=0)
    InColumn:int=Get Column!(self:Grid Slot Object Reference=@n5) -> (Column:int=0)
   ) -> (then, ReturnValue:Grid Slot Object Reference)
   > SetLayer!(execute, self:Grid Slot Object Reference=@n6, InLayer:int=int - int(A=Get Layer!(self:Grid Slot Object Reference=@n5) -> (Layer:int=0))) -> (then)
   Event On Activated()
   n7 = Get CommonInputSubsystem()
   > Bind Event to On Input Method Changed(self=@n7, Delegate=@n1)
   Event On Deactivated()
   > Unbind Event from On Input Method Changed(self=@n7, Delegate=@n1)
  ]}
]
variables [NewVar {type=UIKitDialogButtonViewModel, default=null}]
widgets {
 info {
  inheritedWidgetCount=0
  namedSlotCount=1
  parentClass=&CommonActivatableWidget
  rootWidgetClass=/FortUILibrary/Systems/Dialog/WBP_UIKit_Dialog_Internal.WBP_UIKit_Dialog_Internal_C
  widgetCount=3
 }
 widgets [
  WBP_UIKit_Dialog_Internal {
   widgetClassPath=/FortUILibrary/Systems/Dialog/WBP_UIKit_Dialog_Internal
   bIsVariable=true
   widget=WBP_UIKit_Dialog_Internal
   properties {
    designTimePreview_DialogVM=null
    useImage=true
    useText=Common Rich Text
    dialogSIze {
     height Override {enable=false, value=800}
     max Aspect Ratio {enable=false, value=0}
     max Desired Height {enable=true, value=800}
     max Desired Width {enable=false, value=0}
     min Aspect Ratio {enable=false, value=0}
     min Desired Height {enable=false, value=0}
     min Desired Width {enable=false, value=0}
     width Override {enable=true, value=1424}
    }
    imageProperties {
     material=/Compete/PowerRanking/Materials/MI_UI_PowerRank_Hyper_Modal
     image Size=<512 612>
     placement {
      horizontal Alignment=HAlign_Fill
      vertical Alignment=VAlign_Fill
      grid {column=1, column Span=0, layer=20, row=0, row Span=1, nudge=<0 0>}
      padding=<ltrb 0 0 0 0>
     }
    }
    textProperties {
     auto Wrap Text=true
     justification=Left
     line Height Percentage=1.2
     mobile Font Size Multiplier=1.3
     transform Policy=None
     wrap Text At=0
     color {colorUseRule=UseColor_Specified, specifiedColor=<rgba 1 1 1 1>}
     font {
      bForceMonospaced=false
      bMaterialIsStencil=false
      fontMaterial=null
      fontObject=&HeadingNow
      letterSpacing=0
      monospacedWidth=1
      size=18
      skewAmount=0
      typefaceFontName=74 Regular
      outlineSettings {
       bApplyOutlineToDropShadows=false
       bMiteredCorners=false
       bSeparateFillAlpha=false
       outlineMaterial=null
       outlineSize=0
       outlineColor=<rgba 0 0 0 1>
      }
     }
     placement {
      horizontal Alignment=HAlign_Fill
      vertical Alignment=VAlign_Fill
      grid {column=1, column Span=0, layer=0, row=1, row Span=0, nudge=<0 0>}
      padding=<ltrb 80 0 40 40>
     }
    }
   }
  }
  GridPanel_21 {
   widgetClassPath=&GridPanel
   namedSlotHost=WBP_UIKit_Dialog_Internal
   widget=GridPanel_21
   properties {navigation=&self:WidgetTree.GridPanel_21.WidgetNavigation_0, columnFill [1], rowFill [1]}
  }
  WBP_UIKit_Button_Regular {
   widgetClassPath=&WBP_UIKit_Button_Regular
   bIsVariable=true
   parent=GridPanel_21
   widget=WBP_UIKit_Button_Regular
   properties {
    navigation=&self:WidgetTree.WBP_UIKit_Button_Regular.WidgetNavigation_0
    text=Close
    triggeringInputAction {dataTable=&InputActionDataTable, rowName=Close}
   }
   slot {class=GridSlot, properties {horizontalAlignment=HAlign_Left, verticalAlignment=VAlign_Bottom, padding=<ltrb 80 0 0 48>}}
  }
 ]
}
