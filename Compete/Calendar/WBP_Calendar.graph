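# WBP_Calendar is the Compete screen's tournament calendar widget, pairing a scrollable tournament carousel with an expandable month grid and an optional landing page.
# The calendar view model fills the carousel, padded with empty entries, and generates day tiles; selection picks the previous, current-date, upcoming or closest event.
# Buttons, swipes and overscroll animate transitions between views.
# It also opens filter, eligibility and power-ranking modals, toggles bookmarks and manages focus.
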WidgetBlueprint /Compete/Calendar/WBP_Calendar
animations [
 Anim_OpenCalendar {
  binding_count=6
  length_seconds=1.0038
  object=&self:Anim_OpenCalendar
  sequence {
   bindings [
    GridSlot_3 {
     id=#1
     parent_id=#2
     possessed_class=GridSlot
     tracks [{
       type=MovieSceneMarginTrack
       display_name=Padding
       sections [{
         type=MovieSceneMarginSection
         end_frame=20
         start_frame=0
         channels [
          Left {type=MovieSceneScriptingFloatChannel, const [-132,1]}
          Top {type=MovieSceneScriptingFloatChannel, keys [[0,40],[20,0]]}
          Right {type=MovieSceneScriptingFloatChannel, const [-132,1]}
          Bottom {type=MovieSceneScriptingFloatChannel, keys [[0,25],[20,0]]}
         ]
        }]
      }]
    }
    IconHintsPanel {
     id=#3
     possessed_class=WBP_IconHintsPanel_C
     tracks [{
       type=MovieSceneByteTrack
       display_name=Visibility
       sections [{type=MovieSceneByteSection, end_frame=20, start_frame=0, channels [None {type=MovieSceneScriptingByteChannel, keys [[0,1],[20,4]]}]}]
      }]
    }
    SizeBox_Month {
     id=#4
     possessed_class=SizeBox
     tracks [{
       type=MovieSceneFloatTrack
       display_name=Max Desired Height
       sections [{type=MovieSceneFloatSection, end_frame=20, start_frame=0, channels [None {type=MovieSceneScriptingFloatChannel, keys [[0,135],[20,1000]]}]}]
      }]
    }
    SizeBox {
     id=#2
     possessed_class=SizeBox
     tracks [
      {
       type=MovieSceneFloatTrack
       display_name=Render Opacity
       sections [{type=MovieSceneFloatSection, end_frame=20, start_frame=0, channels [None {type=MovieSceneScriptingFloatChannel, keys [[0,1],[10,0]]}]}]
      }
      {
       type=MovieSceneFloatTrack
       display_name=Height Override
       sections [
        {type=MovieSceneFloatSection, end_frame=20, start_frame=0, channels [None {type=MovieSceneScriptingFloatChannel, keys [[0,590],[20,0]]}]}
        {type=MovieSceneFloatSection, end_frame=20, start_frame=0, channels [None {type=MovieSceneScriptingFloatChannel, keys [[0,440],[20,0]]}]}
       ]
      }
     ]
    }
    Background {
     id=#5
     possessed_class=Image
     tracks [{
       type=MovieSceneWidgetMaterialTrack
       sections [{
         type=MovieSceneParameterSection
         end_frame=20
         start_frame=0
         channels [AnimElipseTransition {type=MovieSceneScriptingFloatChannel, keys [[0,0],[20,1]]}]
        }]
      }]
    }
    NavigationButtons {
     id=#6
     possessed_class=GridPanel
     tracks [
      {
       type=MovieSceneFloatTrack
       display_name=Render Opacity
       sections [{type=MovieSceneFloatSection, end_frame=10, start_frame=0, channels [None {type=MovieSceneScriptingFloatChannel, keys [[0,1],[10,0]]}]}]
      }
      {
       type=MovieSceneByteTrack
       display_name=Visibility
       sections [{
         type=MovieSceneByteSection
         end_frame=20
         start_frame=0
         channels [None {type=MovieSceneScriptingByteChannel, thinned_from=5, keys [[0,4],[10,4],[11,1],[20.0753,1]]}]
        }]
      }
     ]
    }
   ]
   tracks [{
     type=MovieSceneEventTrack
     sections [{
       type=MovieSceneEventTriggerSection
       channels [None {type=MovieSceneScriptingEventChannel, const ["<Struct 'MovieSceneEvent' (0x466b2917620) {}>", 1]}]
      }]
    }]
  }
 }
 Anim_BackgroundColor {
  binding_count=1
  length_seconds=0.9955
  object=&self:Anim_BackgroundColor
  sequence {bindings [Background {
     id=#7
     possessed_class=Image
     tracks [{
       type=MovieSceneWidgetMaterialTrack
       sections [{
         type=MovieSceneParameterSection
         end_frame=20
         start_frame=0
         channels [AnimColorTransition {type=MovieSceneScriptingFloatChannel, keys [[0.0907,0],[20,1]]}]
        }]
      }]
    }]}
 }
 HideLandingPage {
  binding_count=3
  length_seconds=1.0038
  object=&self:HideLandingPage
  sequence {bindings [
    TournamentsContent {
     id=#8
     possessed_class=GridPanel
     tracks [{
       type=MovieSceneFloatTrack
       display_name=Render Opacity
       sections [{type=MovieSceneFloatSection, end_frame=10, start_frame=0, channels [None {type=MovieSceneScriptingFloatChannel, keys [[0,0.4],[10,1]]}]}]
      }]
    }
    LandingSB {
     id=#9
     possessed_class=SizeBox
     tracks [
      {
       type=MovieSceneFloatTrack
       display_name=Height Override
       sections [
        {type=MovieSceneFloatSection, end_frame=20, start_frame=0, channels [None {type=MovieSceneScriptingFloatChannel, keys [[0,671],[20,0]]}]}
        {type=MovieSceneFloatSection, end_frame=20, start_frame=0, channels [None {type=MovieSceneScriptingFloatChannel, keys [[0,671],[20,0]]}]}
       ]
      }
      {
       type=MovieSceneFloatTrack
       display_name=Render Opacity
       sections [{type=MovieSceneFloatSection, end_frame=10, start_frame=0, channels [None {type=MovieSceneScriptingFloatChannel, keys [[0,1],[10,0]]}]}]
      }
     ]
    }
    Background {
     id=#10
     possessed_class=Image
     tracks [{
       type=MovieSceneWidgetMaterialTrack
       sections [{
         type=MovieSceneParameterSection
         end_frame=20
         start_frame=0
         channels [AnimElipseTransition {type=MovieSceneScriptingFloatChannel, keys [[0,1],[20.0753,0]]}]
        }]
      }]
    }
   ]}
 }
]
graphs [
 AdjustColorLuminance {nodes [
   n1 = AdjustColorLuminance!() -> (then, In Color:Linear Color Structure) {is=title K2Node_FunctionEntry}
   n2 = Get Max Color Luminance() -> (Max Color Luminance:double)
   n3 = MakeColor!(R:float=@n2, G:float=@n2, B:float=@n2, A:float=1) -> (ReturnValue:Linear Color Structure)
   n4 = "Luminance (LinearColor)"!(InColor:Linear Color Structure=@n1) -> (ReturnValue:float)
   > Return Node!(
    execute
    Out Color:Linear Color Structure=SelectColor(
     A=@n1
     B=LinearColor * LinearColor(A:Linear Color Structure=@n1, B:Linear Color Structure=@n3) -> (ReturnValue:Linear Color Structure) {is=title K2Node_PromotableOperator}
     bPickA="float < float"(A=@n4, B=@n2)
    )
   )
  ]}
 BP_GetDesiredFocusTarget {nodes [
   Get Desired Focus Target()
   n1 = Get WBP_CompeteLandingPage() -> (WBP_CompeteLandingPage:WBP Compete Landing Page Object Reference)
   n2 = IsCompeteLandingPageEnabled!(self:Tournaments Calendar VM Object Reference=Get FortPoblanoTournamentsCalendarVM() -> (FortPoblanoTournamentsCalendarVM:Tournaments Calendar VM Object Reference)) -> (ReturnValue:bool)
   n3 = Get IsLandingPageOpened() -> (IsLandingPageOpened:bool)
   n4 = Get IsCalendarOpened() -> (IsCalendarOpened:bool)
   n5 = Get Selected Calendar Tile() -> (Selected Calendar Tile:Common Button Base Object Reference)
   n6 = Get ListView!(self:Fort Polymorphic List View Object Reference=Get WBP_TournamentsCarousel() -> (WBP_TournamentsCarousel:WBP Tournaments Carousel Object Reference)) -> (ListView:Common List View Object Reference)
   > Return Node!(
    execute
    ReturnValue:Widget Object Reference=Select(
     Option 0:Widget Object Reference=Select(Option 0=@n6, Option 1=@n5, Index=@n4)
     Option 1:Widget Object Reference=GetDesiredFocusTarget!(self:Common Activatable Widget Object Reference=@n1) -> (ReturnValue:Widget Object Reference)
     Index=AND Boolean(A=@n3, B=@n2)
    ) -> (ReturnValue:Widget Object Reference)
   )
  ]}
 DayOfWeek {nodes [
   n1 = DayOfWeek!() -> (then, Date:Date Time Structure) {is=title K2Node_FunctionEntry}
   n2 = int + int(
    A="% (Integer)"(A=Get Day Of Year!(A:Date Time Structure=@n1) -> (ReturnValue:int=0) {is=title K2Node_CallFunction}, B=7) -> (ReturnValue=0)
    B=6
   )
   n3 = SelectInt(A=int - int(A=@n2, B=7), B=@n2, bPickA="integer > integer"(A=@n2, B=6))
   > Return Node!(
    execute
    0-6:int=@n3
    Day:Text=Select!(
     Option 0:Text="LOCTEXT(\"Mo\")"
     Option 1:Text="LOCTEXT(\"Tu\")"
     Option 2:Text="LOCTEXT(\"We\")"
     Option 3:Text="LOCTEXT(\"Th\")"
     Option 4:Text="LOCTEXT(\"Fr\")"
     Option 5:Text="LOCTEXT(\"Sa\")"
     Option 6:Text="LOCTEXT(\"Su\")"
     Index:int=@n3
    ) -> (ReturnValue:Text)
   )
  ]}
 EventGraph {nodes [
   n1 = "On Clicked (DecoyButton_Bookmark)"!() -> (OutputDelegate, then, Button:Common Button Base Object Reference) {is=title K2Node_ComponentBoundEvent}
   > Fire Event Element Interacted(Element=@n1.Button)
   n2 = Get WBP_TournamentsCarousel()
   > n3 = Cast To FortPoblanoTournamentsCalendarTournamentTileVM!(
    execute
    Object=Get Selected Item!(self:List View Object Reference=Get ListView(self=@n2)) -> (ReturnValue:Object Reference) {is=title K2Node_CallFunction}
   ) -> (then, CastFailed, AsCalendar Tournament Tile VM:Calendar Tournament Tile VM Object Reference) {is=title K2Node_DynamicCast}
   > ToggleTournamentTileBookmarked!(
    execute
    self:Tournaments Calendar VM Object Reference=Get FortPoblanoTournamentsCalendarVM()
    TournamentTile:Calendar Tournament Tile VM Object Reference=@n3
   ) -> (then)
   > UpdateBookmarkButtonEnabled()
   n4 = "On Clicked (PreviousButton)"!() -> (OutputDelegate, then, Button:Common Button Base Object Reference) {is=title K2Node_ComponentBoundEvent}
   > Fire Event Element Interacted(Element=@n4.Button)
   > Branch(Condition="integer > integer"(A=Get Selected Tournament Tile() -> (Selected Tournament Tile:int=0), B=1))
   n5 = Get Selected Tournament Tile() -> (Selected Tournament Tile=0)
   n6 = Get ListView(self=Get WBP_TournamentsCarousel())
   > SetSelectedIndex!(execute, self:List View Object Reference=@n6, Index:int=int - int(A=@n5)) -> (then)
   "On Swipe Right (SwipePanel)"() {is=title K2Node_ComponentBoundEvent}
   n7 = Get FortPoblanoTournamentsCalendarVM()
   n8 = AND Boolean!(
    A=Get IsCalendarOpened()
    B=IsUsingTouch(PlayerController=GetPlayerController!(PlayerIndex:int=0) -> (ReturnValue:Player Controller Object Reference))
    C:bool=NAND Boolean(A=IsCompeteLandingPageEnabled(self=@n7), B=Get IsLandingPageOpened()) {is=title K2Node_CommutativeAssociativeBinaryOperator}
   ) -> (ReturnValue)
   > Branch(Condition=@n8)
   n9 = Get CalendarWrapper() -> (CalendarWrapper:WBP Calendar Wrapper Object Reference)
   > Call On Clicked Next Button!(execute:Exec, self:WBP Calendar Wrapper Object Reference=@n9) -> (then:Exec) {is=title K2Node_CallDelegate}
   "On User Scrolled (ScrollBox_Month)"!() -> (OutputDelegate, then, CurrentOffset:float) {is=title K2Node_ComponentBoundEvent}
   n10 = Get ScrollBox_Month() -> (ScrollBox_Month:Scroll Box Object Reference)
   > Branch(Condition="float < float"(A=GetOverscrollPercentage!(self:Scroll Box Object Reference=@n10) -> (ReturnValue:float), B=-10))
   > HideCalendar!(execute, self, B Instant Animations:bool=false) -> (then, ReturnValue:Widget Object Reference)
   n11 = "On Clicked (DecoyButton_Today)"!() -> (OutputDelegate, then, Button:Common Button Base Object Reference) {is=title K2Node_ComponentBoundEvent}
   > n12 = Fire Event Element Interacted(Element=@n11.Button)
   SelectCurrentDateTournament(execute=[@n14.then, @n12.then])
   n13 = "On Clicked (DecoyButton_FirstUpcoming)"!() -> (OutputDelegate, then, Button:Common Button Base Object Reference) {is=title K2Node_ComponentBoundEvent}
   > n14 = Fire Event Element Interacted(Element=@n13.Button)
   "On Clicked (ArrowButton_Bottom)"!() -> (OutputDelegate, then, Button:Common Button Base Object Reference) {is=title K2Node_ComponentBoundEvent}
   n15 = Get FortPoblanoTournamentsCalendarVM()
   > n16 = Branch(Condition=AND Boolean!(A=Get IsLandingPageOpened(), B=IsCompeteLandingPageEnabled(self=@n15)) -> (ReturnValue))
   > n17 = SwipeLandingPage!(execute, self, IsShow:bool=false) -> (then, ReturnValue:Widget Object Reference)
   RefreshFocusIfEnabled(execute=[@n17.then, @n18.then])
   n18 = ShowCalendar!(execute=@n16.else, self) -> (then, ReturnValue:Widget Object Reference)
   "On Clicked (ArrowButton_Top)"!() -> (OutputDelegate, then, Button:Common Button Base Object Reference) {is=title K2Node_ComponentBoundEvent}
   n19 = Get IsCalendarOpened()
   n20 = Get IsLandingPageOpened()
   n21 = Get FortPoblanoTournamentsCalendarVM()
   > n22 = Branch(Condition=AND Boolean!(A=NOT Boolean(A=@n20), B=NOT Boolean(A=@n19), C:bool=IsCompeteLandingPageEnabled(self=@n21)) -> (ReturnValue))
   > n23 = SwipeLandingPage(IsShow=true)
   RefreshFocusIfEnabled(execute=[@n24.then, @n23.then])
   n24 = HideCalendar(execute=@n22.else, B Instant Animations=false)
   "On Clicked (DecoyButton_PowerRankingEligibility)"!() -> (OutputDelegate, then, Button:Common Button Base Object Reference) {is=title K2Node_ComponentBoundEvent}
   > QueueModalWidgetSoftClassAsync!(
    execute:Exec
    SoftWidgetClass:Common Activatable Widget Soft Class Reference=/Compete/PowerRanking/WBP_PowerRanking_InfoModal.WBP_PowerRanking_InfoModal_C
   ) -> (then:Exec, OnModalWidgetQueued:Exec, ActivatableWidget:Common Activatable Widget Object Reference) {is=title K2Node_AsyncAction}
   n25 = "On Clicked (Button_Eligibility)"!() -> (OutputDelegate, then, Button:Common Button Base Object Reference) {is=title K2Node_ComponentBoundEvent}
   > Fire Event Element Interacted(Element=@n25.Button)
   n26 = Get Selected Item(self=Get ListView(self=Get WBP_TournamentsCarousel()))
   > Is Valid(InputObject=@n26)
   > n27 = Cast To FortPoblanoTournamentsCalendarTournamentTileVM(Object=@n26)
   n28 = Get EligibilityWidget() -> (EligibilityWidget:WBP Eligibility Check Modal Object Reference)
   > n29 = Is Valid(InputObject=@n28)
   n30 = Get EligibilityRequirementsModalVM!(self:Calendar Tournament Tile VM Object Reference=@n27) -> (EligibilityRequirementsModalVM:Tournament Requirements Modal VM Object Reference)
   n31 = Create Widget!(
    execute:Exec=@n29.Is Not Valid
    Class:User Widget Class Reference=/Compete/TournamentLeaderboard/ModalMessages/EligibilityCheck/WBP_EligibilityCheckModal.WBP_EligibilityCheckModal_C
    OwningPlayer:Player Controller Object Reference=GetOwningPlayer()
    FortTournamentRequirementsModalVM:Tournament Requirements Modal VM Object Reference=@n30
   ) -> (then:Exec, ReturnValue:WBP Eligibility Check Modal Object Reference)
   > n32 = Set EligibilityWidget(EligibilityWidget:WBP Eligibility Check Modal Object Reference=@n31) -> (Output_Get:WBP Eligibility Check Modal Object Reference)
   n33 = Get EligibilityWidget()
   TryPushWidgetAsModalPanel!(execute=[@n32.then, @n37.then], PushWidget:User Widget Object Reference=@n33) -> (then)
   > n34 = ActivateWidget(self=@n33)
   n35 = OnEligibilityWidgetDeactivated() {is=title K2Node_CustomEvent}
   Bind Event to On Widget Deactivated(execute=@n34.then, self:Common Activatable Widget Object Reference=@n33, Delegate=@n35) {is=title K2Node_AddDelegate}
   n36 = Get FortTournamentRequirementsModalVM!(self:WBP Eligibility Check Modal Object Reference=Get EligibilityWidget()) -> (FortTournamentRequirementsModalVM:Tournament Requirements Modal VM Object Reference)
   > SetupViewModel(self:Tournament Requirements Modal VM Object Reference=@n36)
   n37 = SetFortTournamentRequirementsModalVM!(
    execute=@n29.Is Valid
    self:WBP Eligibility Check Modal Object Reference=@n28
    Viewmodel:Tournament Requirements Modal VM Object Reference=@n30
   ) -> (then)
   "On Clicked Previous Button (CalendarWrapper)"() {is=title K2Node_ComponentBoundEvent}
   n38 = Get FortPoblanoTournamentsCalendarVM()
   > n39 = DisplayPreviousMonth(self:Tournaments Calendar VM Object Reference=@n38)
   Branch(execute=[@n40.then, @n39.then], Condition=Get IsCalendarOpened())
   > SetNewSelectedTile!(
    execute
    self
    Selected Calendar Tile:Common Button Base Object Reference="Get (a copy)"(Array:Array of WBP Calendar Tile Object References=Get CalendarDayTilesArray() -> (CalendarDayTilesArray:Array of WBP Calendar Tile Object References)) -> (Output:WBP Calendar Tile Object Reference)
   ) -> (then)
   > RefreshFocusIfEnabled()
   "On Clicked Next Button (CalendarWrapper)"() {is=title K2Node_ComponentBoundEvent}
   > n40 = DisplayNextMonth(self:Tournaments Calendar VM Object Reference=@n38)
   "On Swipe Left (SwipePanel)"() {is=title K2Node_ComponentBoundEvent}
   > Branch(Condition=@n8)
   > Call On Clicked Previous Button!(execute:Exec, self:WBP Calendar Wrapper Object Reference=@n9) -> (then:Exec) {is=title K2Node_CallDelegate}
   n41 = "On Clicked (FilterButton)"!() -> (OutputDelegate, then, Button:Common Button Base Object Reference) {is=title K2Node_ComponentBoundEvent}
   > Fire Event Element Interacted(Element=@n41.Button)
   > SetIsEnabled(self=Get ListView(self=Get WBP_TournamentsCarousel()), bInIsEnabled=false)
   > n42 = Create Widget!(
    execute
    Class=/Compete/Filter/WBP_TournamentsFilter.WBP_TournamentsFilter_C
    OwningPlayer=GetOwningPlayer()
    FortPoblanoTournamentsCalendarFilterVM:Tournaments Calendar Filter VM Object Reference=Get CalendarFilterVM!(self:Fort Compete UIManager Object Reference=Get FortCompeteUIManager() -> (ReturnValue:Fort Compete UIManager Object Reference) {is=title K2Node_GetSubsystem}) -> (CalendarFilterVM:Tournaments Calendar Filter VM Object Reference)
   ) -> (then, ReturnValue:WBP Tournaments Filter Object Reference)
   > TryPushWidgetAsModalPanel(PushWidget=@n42)
   > n43 = ActivateWidget(self=@n42)
   n44 = RestoreEnabled() {is=title K2Node_CustomEvent}
   Bind Event to On Widget Deactivated(execute=@n43.then, self=@n42, Delegate=@n44)
   > SetIsEnabled(bInIsEnabled=false)
   n45 = "On Clicked (NextButton)"!() -> (OutputDelegate, then, Button:Common Button Base Object Reference) {is=title K2Node_ComponentBoundEvent}
   > Fire Event Element Interacted(Element=@n45.Button)
   > Branch(Condition="integer < integer"(A=Get Selected Tournament Tile() -> (Selected Tournament Tile=0), B=GetTournamentsLength!() -> (Length:int=0)))
   > SetSelectedIndex(self=@n6, Index=int + int(A=@n5))
   "On Swipe Up (SwipePanel)"() {is=title K2Node_ComponentBoundEvent}
   n46 = Get FortPoblanoTournamentsCalendarVM()
   n47 = Get IsLandingPageOpened()
   n48 = Get IsCalendarOpened()
   > n49 = Branch(Condition=AND Boolean(A=NOT Boolean(A=@n47), B=NOT Boolean(A=@n48), C=IsCompeteLandingPageEnabled(self=@n46)))
   > n50 = SwipeLandingPage(IsShow=true)
   RefreshFocusIfEnabled(execute=[@n50.then, @n51.then])
   n51 = HideCalendar(execute=@n49.else, B Instant Animations=false)
   "On Swipe Down (SwipePanel)"() {is=title K2Node_ComponentBoundEvent}
   n52 = Get FortPoblanoTournamentsCalendarVM()
   > n53 = Branch(Condition=AND Boolean!(A=Get IsLandingPageOpened(), B=IsCompeteLandingPageEnabled(self=@n52)) -> (ReturnValue))
   > n54 = SwipeLandingPage(IsShow=false)
   RefreshFocusIfEnabled(execute=[@n55.then, @n54.then])
   n55 = ShowCalendar(execute=@n53.else)
   n56 = "On Clicked (DecoyButton_Back)"!() -> (OutputDelegate, then, Button:Common Button Base Object Reference) {is=title K2Node_ComponentBoundEvent}
   > Fire Event Element Interacted(Element=@n56.Button)
   > HideCalendar(B Instant Animations=false)
   > RefreshFocusIfEnabled()
   UpdateVM() {is=title K2Node_CustomEvent}
   > ClearListItems(self=Get ListView(self=Get WBP_TournamentsCarousel()))
   > n57 = Is Valid(InputObject=Get FortPoblanoTournamentsCalendarVM())
   n58 = SetFortPoblanoTournamentsCalendarVM!(
    execute=@n57.Is Not Valid
    self
    Viewmodel:Tournaments Calendar VM Object Reference=Get CalendarVM!(self:Fort Compete UIManager Object Reference=Get FortCompeteUIManager()) -> (CalendarVM:Tournaments Calendar VM Object Reference)
   ) -> (then)
   InitializeVM(execute=[@n58.then, @n57.Is Valid], self:Tournaments Calendar VM Object Reference=Get FortPoblanoTournamentsCalendarVM())
   > Update Filter Info(self:WBP Filter Button Object Reference=Get FilterButton() -> (FilterButton:WBP Filter Button Object Reference))
   > HideCalendar(B Instant Animations=true)
   > Set Month Text() {is=title K2Node_CallFunction}
   OnTournamentCarouselFinishedScrolling() {is=title K2Node_CustomEvent}
   n59 = "Equal (Enum)"(A=Get Input Type() -> (Input Type:ECommonInputType Enum), B=Touch)
   > Branch(Condition=@n59)
   n60 = GetScrollOffset!(self:List View Base Object Reference=Get ListView(self=Get WBP_TournamentsCarousel())) -> (ReturnValue:float)
   > n61 = Branch(Condition="integer > integer"(A=Round(A=@n60) {is=title K2Node_CallFunction}, B=GetTournamentsLength() -> (Length=0)))
   n62 = Get ListView(self=Get WBP_TournamentsCarousel())
   > NavigateToIndex!(execute, self:List View Object Reference=@n62, Index:int=GetTournamentsLength() -> (Length=0)) -> (then)
   > EndInertialScrolling(self:List View Base Object Reference=@n62)
   UpdateSelectedTileFromScrollOffset!(execute=@n61.else, self, ScrollOffset:double=@n60) -> (then)
   n63 = OnTournamentCarouselSelectionChanged!() -> (OutputDelegate, then, Item:Object Reference, bIsSelected:bool=~) {is=title K2Node_CustomEvent}
   > n64 = Branch(Condition=@n63.bIsSelected)
   > n65 = Cast To FortPoblanoTournamentsCalendarTournamentTileVM(Object=@n63.Item)
   > n66 = Set CurrentSelectedTileVM(CurrentSelectedTileVM:Calendar Tournament Tile VM Object Reference=@n65) -> (Output_Get:Calendar Tournament Tile VM Object Reference)
   > n67 = Scroll to Tournament Tile!(execute, self, In Tournament Tile:Calendar Tournament Tile VM Object Reference=@n66) -> (then)
   UpdateBookmarkButtonEnabled(execute=[@n67.then, @n68.then])
   > UpdateEligibilityButtonEnabled()
   n68 = Set CurrentSelectedTileVM(execute=[@n65.CastFailed, @n64.else])
   n69 = OnTournamentCarouselScrolled!() -> (OutputDelegate, then, ItemOffset:float=~, DistanceRemaining:float=~) {is=title K2Node_CustomEvent}
   > Branch(Condition=@n59)
   > n70 = Branch(Condition="integer > integer"(A=Round(A=@n69.ItemOffset), B=GetTournamentsLength() -> (Length=0)))
   n71 = Get ListView(self=Get WBP_TournamentsCarousel())
   > SetScrollOffset!(execute, self:List View Base Object Reference=@n71, InScrollOffset:float="To Float (Integer)"(InInt=GetTournamentsLength() -> (Length=0))) -> (then)
   > EndInertialScrolling(self=@n71)
   UpdateSelectedTileFromScrollOffset(execute=@n70.else, ScrollOffset=@n69.ItemOffset)
   n72 = HandleDualBladeMenuDeactivated() {is=title K2Node_CustomEvent}
   > n73 = Branch(Condition=Get IsCalendarOpened())
   > n74 = Is Valid(InputObject=Get Selected Calendar Tile())
   SetNewSelectedTile(
    execute=@n74.Is Not Valid
    Selected Calendar Tile="Get (a copy)"(Array=Get CalendarDayTilesArray(), Dimension 1=Get PreviouslySelectedCalendarTileIndex() -> (PreviouslySelectedCalendarTileIndex:int=0))
   )
   > RefreshFocusIfEnabled()
   RefreshFocusIfEnabled(execute=@n74.Is Valid)
   RestoreCalendarFocusAfterProcessingFilter(execute=@n73.else)
   Event OnInitialized()
   > Parent: OnInitialized() {is=title K2Node_CallParentFunction}
   > Branch(Condition=IsMobileGame())
   > SetPadding!(
    execute
    self:Grid Slot Object Reference=SlotAsGridSlot(Widget=Get SafeZone_0() -> (SafeZone_0:Safe Zone Object Reference))
    InPadding_Left:float=48
    InPadding_Top:float=0
    InPadding_Right:float=0
    InPadding_Bottom:float=24
   ) -> (then)
   Event PreConstruct()
   > SetColorAndOpacity(
    self:User Widget Object Reference=[
     Get Block_Icon!(self:WBP UIKit Button Generic Base Object Reference=Get NextButton() -> (NextButton:WBP UIKit Button Generic Object Reference)) -> (Block_Icon:WBP UIKit Block Image Object Reference)
     Get Block_Icon(self=Get PreviousButton() -> (PreviousButton:WBP UIKit Button Generic Object Reference))
    ]
    InColorAndOpacity:Linear Color Structure=(R=0,G=0,B=0,A=1)
   )
   > n75 = GetDynamicMaterial(self=Get Background() -> (Background:Image Object Reference))
   > Set MID_Background(MID_Background:Material Instance Dynamic Object Reference=@n75) -> (Output_Get:Material Instance Dynamic Object Reference)
   > n76 = Construct Object from Class!(execute, Class=/Compete/Calendar/Data/MVVM_CompeteEmptyTile_Entry.MVVM_CompeteEmptyTile_Entry_C, self, IsSelectable:bool=False) -> (then, ReturnValue:MVVM Compete Empty Tile Entry Object Reference)
   > Set FirstItem(FirstItem:MVVM Compete Empty Tile Entry Object Reference=@n76) -> (Output_Get:MVVM Compete Empty Tile Entry Object Reference)
   > n77 = Branch(Condition=IsUsingCompactUI())
   > Clear(TargetArray:Array of MVVM Compete Empty Tile Entry Object References=Get CompactUILastItems() -> (CompactUILastItems:Array of MVVM Compete Empty Tile Entry Object References))
   > For Loop(LastIndex=9)
   > n78 = Construct Object from Class(Class=&MVVM_CompeteEmptyTile_Entry_C, IsSelectable=False)
   > Add(
    TargetArray:Array of MVVM Compete Empty Tile Entry Object References=Get CompactUILastItems()
    NewItem:"MVVM Compete Empty Tile Entry Object Reference (by ref)"=@n78
   ) {is=title K2Node_CallArrayFunction}
   n79 = Construct Object from Class(execute=@n77.else, Class=&MVVM_CompeteEmptyTile_Entry_C, IsSelectable=False)
   > Set LastItem(LastItem:MVVM Compete Empty Tile Entry Object Reference=@n79) -> (Output_Get:MVVM Compete Empty Tile Entry Object Reference)
   Event Construct()
   > SetMobileStyle() {is=title K2Node_CallFunction}
   > UpdateVM() {is=title K2Node_CallFunction}
   > n80 = FindParentWidgetOfType!(
    execute
    StartingWidget:Widget Object Reference=GetParent!(self:Widget Object Reference) -> (ReturnValue:Panel Widget Object Reference)
    Type:Widget Class Reference=/FortniteCoreUI/Frontend/AthenaFrontend.AthenaFrontend_C
   ) -> (then, ReturnValue:Athena Frontend Object Reference)
   > Bind Event to On Widget Deactivated(
    self=Get DualBladeMenu_Widget!(self:Athena UIState Widget Frontend Object Reference=@n80) -> (DualBladeMenu_Widget:Fort Dual Blade Menu Object Reference)
    Delegate=@n72
   )
   Event On Activated()
   > Fire Event Main View Opened!(execute, View:Widget Object Reference=Self-Reference(), AdditionalInfo:Flow Analytics Additional Info Structure, bIsOverlay:bool=false) -> (then)
   > InputMethodChanged!(execute, self, InputType:ECommonInputType Enum=GetCurrentInputType(self=Get CommonInputSubsystem())) -> (then)
   n81 = Create Event()
   n82 = Get CommonInputSubsystem()
   > Bind Event to On Input Method Changed(self=@n82, Delegate=@n81)
   n83 = Create Event()
   n84 = Get ListView(self=Get WBP_TournamentsCarousel())
   > Assign On Item Selection Changed!(execute:Exec, self:List View Object Reference=@n84, Delegate:"Delegate (by ref)"=@n83) -> (then:Exec) {is=title K2Node_AssignDelegate}
   n85 = Create Event()
   > Assign On List View Scrolled!(execute:Exec, self:List View Object Reference=@n84, Delegate:"Delegate (by ref)"=@n85) -> (then:Exec) {is=title K2Node_AssignDelegate}
   n86 = Create Event()
   > Bind Event to On List View Finished Scrolling(self:List View Object Reference=@n84, Delegate=@n86) {is=title K2Node_AddDelegate}
   > UpdateVM()
   > SpawnSound2D!(
    execute
    Sound:Sound Base Object Reference=/Compete/Assets/Audio/MSS_UI_CompeteScreen_WooshToTab
    VolumeMultiplier:float=1
    PitchMultiplier:float=1
    StartTime:float=0
    ConcurrencySettings:Sound Concurrency Object Reference
    bPersistAcrossLevelTransition:bool=false
    bAutoDestroy:bool=true
   ) -> (then, ReturnValue:Audio Component Object Reference)
   > RefreshButtons()
   > RequestRefresh(self:List View Base Object Reference=Get ListView(self=Get WBP_TournamentsCarousel()))
   Event On Deactivated()
   n87 = Get FortCompeteUIManager()
   > n88 = Branch(Condition=GetCanResetCalendarFilterVM!(self:Fort Compete UIManager Object Reference=@n87) -> (ReturnValue:bool))
   > n89 = RestoreRegion!(execute, self:Fort Compete UIManager Object Reference=@n87) -> (then, ReturnValue:bool)
   Unbind Event from On Input Method Changed(execute=[@n89.then, @n88.else], self=@n82, Delegate=@n81)
   > Unbind Event from On Item Selection Changed(self:List View Object Reference=@n84, Delegate=@n83) {is=title K2Node_RemoveDelegate}
   > Unbind Event from On List View Scrolled(self:List View Object Reference=@n84, Delegate=@n85) {is=title K2Node_RemoveDelegate}
   > Unbind Event from On List View Finished Scrolling(self:List View Object Reference=@n84, Delegate=@n86) {is=title K2Node_RemoveDelegate}
   "Animation Finished (Anim_OpenCalendar)"!() -> (OutputDelegate:Delegate, then:Exec) {is=title K2Node_WidgetAnimationEvent}
   > n90 = Branch(Condition=Get IsCalendarOpened())
   n91 = Get WBP_TournamentsCarousel()
   > SetVisibility(self=@n91, InVisibility=Collapsed)
   RefreshFocusIfEnabled(execute=@n90.else)
   > SetVisibility(self=@n91, InVisibility=SelfHitTestInvisible)
   SetIsEnabled(execute=@n93.then, self=Get ListView(self=Get WBP_TournamentsCarousel()), bInIsEnabled=true)
   Deactivate(execute=@n92.then, self:Tournament Requirements Modal VM Object Reference=@n36)
   Reroute Node(InputPin:WBP Tournaments Filter Object Reference=@n42) -> (OutputPin:WBP Tournaments Filter Object Reference)
   n92 = RequestRefreshFocus(execute=@n35.then, self:Common Activatable Widget Object Reference)
   n93 = RefocusCalendarIfDirtyFilter(execute=@n44.then)
  ]}
 GenerateMonth {nodes [
   GenerateMonth!() -> (then, TilesData:Calendar Tiles Data Structure) {is=title K2Node_FunctionEntry}
   > Unbind all Events from On Selected Changed!(execute:Exec, self:Common Button Base Object Reference=Get CalendarDayTilesArray()) -> (then:Exec) {is=title K2Node_ClearDelegate}
   > Unbind all Events from On Clicked!(execute:Exec, self:Common Button Base Object Reference=Get CalendarDayTilesArray()) -> (then:Exec) {is=title K2Node_ClearDelegate}
   > ClearChildren(self=Get Days() -> (Days:UIKit Grid Panel Object Reference))
   > Clear(TargetArray:Array of WBP Calendar Tile Object References=Get CalendarDayTilesArray())
   n1 = Break Calendar Tiles Data!(CalendarTilesData:"Calendar Tiles Data Structure (by ref)"=Get CalendarTilesData!(self:Tournaments Calendar VM Object Reference=Get FortPoblanoTournamentsCalendarVM()) -> (CalendarTilesData:Calendar Tiles Data Structure)) -> (DayTiles:Array of Calendar Monthly Day Tile VM Object References, FirstDayOfWeekForCurrentMonth:int) {is=title K2Node_BreakStruct}
   > Set L_FirstDay(L_FirstDay:int=@n1.FirstDayOfWeekForCurrentMonth) -> (Output_Get:int=0)
   > n2 = For Each Loop(Array:Array of Calendar Monthly Day Tile VM Object References=@n1) -> (Array Element:Calendar Monthly Day Tile VM Object Reference)
   n3 = Get CalendarDayTilesArray()
   n4 = Branch(execute=@n2.Completed, Condition=Is Empty!(TargetArray:Array of WBP Calendar Tile Object References=@n3) -> (ReturnValue:bool))
   Bind Event to On Selected Changed(execute=@n4.else, self:Common Button Base Object Reference=@n3, Delegate=Create Event()) {is=title K2Node_AddDelegate}
   > Bind Event to On Clicked(self:Common Button Base Object Reference=@n3, Delegate=Create Event()) {is=title K2Node_AddDelegate}
   > SetAsToday!(execute, self:WBP Calendar Tile Object Reference=Get CalendarDayTilesArray(), IsToday:bool=false) -> (then)
   > Scroll to Selected Tournament In Carousel()
   > Set Month Text()
   n5 = Get CurrentTime!(self:Tournaments Calendar VM Object Reference=Get FortPoblanoTournamentsCalendarVM()) -> (CurrentTime:Date Time Structure)
   > n6 = Branch(Condition=Is Displaying Month!(In Date:"Date Time Structure (by ref)"=@n5) -> (B Is Displaying Current Month:bool))
   n7 = Get Day!(A:Date Time Structure=@n5) -> (ReturnValue:int=0) {is=title K2Node_CallFunction}
   > n8 = SetAsToday(self="Get (a copy)"(Array=Get CalendarDayTilesArray(), Dimension 1=int - int(A=@n7)), IsToday=true)
   PlayTransitionShowIcons!(
    execute=[@n8.then, @n6.else]
    self:WBP Calendar Tile Object Reference=Get CalendarDayTilesArray()
    PlayMode:EUMGSequencePlayMode Enum=Select!(Option 0:EUMGSequencePlayMode Enum=Reverse, Option 1:EUMGSequencePlayMode Enum, Index:bool=Get IsCalendarOpened()) -> (ReturnValue:EUMGSequencePlayMode Enum)
    Instant:bool=true
   ) -> (then)
   n9 = int + int(
    A=Last Index!(TargetArray:Array of WBP Calendar Tile Object References=Get CalendarDayTilesArray()) -> (ReturnValue=0)
    B=Get L_FirstDay() -> (L_FirstDay:int=0)
   )
   > Set IsCalendarScrollableForTouch(IsCalendarScrollableForTouch:bool="integer > integer"(A=int / int(A=@n9, B=7), B=4)) -> (Output_Get:bool)
   > n10 = Branch(Condition=Get IsCalendarOpened())
   > n11 = Set bEnableTouchScrolling!(
    execute
    bEnableTouchScrolling:bool=Get IsCalendarScrollableForTouch() -> (IsCalendarScrollableForTouch:bool)
    self:Scroll Box Object Reference=Get ScrollBox_Month()
   ) -> (then, Output_Get:bool)
   CheckIsScrollable(execute=[@n11.then, @n12.then])
   > Call On Month Just Generated!(execute:Exec, self:WBP Calendar Object Reference) -> (then:Exec) {is=title K2Node_CallDelegate}
   n12 = Set bEnableTouchScrolling(execute=@n10.else, bEnableTouchScrolling=false, self=Get ScrollBox_Month())
   n13 = int + int(A=@n2.Array Index, B=Get L_FirstDay() -> (L_FirstDay=0))
   n14 = CreateChildOnGrid!(
    execute=@n2.LoopBody
    self:UIKit Grid Panel Object Reference=Get Days()
    WidgetClass:"User Widget Class Reference (by ref)"=Get Tile() -> (Tile:WBP Calendar Tile Class Reference)
    InRow:int=int / int(A=@n13, B=7)
    InColumn:int="% (Integer)"(A=@n13, B=7) -> (ReturnValue=0)
   ) -> (then, ReturnValue:Grid Slot Object Reference)
   > SetPadding(self=@n14, InPadding_Left=8, InPadding_Top=8, InPadding_Right=8, InPadding_Bottom=8)
   > n15 = Cast To WBP_CalendarTile!(execute, Object=GetContent!(self:Panel Slot Object Reference=@n14) -> (ReturnValue:Widget Object Reference)) -> (then, CastFailed, AsWBP Calendar Tile:WBP Calendar Tile Object Reference) {is=title K2Node_DynamicCast}
   > SetFortPoblanoTournamentsCalendarDayTileVM!(execute, self:WBP Calendar Tile Object Reference=@n15, Viewmodel:Calendar Monthly Day Tile VM Object Reference=@n2) -> (then)
   > Add(
    TargetArray:Array of WBP Calendar Tile Object References=Get CalendarDayTilesArray()
    NewItem:"WBP Calendar Tile Object Reference (by ref)"=@n15
   )
  ]}
 GenerateTournaments {nodes [
   n1 = GenerateTournaments!() -> (
    then
    Tournaments_TournamentTiles:Array of Calendar Tournament Tile VM Object References
    Tournaments_PreviouslySelectedEvent:int
    Tournaments_FirstUpcomingEvent:int
    Tournaments_FirstLiveEvent:int
    Tournaments_FirstCurrentDateEvent:int
    Tournaments_ClosestEvent:int
    Tournaments_bIsLoadingTournaments:bool=~
   ) {is=title K2Node_FunctionEntry}
   > n2 = Set IsLoadingTournaments(IsLoadingTournaments:bool=@n1.Tournaments_bIsLoadingTournaments) -> (Output_Get:bool)
   > SetIsEnabled(self=Get FilterButton(), bInIsEnabled=NOT Boolean(A=@n2))
   > Clear(TargetArray:Array of Object References=Get TilesVMArr() -> (TilesVMArr:Array of Object References))
   > Unbind all Events from On Selected Changed(self=Get TournamentTilesArray() -> (TournamentTilesArray:Array of WBP Tournament Tile Object References))
   > Unbind all Events from On Clicked(self=Get TournamentTilesArray())
   > ClearListItems(self=Get ListView(self=Get WBP_TournamentsCarousel()))
   > Clear(TargetArray:Array of WBP Tournament Tile Object References=Get TournamentTilesArray())
   > SetTilesState()
   n3 = Break Tournaments Tiles Data!(TournamentsTilesData:"Tournaments Tiles Data Structure (by ref)"=Get Tournaments() -> (Tournaments:Tournaments Tiles Data Structure)) -> (TournamentTiles:Array of Calendar Tournament Tile VM Object References, FirstCurrentDateEvent:int, ClosestEvent:int) {is=title K2Node_BreakStruct}
   > n4 = Branch(Condition=Is Empty(TargetArray:Array of Calendar Tournament Tile VM Object References=@n3))
   > ClearListItems(self=Get ListView(self=Get WBP_TournamentsCarousel()))
   > Set CurrentSelectedTileVM()
   n5 = Get TilesVMArr()
   Add(
    execute=@n4.else
    TargetArray:Array of Object References=@n5
    NewItem:"Object Reference (by ref)"=Get FirstItem() -> (FirstItem:MVVM Compete Empty Tile Entry Object Reference)
   )
   > Append Array!(execute, TargetArray:Array of Object References=@n5, SourceArray:Array of Object References=@n3) -> (then)
   > n6 = Branch(Condition=IsUsingCompactUI())
   > n7 = Append Array(TargetArray=@n5, SourceArray=Get CompactUILastItems())
   SetListItems!(
    execute=[@n19.then, @n7.then]
    self:Fort Polymorphic View Object Reference=Get WBP_TournamentsCarousel()
    InListItems:Array of Object References=@n5
   ) -> (then)
   n8 = Break Tournaments Tiles Data!(TournamentsTilesData=Get TournamentsTilesData!(self:Tournaments Calendar VM Object Reference=Get FortPoblanoTournamentsCalendarVM()) -> (TournamentsTilesData:Tournaments Tiles Data Structure)) -> (TournamentTiles, PreviouslySelectedEvent:int, FirstUpcomingEvent:int, FirstLiveEvent:int, FirstCurrentDateEvent, ClosestEvent)
   > n9 = Branch(Condition=Is Empty(TargetArray=@n8))
   n10 = Branch(execute=@n9.else, Condition="integer >= integer"(A=@n8.PreviouslySelectedEvent))
   > n11 = Set New Selected Tile Index(New Selected Tile Index:int=@n8.PreviouslySelectedEvent) -> (Output_Get:int=0)
   SetSelectedIndex(
    execute=[@n11.then, @n18.then, @n17.then, @n13.then, @n15.then]
    self=Get ListView(self=Get WBP_TournamentsCarousel())
    Index=int + int(A=Get New Selected Tile Index() -> (New Selected Tile Index:int=0))
   )
   > RefreshFocusIfEnabled()
   n12 = Branch(execute=@n10.else, Condition="integer >= integer"(A=@n8.FirstCurrentDateEvent))
   > n13 = Set New Selected Tile Index(New Selected Tile Index=@n8.FirstCurrentDateEvent) -> (Output_Get=0)
   n14 = Branch(execute=@n12.else, Condition="integer >= integer"(A=@n8.FirstUpcomingEvent))
   > n15 = Set New Selected Tile Index(New Selected Tile Index=@n8.FirstUpcomingEvent) -> (Output_Get=0)
   n16 = Branch(execute=@n14.else, Condition="integer >= integer"(A=@n8.ClosestEvent))
   > n17 = Set New Selected Tile Index(New Selected Tile Index=@n8.ClosestEvent) -> (Output_Get=0)
   n18 = Set New Selected Tile Index(execute=@n16.else, New Selected Tile Index=0) -> (Output_Get=0)
   n19 = Add(execute=@n6.else, TargetArray=@n5, NewItem=Get LastItem() -> (LastItem:MVVM Compete Empty Tile Entry Object Reference))
  ]}
 Get Selected Tournament Date {nodes [
   Get Selected Tournament Date() {is=title K2Node_FunctionEntry}
   n1 = Get FortPoblanoTournamentsCalendarVM()
   n2 = "Get (a copy)"(
    Array:Array of Calendar Tournament Tile VM Object References=Break Tournaments Tiles Data!(TournamentsTilesData=Get TournamentsTilesData(self=@n1)) -> (
     TournamentTiles
     PreviouslySelectedEvent
     FirstUpcomingEvent
     FirstLiveEvent
     FirstCurrentDateEvent
     ClosestEvent
     bIsLoadingTournaments:bool=~
    )
    Dimension 1=Get Selected Tournament Tile() -> (Selected Tournament Tile=0)
   ) -> (Output:Calendar Tournament Tile VM Object Reference)
   > Return Node!(
    execute
    Date:Date Time Structure=Select(
     Option 0:Date Time Structure=MakeDateTime!(Year:int=1, Month:int=1, Day:int=1, Hour:int=0, Minute:int=0, Second:int=0, Millisecond:int=0) -> (ReturnValue:Date Time Structure)
     Option 1:Date Time Structure=Get Date!(self:Calendar Tournament Tile VM Object Reference=@n2) -> (Date:Date Time Structure)
     Index=IsValid(Object=@n2)
    ) -> (ReturnValue:Date Time Structure)
   )
  ]}
 GetTournamentsLength {nodes [
   GetTournamentsLength() {is=title K2Node_FunctionEntry}
   n1 = Get TournamentsTilesData(self=Get FortPoblanoTournamentsCalendarVM())
   > Return Node!(
    execute
    Length:int=Length(TargetArray:Array of Calendar Tournament Tile VM Object References=Break Tournaments Tiles Data(TournamentsTilesData=@n1) -> (bIsLoadingTournaments=~))
   )
  ]}
 HideCalendar {nodes [
   n1 = HideCalendar!() -> (then, B Instant Animations:bool=~) {is=title K2Node_FunctionEntry}
   > n2 = Branch(Condition=Get IsCalendarOpened())
   > UpdateIsCalendarOpened!(execute, self, IsOpened:bool=false) -> (then)
   > PlayAnimation(
    InAnimation=Get Anim_OpenCalendar() -> (Anim_OpenCalendar:Widget Animation Object Reference)
    PlayMode=Reverse
    PlaybackSpeed=SelectFloat(A=9999, B=5, bPickA=@n1)
   )
   > PlayButtonAnimation!(execute, self:WBP Calendar Wrapper Object Reference=Get CalendarWrapper(), IsShow:bool=false) -> (then)
   > PlayTransitionShowIcons(self=Get CalendarDayTilesArray(), PlayMode=Reverse, Instant=Get B Instant Animations() -> (B Instant Animations:bool))
   > Set ScrollWhenFocusChanges!(
    execute
    ScrollWhenFocusChanges:EScrollWhenFocusChanges Enum=Select!(
     MouseAndKeyboard:EScrollWhenFocusChanges Enum=AnimatedScroll
     Gamepad:EScrollWhenFocusChanges Enum=AnimatedScroll
     Touch:EScrollWhenFocusChanges Enum=NoScroll
     Count:EScrollWhenFocusChanges Enum=AnimatedScroll
     Index:ECommonInputType Enum=Get Input Type()
    ) -> (ReturnValue:EScrollWhenFocusChanges Enum)
    self:Scroll Box Object Reference=Get ScrollBox_Month()
   ) -> (then, Output_Get:EScrollWhenFocusChanges Enum=NoScroll)
   > Set bEnableTouchScrolling(bEnableTouchScrolling=false, self=Get ScrollBox_Month())
   > SetMobileMonthHeight!(execute, self, IsCalendarShown:bool=Get IsCalendarOpened()) -> (then)
   Return Node!(execute=@n2.else, ReturnValue:Widget Object Reference=Get ListView(self=Get WBP_TournamentsCarousel()))
  ]}
 InputMethodChanged {nodes [
   n1 = InputMethodChanged!() -> (then, InputType:ECommonInputType Enum) {is=title K2Node_FunctionEntry}
   > Set Input Type(Input Type:ECommonInputType Enum=@n1) -> (Output_Get:ECommonInputType Enum)
   > Set ScrollWhenFocusChanges(
    ScrollWhenFocusChanges=Select(MouseAndKeyboard=AnimatedScroll, Gamepad=AnimatedScroll, Touch=NoScroll, Count=AnimatedScroll, Index=Get Input Type())
    self=Get ScrollBox_Month()
   ) -> (Output_Get=NoScroll)
   n2 = Get Input Type()
   > SetVisibility(
    self=[Get PreviousButton(), Get NextButton()]
    InVisibility=Select!(
     Option 0:ESlateVisibility Enum=Collapsed
     Option 1:ESlateVisibility Enum=SelfHitTestInvisible
     Index:bool=Select!(MouseAndKeyboard:bool=true, Gamepad:bool=false, Touch:bool=false, Count:bool=true, Index:ECommonInputType Enum=@n2) -> (ReturnValue:bool)
    ) -> (ReturnValue:ESlateVisibility Enum)
   )
   > UpdateBackButtonEnabled()
   > RefreshFocusIfEnabled()
  ]}
 Is Displaying Month {nodes [
   n1 = Is Displaying Month!() -> (then, In Date:"Date Time Structure (by ref)") {is=title K2Node_FunctionEntry}
   n2 = Get DisplayMonthDate!(self:Tournaments Calendar VM Object Reference=Get FortPoblanoTournamentsCalendarVM()) -> (DisplayMonthDate:Date Time Structure)
   n3 = Get Year!(A:Date Time Structure=@n2) -> (ReturnValue:int=0) {is=title K2Node_CallFunction}
   n4 = Get Year(A=@n1) -> (ReturnValue=0)
   n5 = Get Month!(A:Date Time Structure=@n1) -> (ReturnValue:int=0) {is=title K2Node_CallFunction}
   n6 = Get Month(A=@n2) -> (ReturnValue=0)
   > Return Node!(execute, B Is Displaying Current Month:bool=AND Boolean(A="Equal (Integer)"(A=@n4, B=@n3), B="Equal (Integer)"(A=@n5, B=@n6)))
  ]}
 IsLandingPageEnable {nodes [
   n1 = IsLandingPageEnable!() -> (then, IsEnable:bool=~) {is=title K2Node_FunctionEntry}
   n2 = Get IsLandingPageOpened()
   > SetIsEnabled(self=Get TournamentsContent() -> (TournamentsContent:Grid Panel Object Reference), bInIsEnabled=NAND Boolean(A=@n2, B=@n1))
   > SetVisibility(
    self=[
     Get WBP_CompeteLandingPage()
     Get ArrowBottom() -> (ArrowBottom:Overlay Object Reference)
     Get ArrowTop() -> (ArrowTop:Overlay Object Reference)
    ]
    InVisibility=Select!(Option 0:ESlateVisibility Enum=Collapsed, Option 1:ESlateVisibility Enum=SelfHitTestInvisible, Index:bool=@n1) -> (ReturnValue:ESlateVisibility Enum)
   )
   n3 = AND Boolean(A=@n1, B=@n2)
   > SetRenderOpacity(self=Get TournamentsContent(), InOpacity=Select(Option 0:float=1, Option 1:float=0.4, Index=@n3) -> (ReturnValue:float))
   > Branch(Condition=@n3)
   > SetScalarParameterValue(self=Get MID_Background() -> (MID_Background:Material Instance Dynamic Object Reference), ParameterName=AnimElipseTransition, Value=1)
  ]}
 LerpBackgroundColor {nodes [
   LerpBackgroundColor() {is=title K2Node_FunctionEntry}
   > n1 = Sequence()
   n2 = Get MID_Background()
   > n3 = Get Scalar Parameter Value!(execute, self:Material Instance Dynamic Object Reference=@n2, ParameterName:Name=AnimColorTransition) -> (then, ReturnValue:float) {is=title K2Node_CallFunction}
   > n4 = Get Vector Parameter Value!(execute, self:Material Instance Dynamic Object Reference=@n2, ParameterName:Name=AnimColor1) -> (then, ReturnValue:Linear Color Structure) {is=title K2Node_CallFunction}
   > n5 = Get Vector Parameter Value(self=@n2, ParameterName=Color1)
   > SetVectorParameterValue(self=@n2, ParameterName=Color1, Value="Lerp (LinearColor)"(A=@n5, B=@n4, Alpha=@n3))
   n6 = Get MID_Background()
   n7 = Get Vector Parameter Value(execute=@n1.then_1, self=@n6, ParameterName=AnimColor2)
   > n8 = Get Vector Parameter Value(self=@n6, ParameterName=Color2)
   > SetVectorParameterValue(self=@n6, ParameterName=Color2, Value="Lerp (LinearColor)"(A=@n8, B=@n7, Alpha=@n3))
  ]}
 Navigation_DoCustom {nodes [
   n1 = Navigation_DoCustom!() -> (then, Navigation:EUINavigation Enum) {is=title K2Node_FunctionEntry}
   > n2 = Switch on EUINavigation!(execute, Selection:EUINavigation Enum=@n1) -> (Left:Exec, Right:Exec, Up:Exec, Down:Exec, Next:Exec, Previous:Exec, Invalid:Exec) {is=title K2Node_SwitchEnum}
   n3 = ShowCalendar(execute=@n2.Down)
   > Return Node!(execute, ReturnValue:Widget Object Reference=@n3)
   n4 = HideCalendar(execute=@n2.Up, B Instant Animations=false) {is=title K2Node_CallFunction}
   > Return Node(ReturnValue=@n4)
  ]}
 Navigation_DoCustom_Screen {nodes [
   n1 = Navigation_DoCustom_Screen!() -> (then, Navigation:EUINavigation Enum) {is=title K2Node_FunctionEntry}
   > n2 = Switch on EUINavigation(Selection=@n1)
   n3 = Get FortPoblanoTournamentsCalendarVM()
   n4 = Get IsLandingPageOpened()
   n5 = Get IsCalendarOpened()
   n6 = Branch(
    execute=@n2.Up
    Condition=AND Boolean!(A=NOT Boolean(A=@n4), B=NOT Boolean(A=@n5), C:bool=IsCompeteLandingPageEnabled(self=@n3)) -> (ReturnValue)
   )
   > n7 = SwipeLandingPage(IsShow=true)
   > Return Node(ReturnValue=@n7)
   n8 = HideCalendar(execute=@n6.else, B Instant Animations=false)
   > Return Node(ReturnValue=@n8)
   n9 = Get FortPoblanoTournamentsCalendarVM()
   n10 = Branch(execute=@n2.Down, Condition=AND Boolean!(A=Get IsLandingPageOpened(), B=IsCompeteLandingPageEnabled(self=@n9)) -> (ReturnValue))
   > n11 = SwipeLandingPage(IsShow=false)
   > Return Node(ReturnValue=@n11)
   n12 = ShowCalendar(execute=@n10.else)
   > Return Node(ReturnValue=@n12)
  ]}
 Navigation_FocusToCalendarTile {nodes [
   Navigation_FocusToCalendarTile!() -> (then, Navigation:EUINavigation Enum) {is=title K2Node_FunctionEntry}
   > Return Node(ReturnValue=Get Selected Calendar Tile())
  ]}
 On Select Calendar Tile {nodes [
   n1 = On Select Calendar Tile!() -> (then, Button:Common Button Base Object Reference, B Is Selected:bool=~) {is=title K2Node_FunctionEntry}
   > n2 = Sequence!(execute) -> (then_0, then_1, then_2:Exec)
   n3 = Get IsCalendarOpened()
   > Branch(Condition=AND Boolean(A=NOT Boolean(A=@n3), B=HasAnyUserFocusedDescendants!(self:Widget Object Reference=@n1) -> (ReturnValue:bool)))
   n4 = SlotAsGridSlot(Widget=@n1)
   n5 = SlotAsGridSlot(Widget=Get Selected Calendar Tile())
   > Branch(Condition="integer > integer"(A=Get Row!(self:Grid Slot Object Reference=@n4) -> (Row:int=0), B=Get Row(self=@n5) -> (Row=0)))
   > ShowCalendar()
   Branch(execute=@n2.then_1, Condition=@n1.B Is Selected)
   > SetNewSelectedTile(Selected Calendar Tile=@n1)
   n6 = Switch on ECommonInputType!(execute=@n2.then_2, Selection:ECommonInputType Enum=Get Input Type()) -> (MouseAndKeyboard:Exec, Gamepad:Exec, Touch:Exec, Count:Exec) {is=title K2Node_SwitchEnum}
   Branch(execute=[@n6.Count, @n6.Gamepad, @n6.MouseAndKeyboard], Condition=Get IsCalendarOpened())
   n7 = Get CalendarDayTilesArray()
   n8 = Get Row(self=SlotAsGridSlot(Widget=Get Selected Calendar Tile())) -> (Row=0)
   > ScrollWidgetIntoView!(
    execute
    self:Scroll Box Object Reference=Get ScrollBox_Month()
    WidgetToFind:Widget Object Reference=Select(
     Option 0:Widget Object Reference="Get (a copy)"(Array:Array of WBP Calendar Tile Object References=@n7) -> (Output:WBP Calendar Tile Object Reference)
     Option 1:Widget Object Reference="Get (a copy)"(Array=@n7, Dimension 1=7)
     Index="Equal (Integer)"(A=@n8, B=5)
    ) -> (ReturnValue:Widget Object Reference)
    AnimateScroll:bool=true
    ScrollDestination:EDescendantScrollDestination Enum=TopOrLeft
    Padding:float=0
   ) -> (then)
  ]}
 OnCalendarTileClicked {nodes [
   n1 = OnCalendarTileClicked!() -> (then, Button:Common Button Base Object Reference) {is=title K2Node_FunctionEntry}
   > n2 = Cast To WBP_CalendarTile(Object=@n1)
   > n3 = Branch(Condition=Get FakeLockedState!(self:WBP Calendar Tile Object Reference=@n2) -> (FakeLockedState:bool))
   n4 = Get FortPoblanoTournamentsCalendarDayTileVM!(self:WBP Calendar Tile Object Reference=@n2) -> (FortPoblanoTournamentsCalendarDayTileVM:Calendar Monthly Day Tile VM Object Reference)
   SetSelectedIndex(
    execute=@n3.else
    self=Get ListView(self=Get WBP_TournamentsCarousel())
    Index=int + int(A=GetFirstEventOnThisDate!(self:Calendar Monthly Day Tile VM Object Reference=@n4) -> (ReturnValue:int=0))
   )
   n5 = Get Anim_OpenCalendar()
   > Branch(Condition=NOT Boolean(A=IsAnimationPlaying(InAnimation=@n5)))
   > HideCalendar(B Instant Animations=false)
   > RefreshFocusIfEnabled()
  ]}
 OnMonthJustGenerated {nodes [OnMonthJustGenerated() {is=title K2Node_FunctionEntry}]}
 OnMouseWheel {nodes [
   n1 = OnMouseWheel!() -> (then, MyGeometry:Geometry Structure, MouseEvent:"Pointer Event Structure (by ref)") {is=title K2Node_FunctionEntry}
   > n2 = Branch(Condition="float > float"(A=Get Wheel Delta!(Input:"Pointer Event Structure (by ref)"=@n1.MouseEvent) -> (ReturnValue:float) {is=title K2Node_CallFunction}))
   n3 = Get ScrollBox_Month()
   n4 = GetScrollOffset(self:Scroll Box Object Reference=@n3)
   > n5 = Branch(Condition=AND Boolean(A=Get IsCalendarOpened(), B="float > float"(A=@n4)))
   > n6 = ScrollToStart(self:Scroll Box Object Reference=@n3)
   Return Node(execute=[@n16.then, @n13.then, @n15.then, @n9.then, @n6.then], ReturnValue:Event Reply Structure=Handled())
   n7 = Branch(execute=@n5.else, Condition=Get IsCalendarOpened())
   > n8 = HideCalendar(B Instant Animations=false)
   n9 = RefreshFocusIfEnabled(execute=[@n8.then, @n10.then])
   n10 = SwipeLandingPage(execute=@n7.else, IsShow=true)
   n11 = Get FortPoblanoTournamentsCalendarVM()
   n12 = Branch(execute=@n2.else, Condition=AND Boolean(A=Get IsLandingPageOpened(), B=IsCompeteLandingPageEnabled(self=@n11)))
   > n13 = SwipeLandingPage(IsShow=false)
   n14 = Branch(execute=@n12.else, Condition=Get IsCalendarOpened())
   > n15 = ScrollToEnd(self:Scroll Box Object Reference=Get ScrollBox_Month())
   ShowCalendar(execute=@n14.else)
   > n16 = RefreshFocusIfEnabled()
  ]}
 PlayBackgroundTransition {nodes [
   n1 = PlayBackgroundTransition!() -> (then, NextColor1:Linear Color Structure, NextColor2:Linear Color Structure) {is=title K2Node_FunctionEntry}
   > n2 = Branch(Condition=IsAnimationPlaying(InAnimation=Get Anim_BackgroundColor() -> (Anim_BackgroundColor:Widget Animation Object Reference)))
   > LerpBackgroundColor() {is=title K2Node_CallFunction}
   > n3 = StopAnimation(InAnimation=Get Anim_BackgroundColor())
   n4 = Get MID_Background()
   SetScalarParameterValue(execute=[@n7.then, @n3.then], self=@n4, ParameterName=AnimColorTransition, Value=0)
   > n5 = AdjustColorLuminance!(execute, self, In Color:Linear Color Structure=@n1) -> (then, Out Color:Linear Color Structure) {is=title K2Node_CallFunction}
   > Set NextColor1(NextColor1:Linear Color Structure=@n5) -> (Output_Get:Linear Color Structure)
   > n6 = AdjustColorLuminance(In Color=@n1.NextColor2)
   > Set NextColor2(NextColor2:Linear Color Structure=@n6) -> (Output_Get:Linear Color Structure)
   > SetVectorParameterValue(self=@n4, ParameterName=AnimColor1, Value=Get NextColor1() -> (NextColor1:Linear Color Structure))
   > SetVectorParameterValue(self=@n4, ParameterName=AnimColor2, Value=Get NextColor2() -> (NextColor2:Linear Color Structure))
   > PlayAnimation(InAnimation=Get Anim_BackgroundColor(), PlaybackSpeed=4)
   SetVectorParameterValue(execute=@n2.else, self=@n4, ParameterName=Color1, Value=Get NextColor1())
   > n7 = SetVectorParameterValue(self=@n4, ParameterName=Color2, Value=Get NextColor2())
  ]}
 RefocusCalendarIfDirtyFilter {nodes [
   RefocusCalendarIfDirtyFilter() {is=title K2Node_FunctionEntry}
   > SetIsEnabled(bInIsEnabled=true)
   > n1 = Sequence!(execute) -> (then_0, then_1)
   n2 = Get IsLoadingTournaments() -> (IsLoadingTournaments:bool)
   > Branch(Condition=AND Boolean(A=NOT Boolean(A=@n2), B=Get IsCalendarOpened()))
   > n3 = Cast To WBP_CalendarTile(Object=Get Selected Calendar Tile())
   > Set PreviouslySelectedCalendarTileIndex(PreviouslySelectedCalendarTileIndex:int=Find Item!(
     TargetArray:Array of WBP Calendar Tile Object References=Get CalendarDayTilesArray()
     ItemToFind:"WBP Calendar Tile Object Reference (by ref)"=@n3
    ) -> (ReturnValue)) -> (Output_Get:int=0)
   Set PreviouslySelectedCalendarTileIndex(execute=@n3.CastFailed, PreviouslySelectedCalendarTileIndex=0) -> (Output_Get=0)
   Update Filter Info(execute=@n1.then_1, self=Get FilterButton())
   > ProcessFilter(self:Tournaments Calendar VM Object Reference=Get FortPoblanoTournamentsCalendarVM())
   > n4 = Branch(Condition=Get IsLoadingTournaments())
   > Bind Event to On Month Just Generated(self:WBP Calendar Object Reference, Delegate=Create Event()) {is=title K2Node_AddDelegate}
   RestoreCalendarFocusAfterProcessingFilter(execute=@n4.else)
  ]}
 RefreshButtons {nodes [
   RefreshButtons() {is=title K2Node_FunctionEntry}
   > UpdateBookmarkButtonEnabled()
   > UpdateEligibilityButtonEnabled()
   > SetTodayButton()
   > UpdateBackButtonEnabled()
   > SetVisibility(
    self=[
     Get ArrowButton_Top() -> (ArrowButton_Top:WBP UIKit View More Arrow Button Object Reference)
     Get ArrowText_Top() -> (ArrowText_Top:Common Text Object Reference)
    ]
    InVisibility=Select(Option 0:ESlateVisibility Enum=SelfHitTestInvisible, Option 1:ESlateVisibility Enum=Collapsed, Index=Get IsLandingPageOpened()) -> (ReturnValue:ESlateVisibility Enum)
   )
   > SetBottomButtonVisibility!(
    execute
    self:WBP Bottom Bar Decoy Button Object Reference=Get DecoyButton_PowerRankingEligibility() -> (DecoyButton_PowerRankingEligibility:WBP Bottom Bar Decoy Button Object Reference)
    ShouldButtonBeVisible:bool=false
   ) -> (then)
   > SetVisibility(
    self=[
     Get ArrowText_Bottom() -> (ArrowText_Bottom:Common Text Object Reference)
     Get ArrowButton_Bottom() -> (ArrowButton_Bottom:WBP UIKit View More Arrow Button Object Reference)
    ]
    InVisibility=Select(Option 0=SelfHitTestInvisible, Option 1=Collapsed, Index=Get IsCalendarOpened())
   )
   n1 = MakeLiteralText(Value="LOCTEXT(\"Tournaments\")")
   > "SetText (Text)"(
    self=Get ArrowText_Top()
    InText=Select(Option 0:Text=MakeLiteralText(Value="LOCTEXT(\"Landing Page\")"), Option 1:Text=@n1, Index=Get IsCalendarOpened()) -> (ReturnValue:Text)
   )
   > "SetText (Text)"(
    self=Get ArrowText_Bottom()
    InText=Select(Option 0=MakeLiteralText(Value="LOCTEXT(\"Calendar\")"), Option 1=@n1, Index=Get IsLandingPageOpened())
   )
   AND Boolean(A=Get IsLandingPageOpened(), B=IsCompeteLandingPageEnabled(self=Get FortPoblanoTournamentsCalendarVM()))
  ]}
 RefreshFocusIfEnabled {nodes [
   RefreshFocusIfEnabled() {is=title K2Node_FunctionEntry}
   > Branch(Condition=Get bIsEnabled() -> (bIsEnabled:bool))
   > RequestRefreshFocus()
  ]}
 RestoreCalendarFocusAfterProcessingFilter {nodes [
   RestoreCalendarFocusAfterProcessingFilter() {is=title K2Node_FunctionEntry}
   > n1 = Branch(Condition=Get IsCalendarOpened())
   > n2 = SetNewSelectedTile(Selected Calendar Tile="Get (a copy)"(Array=Get CalendarDayTilesArray(), Dimension 1=Get PreviouslySelectedCalendarTileIndex() -> (PreviouslySelectedCalendarTileIndex=0)))
   RefreshFocusIfEnabled(execute=[@n2.then, @n1.else]) {is=title K2Node_CallFunction}
  ]}
 Scroll to Selected Tournament In Carousel {nodes [
   Scroll to Selected Tournament In Carousel() {is=title K2Node_FunctionEntry}
   > n1 = Branch(Condition=Get IsCalendarOpened())
   n2 = Get Selected Tournament Tile() -> (Selected Tournament Tile=0)
   Branch(execute=@n1.else, Condition="integer >= integer"(A=@n2))
   > SetSelectedIndex(self=Get ListView(self=Get WBP_TournamentsCarousel()), Index=@n2)
  ]}
 Scroll to Tournament Tile {nodes [
   n1 = Scroll to Tournament Tile!() -> (then, In Tournament Tile:Calendar Tournament Tile VM Object Reference) {is=title K2Node_FunctionEntry}
   > n2 = Branch(Condition=Get IsTournamentCarouselSwiped() -> (IsTournamentCarouselSwiped:bool))
   n3 = Get ListView(self=Get WBP_TournamentsCarousel())
   n4 = GetIndexForItem!(self:List View Object Reference=@n3, Item:Object Reference=@n1) -> (ReturnValue:int=0)
   Set Selected Tournament Tile(execute=[@n6.then, @n2.then], Selected Tournament Tile:int=@n4) -> (Output_Get:int=0)
   n5 = Get!(self:Calendar Tournament Tile VM Object Reference=@n1) -> (BackgroundCalendarColors_Main:Linear Color Structure, BackgroundCalendarColors_Sub:Linear Color Structure)
   > PlayBackgroundTransition!(execute, self, NextColor1:Linear Color Structure=@n5, NextColor2:Linear Color Structure=@n5.BackgroundCalendarColors_Sub) -> (then) {is=title K2Node_CallFunction}
   > ScrollToCalendarTile!(execute, self, New Date:"Date Time Structure (by ref)"=Get Date(self=@n1)) -> (then)
   > Set IsTournamentCarouselSwiped(IsTournamentCarouselSwiped:bool=false) -> (Output_Get:bool)
   n6 = ScrollIndexIntoView!(execute=@n2.else, self:List View Object Reference=@n3, Index:int=@n4) -> (then)
  ]}
 ScrollToCalendarTile {nodes [
   n1 = ScrollToCalendarTile!() -> (then, New Date:"Date Time Structure (by ref)") {is=title K2Node_FunctionEntry}
   > n2 = Branch(Condition=Is Displaying Month(In Date=@n1) {is=title K2Node_CallFunction})
   n3 = Get Day(A=@n1) -> (ReturnValue=0)
   > n4 = Set New Day Tile(New Day Tile:WBP Calendar Tile Object Reference="Get (a copy)"(Array=Get CalendarDayTilesArray(), Dimension 1=int - int(A=@n3))) -> (Output_Get:WBP Calendar Tile Object Reference)
   > Branch(Condition=IsValid(Object=@n4))
   > ScrollWidgetIntoView(self=Get ScrollBox_Month(), WidgetToFind=@n4, AnimateScroll=true, ScrollDestination=TopOrLeft, Padding=0)
   > SetIsSelected!(execute, self:Common Button Base Object Reference=@n4, InSelected:bool=true, bGiveClickFeedback:bool=false) -> (then)
   SetDisplayMonthDate!(
    execute=@n2.else
    self:Tournaments Calendar VM Object Reference=Get FortPoblanoTournamentsCalendarVM()
    InDisplayMonthDate:"Date Time Structure (by ref)"=@n1
   ) -> (then)
  ]}
 SelectCurrentDateTournament {nodes [
   SelectCurrentDateTournament() {is=title K2Node_FunctionEntry}
   n1 = Break Tournaments Tiles Data!(TournamentsTilesData=Get TournamentsTilesData(self=Get FortPoblanoTournamentsCalendarVM())) -> (TournamentTiles, PreviouslySelectedEvent, FirstUpcomingEvent, FirstLiveEvent, FirstCurrentDateEvent, ClosestEvent)
   > n2 = Branch(Condition=Is Empty(TargetArray=@n1))
   n3 = Branch(execute=@n2.else, Condition="integer >= integer"(A=@n1.FirstCurrentDateEvent))
   > n4 = Set New Selected Tile Index(New Selected Tile Index=@n1.FirstCurrentDateEvent) -> (Output_Get=0)
   RefreshFocusIfEnabled(execute=[@n4.then, @n7.then, @n6.then])
   > SetSelectedIndex(self=Get ListView(self=Get WBP_TournamentsCarousel()), Index=int + int(A=Get New Selected Tile Index() -> (New Selected Tile Index=0)))
   > HideCalendar(B Instant Animations=false)
   n5 = Branch(execute=@n3.else, Condition="integer >= integer"(A=@n1.FirstUpcomingEvent))
   > n6 = Set New Selected Tile Index(New Selected Tile Index=@n1.FirstUpcomingEvent) -> (Output_Get=0)
   n7 = Set New Selected Tile Index(execute=@n5.else, New Selected Tile Index=int - int(A=Length(TargetArray=@n1))) -> (Output_Get=0)
  ]}
 Sequencer Events {nodes [
   CheckIsScrollable() {is=title K2Node_CustomEvent}
   n1 = Get ScrollBox_Month()
   > ForceLayoutPrepass(self:Widget Object Reference=@n1)
   > DelayUntilNextTick!(execute) -> (then)
   n2 = GetScrollOffsetOfEnd!(self:Scroll Box Object Reference=@n1) -> (ReturnValue:float)
   > Branch(Condition=AND Boolean(A=Get IsCalendarOpened(), B="Equal (Float)"(A=@n2)))
   > Set bEnableTouchScrolling(bEnableTouchScrolling=false, self=Get ScrollBox_Month())
  ]}
 Set Month Text {nodes [
   Set Month Text() {is=title K2Node_FunctionEntry}
   > n1 = GetMonthText!(execute, DateTime:"Date Time Structure (by ref)"=Get DisplayMonthDate(self=Get FortPoblanoTournamentsCalendarVM())) -> (then, ReturnValue:Text)
   > SetText(self:WBP Calendar Wrapper Object Reference=Get CalendarWrapper(), InText:Text=@n1)
  ]}
 SetMobileMonthHeight {nodes [
   n1 = SetMobileMonthHeight!() -> (then, IsCalendarShown:bool=~) {is=title K2Node_FunctionEntry}
   n2 = IsTabletAspectRatio!() -> (ReturnValue:bool)
   > Branch(Condition=AND Boolean(A=IsMobileGame(), B=NOT Boolean(A=@n2)))
   > n3 = Branch(Condition=@n1)
   n4 = Get SizeBox_Month() -> (SizeBox_Month:Size Box Object Reference)
   > SetHeightOverride(self=@n4, InHeightOverride=700)
   ClearHeightOverride(execute=@n3.else, self:Size Box Object Reference=@n4)
  ]}
 SetMobileStyle {nodes [
   SetMobileStyle() {is=title K2Node_FunctionEntry}
   > Branch(Condition=IsMobileGame())
   n1 = SlotAsGridSlot(Widget=Get SafeZone() -> (SafeZone:Safe Zone Object Reference))
   > SetPadding(self=@n1, InPadding_Left=48, InPadding_Top=SelectFloat(A=292, B=140, bPickA=IsTabletAspectRatio()), InPadding_Right=48, InPadding_Bottom=0)
   > SetHorizontalAlignment(self:Grid Slot Object Reference=@n1, InHorizontalAlignment=HAlign_Fill)
   > SetVerticalAlignment(
    self:Grid Slot Object Reference=SlotAsGridSlot(Widget=Get IconHintsPanel() -> (IconHintsPanel:WBP Icon Hints Panel Object Reference))
    InVerticalAlignment=VAlign_Bottom
   )
   > AddChild(
    self=Get VertBox_MobileCalendarContainer() -> (VertBox_MobileCalendarContainer:Vertical Box Object Reference)
    Content=Get UniformGridPanel() -> (UniformGridPanel:Uniform Grid Panel Object Reference)
   )
   n2 = Get SizeBox_Month()
   > AddChildToVerticalBox(self=Get VertBox_MobileCalendarContainer(), Content=@n2)
   > SetHorizontalAlignment(self:Vertical Box Slot Object Reference=SlotAsVerticalBoxSlot(Widget=@n2), InHorizontalAlignment=HAlign_Center)
   > SetPadding(
    self:Size Box Slot Object Reference=SlotAsSizeBoxSlot!(Widget:Widget Object Reference=Get ScrollBox_Month()) -> (ReturnValue:Size Box Slot Object Reference)
    InPadding_Left=18
    InPadding_Top=8
    InPadding_Right=0
    InPadding_Bottom=0
   )
   > SetPadding(
    self:Grid Slot Object Reference=SlotAsGridSlot(Widget=Get FilterButton())
    InPadding_Left=0
    InPadding_Top=0
    InPadding_Right=0
    InPadding_Bottom=0
   )
   n3 = SlotAsVerticalBoxSlot(Widget=Get UniformGridPanel())
   n4 = Get!(self:Vertical Box Slot Object Reference=@n3) -> (Padding_Left:float=~, Padding_Top:float=~, Padding_Right:float=~, Padding_Bottom:float=~)
   > SetPadding(
    self:Vertical Box Slot Object Reference=@n3
    InPadding_Left=32
    InPadding_Top=@n4.Padding_Top
    InPadding_Right=@n4.Padding_Right
    InPadding_Bottom=@n4.Padding_Bottom
   )
   > n5 = Branch(Condition=IsTabletAspectRatio())
   SetHeightOverride(execute=@n5.else, self=Get SizeBox() -> (SizeBox:Size Box Object Reference), InHeightOverride=440)
  ]}
 SetNewSelectedTile {nodes [
   n1 = SetNewSelectedTile!() -> (then, Selected Calendar Tile:Common Button Base Object Reference) {is=title K2Node_FunctionEntry}
   > n2 = Get!(execute:Exec) -> (then:Exec, else:Exec, Selected Calendar Tile:Common Button Base Object Reference)
   > n3 = ClearSelection(self:Common Button Base Object Reference=@n2)
   Set Selected Calendar Tile(execute=[@n3.then, @n2.else], Selected Calendar Tile:Common Button Base Object Reference=@n1) -> (Output_Get:Common Button Base Object Reference)
  ]}
 SetTilesState {nodes [
   SetTilesState() {is=title K2Node_FunctionEntry}
   n1 = Break Tournaments Tiles Data!(TournamentsTilesData=Get TournamentsTilesData(self=Get FortPoblanoTournamentsCalendarVM())) -> (TournamentTiles, FirstCurrentDateEvent, ClosestEvent, bIsLoadingTournaments:bool=~)
   n2 = Select(Option 0:int=Select(Option 0=0, Option 1=2, Index=Is Empty(TargetArray=@n1)), Option 1:int=1, Index=@n1.bIsLoadingTournaments) -> (ReturnValue:int)
   > SetActiveWidgetIndex(
    self:Common Visibility Switcher Object Reference=Get TournamentsStateVisibilitySwitcher() -> (TournamentsStateVisibilitySwitcher:Common Visibility Switcher Object Reference)
    Index=@n2
   )
   > SetTodayButton()
   n3 = NOT Boolean(A=Get IsCalendarOpened())
   n4 = "Equal (Integer)"(A=@n2)
   > SetVisibility(
    self=Get NavigationButtons() -> (NavigationButtons:Grid Panel Object Reference)
    InVisibility=Select(Option 0:ESlateVisibility Enum=Collapsed, Option 1:ESlateVisibility Enum=SelfHitTestInvisible, Index=AND Boolean(A=@n3, B=@n4)) -> (ReturnValue:ESlateVisibility Enum)
   )
   > n5 = Branch(Condition=@n4)
   PlayBackgroundTransition(execute=@n5.else, NextColor1=(R=0,G=0.0116,B=0.063,A=1), NextColor2=(R=0.004,G=0.003,B=0.0152,A=1))
  ]}
 SetTodayButton {nodes [
   SetTodayButton() {is=title K2Node_FunctionEntry}
   n1 = Break Tournaments Tiles Data(TournamentsTilesData=Get TournamentsTilesData(self=Get FortPoblanoTournamentsCalendarVM())) -> (bIsLoadingTournaments=~)
   n2 = Get FortPoblanoTournamentsCalendarVM()
   n3 = AND Boolean!(
    A=NOT Boolean(A=Is Empty(TargetArray=@n1))
    B=NOT Boolean(A=@n1.bIsLoadingTournaments)
    C:bool=NAND Boolean(A=Get IsLandingPageOpened(), B=IsCompeteLandingPageEnabled(self=@n2))
   ) -> (ReturnValue)
   > n4 = Branch(Condition=@n3)
   > n5 = Branch(Condition="integer >= integer"(A=@n1.FirstCurrentDateEvent))
   > SetBottomButtonVisibility(self=Get DecoyButton_Today() -> (DecoyButton_Today:WBP Bottom Bar Decoy Button Object Reference), ShouldButtonBeVisible=@n3)
   n6 = NOT Boolean(A=@n3)
   > SetBottomButtonVisibility(self=Get DecoyButton_FirstUpcoming() -> (DecoyButton_FirstUpcoming:WBP Bottom Bar Decoy Button Object Reference), ShouldButtonBeVisible=@n6)
   SetBottomButtonVisibility(execute=@n5.else, self=Get DecoyButton_FirstUpcoming(), ShouldButtonBeVisible=@n3)
   > SetBottomButtonVisibility(self=Get DecoyButton_Today(), ShouldButtonBeVisible=@n6)
   SetBottomButtonVisibility(execute=@n4.else, self=Get DecoyButton_FirstUpcoming(), ShouldButtonBeVisible=false)
   > SetBottomButtonVisibility(self=Get DecoyButton_Today(), ShouldButtonBeVisible=false)
  ]}
 ShowCalendar {nodes [
   ShowCalendar() {is=title K2Node_FunctionEntry}
   > n1 = Branch(Condition=Get IsCalendarOpened())
   n2 = Get CalendarDayTilesArray()
   n3 = "Get (a copy)"(Array=@n2, Dimension 1=Last Index(TargetArray=@n2) -> (ReturnValue=0))
   n4 = Get Row(self=SlotAsGridSlot(Widget=@n3)) -> (Row=0)
   n5 = Get Selected Calendar Tile()
   n6 = Get Row(self=SlotAsGridSlot(Widget=@n5)) -> (Row=0)
   > Return Node(ReturnValue:Widget Object Reference=Select(Option 0:Widget Object Reference=@n5, Option 1:Widget Object Reference=@n3, Index="integer > integer"(A=@n4, B=@n6)) -> (ReturnValue:Widget Object Reference))
   PlayAnimation(execute=@n1.else, InAnimation=Get Anim_OpenCalendar(), PlaybackSpeed=5)
   > Set bEnableTouchScrolling(bEnableTouchScrolling=Get IsCalendarScrollableForTouch(), self=Get ScrollBox_Month())
   > UpdateIsCalendarOpened(IsOpened=true)
   > PlayButtonAnimation(self=Get CalendarWrapper(), IsShow=true)
   > PlayTransitionShowIcons(self=Get CalendarDayTilesArray(), Instant=false)
   > ScrollToStart(self=Get ScrollBox_Month())
   > SetMobileMonthHeight(IsCalendarShown=Get IsCalendarOpened()) {is=title K2Node_CallFunction}
   > Set ScrollWhenFocusChanges(
    ScrollWhenFocusChanges=Select!(
     MouseAndKeyboard:EScrollWhenFocusChanges Enum=AnimatedScroll
     Gamepad:EScrollWhenFocusChanges Enum=AnimatedScroll
     Touch:EScrollWhenFocusChanges Enum=NoScroll
     Count:EScrollWhenFocusChanges Enum=AnimatedScroll
     Index:ECommonInputType Enum=Get Input Type()
    ) -> (ReturnValue:EScrollWhenFocusChanges Enum)
    self=Get ScrollBox_Month()
   ) -> (Output_Get=NoScroll)
   > Return Node(ReturnValue=Get Selected Calendar Tile())
  ]}
 SwipeLandingPage {nodes [
   n1 = SwipeLandingPage!() -> (then, IsShow:bool=~) {is=title K2Node_FunctionEntry}
   > n2 = Branch(Condition=IsCompeteLandingPageEnabled(self=Get FortPoblanoTournamentsCalendarVM()))
   > SetIsEnabled(self=Get TournamentsContent(), bInIsEnabled=NOT Boolean(A=@n1))
   > n3 = Branch(Condition=@n1)
   > n4 = Branch(Condition=Get IsLandingPageOpened())
   Return Node(execute=[@n5.then, @n4.then], ReturnValue=GetDesiredFocusTarget(self=Get WBP_CompeteLandingPage()))
   Set IsLandingPageOpened(execute=@n4.else, IsLandingPageOpened:bool=true) -> (Output_Get:bool)
   > PlayAnimationReverse(InAnimation=Get HideLandingPage() -> (HideLandingPage:Widget Animation Object Reference), PlaybackSpeed=5)
   > RefreshButtons() {is=title K2Node_CallFunction}
   > n5 = SetVisibility(self=Get BottomGradient() -> (BottomGradient:Image Object Reference), InVisibility=SelfHitTestInvisible)
   n6 = Branch(execute=@n3.else, Condition=Get IsLandingPageOpened())
   > n7 = HideCalendar(B Instant Animations=true)
   > Set IsLandingPageOpened(IsLandingPageOpened=false)
   > PlayAnimationForward(InAnimation=Get HideLandingPage(), PlaybackSpeed=5)
   > SetVisibility(self=Get BottomGradient(), InVisibility=Collapsed)
   > RefreshButtons()
   > Return Node(ReturnValue=@n7)
   Return Node(execute=@n6.else, ReturnValue=GetDesiredFocusTarget())
   Return Node(execute=@n2.else, ReturnValue=GetDesiredFocusTarget())
  ]}
 UpdateBackButtonEnabled {nodes [
   UpdateBackButtonEnabled() {is=title K2Node_FunctionEntry}
   n1 = IsCompeteLandingPageEnabled(self=Get FortPoblanoTournamentsCalendarVM())
   n2 = Get IsLandingPageOpened()
   > SetBottomButtonVisibility(
    self=Get DecoyButton_Back() -> (DecoyButton_Back:WBP Bottom Bar Decoy Button Object Reference)
    ShouldButtonBeVisible=AND Boolean!(A=Get IsCalendarOpened(), B=NAND Boolean(A=@n2, B=@n1)) -> (ReturnValue)
   )
  ]}
 UpdateBookmarkButtonEnabled {nodes [
   UpdateBookmarkButtonEnabled() {is=title K2Node_FunctionEntry}
   n1 = Get CurrentSelectedTileVM() -> (CurrentSelectedTileVM:Calendar Tournament Tile VM Object Reference)
   > n2 = Is Valid(InputObject=@n1)
   SetBottomButtonVisibility(
    execute=@n2.Is Not Valid
    self=Get DecoyButton_Bookmark() -> (DecoyButton_Bookmark:WBP Bottom Bar Decoy Button Object Reference)
    ShouldButtonBeVisible=false
   )
   n3 = Get DecoyButton_Bookmark()
   SetTriggeringInputAction!(
    execute=@n2.Is Valid
    self:Common Button Base Object Reference=@n3
    InputActionRow:"Data Table Row Handle Structure (by ref)"=Select!(
     Option 0:"Data Table Row Handle Structure (by ref)"=Get AddBookmarkInputAction() -> (AddBookmarkInputAction:Data Table Row Handle Structure)
     Option 1:"Data Table Row Handle Structure (by ref)"=Get RemoveBookmarkInputAction() -> (RemoveBookmarkInputAction:Data Table Row Handle Structure)
     Index:bool=Get bIsBookmarked!(self:Calendar Tournament Tile VM Object Reference=@n1) -> (bIsBookmarked:bool)
    ) -> (ReturnValue:"Data Table Row Handle Structure (by ref)")
   ) -> (then)
   > CacheInputAction(self:WBP Bottom Bar Decoy Button Object Reference=@n3)
   n4 = Get IsCalendarOpened()
   n5 = HasTournamentEnded!(self:Calendar Tournament Tile VM Object Reference=@n1) -> (ReturnValue:bool)
   n6 = IsCompeteLandingPageEnabled(self=Get FortPoblanoTournamentsCalendarVM())
   n7 = Get IsLandingPageOpened()
   > SetBottomButtonVisibility(self=@n3, ShouldButtonBeVisible=AND Boolean!(A=NOT Boolean(A=@n5), B=NOT Boolean(A=@n4), C:bool=NAND Boolean(A=@n7, B=@n6)) -> (ReturnValue))
  ]}
 UpdateEligibilityButtonEnabled {nodes [
   UpdateEligibilityButtonEnabled() {is=title K2Node_FunctionEntry}
   > n1 = Is Valid(InputObject=Get CurrentSelectedTileVM())
   n2 = NOT Boolean(A=Get IsLoadingTournaments())
   n3 = GetEligibilityInputActionEnabled!(self:Fort Compete UIManager Object Reference=Get FortCompeteUIManager()) -> (ReturnValue:bool)
   n4 = NOT Boolean(A=Get IsCalendarOpened())
   n5 = NAND Boolean(A=Get IsLandingPageOpened(), B=IsCompeteLandingPageEnabled(self=Get FortPoblanoTournamentsCalendarVM()))
   n6 = Get Overlay_Eligibility() -> (Overlay_Eligibility:Overlay Object Reference)
   > SetVisibility(
    self=@n6
    InVisibility=Select(
     Option 0:ESlateVisibility Enum=Collapsed
     Option 1:ESlateVisibility Enum=SelfHitTestInvisible
     Index=AND Boolean!(A=@n3, B=@n2, C=@n4, D:bool=@n5) -> (ReturnValue)
    ) -> (ReturnValue:ESlateVisibility Enum)
   )
   SetVisibility(execute=@n1.Is Not Valid, self=@n6, InVisibility=Collapsed)
  ]}
 UpdateIsCalendarOpened {nodes [
   n1 = UpdateIsCalendarOpened!() -> (then, IsOpened:bool=~) {is=title K2Node_FunctionEntry}
   > Set IsCalendarOpened(IsCalendarOpened:bool=@n1) -> (Output_Get:bool)
   > RefreshButtons()
  ]}
 UpdateSelectedTileFromScrollOffset {nodes [
   n1 = UpdateSelectedTileFromScrollOffset!() -> (then, ScrollOffset:double=~) {is=title K2Node_FunctionEntry}
   n2 = Get ListView(self=Get WBP_TournamentsCarousel())
   > n3 = Branch(Condition="Equal (Integer)"(A=GetNumItems!(self:List View Object Reference=@n2) -> (ReturnValue:int=0)))
   n4 = "Clamp (Integer)"(Value=Round(A=@n1), Min=1, Max=GetTournamentsLength() -> (Length=0) {is=title K2Node_CallFunction}) -> (ReturnValue=0)
   Branch(execute=@n3.else, Condition="Not Equal (Integer)"(A=Get Selected Tournament Tile() -> (Selected Tournament Tile=0), B=@n4))
   > Set IsTournamentCarouselSwiped(IsTournamentCarouselSwiped=true)
   > SetSelectedIndex(self=Get ListView(self=Get WBP_TournamentsCarousel()), Index=@n4)
  ]}
]
variables [
 AddBookmarkInputAction {type=DataTableRowHandle, default {dataTable=/Compete/DataTables/DT_CompeteUIInputActions, rowName=Bookmark}}
 B Has Init Months {type=boolean}
 CalendarDayTilesArray {type="array<WBP_CalendarTile_C>", default []}
 CompactUILastItems {type="array<MVVM_CompeteEmptyTile_Entry_C>", default []}
 CurrentSelectedTileVM {type=FortPoblanoTournamentsCalendarTournamentTileVM, default=null}
 EligibilityWidget {type=WBP_EligibilityCheckModal_C, default=null}
 FirstItem {type=MVVM_CompeteEmptyTile_Entry_C, default=null}
 Input Type {type=ECommonInputType, default=MouseAndKeyboard}
 IsCalendarOpened {type=boolean}
 IsCalendarScrollableForTouch {type=boolean}
 IsLandingPageOpened {type=boolean, default=true}
 IsLoadingTournaments {type=boolean}
 IsTournamentCarouselSwiped {type=boolean}
 LastItem {type=MVVM_CompeteEmptyTile_Entry_C, default=null}
 Max Color Luminance {type=number, default=0.15}
 MID_Background {type=MaterialInstanceDynamic}
 NextColor1 {type=LinearColor, default=<rgba 0 0 0 0>}
 NextColor2 {type=LinearColor, default=<rgba 0 0 0 0>}
 OnMonthJustGenerated {}
 PreviouslySelectedCalendarTileIndex {type=integer}
 RemoveBookmarkInputAction {type=DataTableRowHandle, default {dataTable=&DT_CompeteUIInputActions, rowName=RemoveBookmark}}
 Selected Calendar Tile {type=CommonButtonBase, default=null}
 Selected Tournament Tile {type=integer, default=-1}
 TodayInputAction {type=DataTableRowHandle, default {dataTable=&DT_CompeteUIInputActions, rowName=ActionToday}}
 TournamentTilesArray {type="array<WBP_TournamentTile_C>", default []}
]
widgets {
 info {
  inheritedWidgetCount=0
  parentClass=&CommonActivatableWidget
  rootWidgetClass=/Script/FortniteUI.FortSwipePanel
  widgetCount=51
 }
 widgets [
  SwipePanel {widgetClassPath=&FortSwipePanel, bIsVariable=true, widget=SwipePanel, properties {swipeThreshold=<60 10>}}
  Root {
   widgetClassPath=&GridPanel
   parent=SwipePanel
   widget=Root
   properties {clipping=ClipToBounds, navigation=&self:WidgetTree.Root.WidgetNavigation_0, columnFill [1], rowFill [1]}
   slot {class=FortSwipePanelSlot}
  }
  Background {
   widgetClassPath=&Image
   bIsVariable=true
   parent=Root
   widget=Background
   properties {
    visibility=SelfHitTestInvisible
    brush {
     bIsDynamicallyLoaded=false
     drawAs=Image
     imageType=FullColor
     mirroring=NoMirror
     resourceName=None
     resourceObject=/Compete/MaterialInstances/MI_UI_Compete_CalendarBackground
     tiling=NoTile
     imageSize=<32 32>
     margin=<ltrb 0 0 0 0>
     outlineSettings {
      bUseBrushTransparency=false
      roundingType=HalfHeightRadius
      width=0
      color {colorUseRule=UseColor_Specified, specifiedColor=<rgba 0 0 0 0>}
      cornerRadii=<0 0 0 0>
     }
     tintColor {colorUseRule=UseColor_Specified, specifiedColor=<rgba 1 1 1 1>}
     uVRegion {bIsValid=false, max=<0 0>, min=<0 0>}
    }
   }
   slot {class=GridSlot}
  }
  CaptureForPostBufferUpdate {
   widgetClassPath=&WBP_CaptureForPostBufferUpdate
   bIsVariable=true
   parent=Root
   widget=CaptureForPostBufferUpdate
   slot {class=GridSlot, properties {layer=1}}
  }
  SafeZone {
   widgetClassPath=&SafeZone
   bIsVariable=true
   parent=Root
   widget=SafeZone
   slot {class=GridSlot, properties {horizontalAlignment=HAlign_Center, layer=10, padding=<ltrb 0 120 0 0>}}
  }
  GridPanel_0 {
   widgetClassPath=&GridPanel
   parent=SafeZone
   widget=GridPanel_0
   properties {columnFill [1], rowFill [0,1]}
   slot {class=SafeZoneSlot}
  }
  LandingSB {
   widgetClassPath=&SizeBox
   parent=GridPanel_0
   widget=LandingSB
   properties {
    bOverride_HeightOverride=true
    clipping=ClipToBounds
    heightOverride=671
    navigation=&self:WidgetTree.LandingSB.WidgetNavigation_0
   }
   slot {class=GridSlot}
  }
  WBP_CompeteLandingPage {
   widgetClassPath=/Compete/LandingPage/WBP_CompeteLandingPage
   bIsVariable=true
   parent=LandingSB
   widget=WBP_CompeteLandingPage
   slot {class=SizeBoxSlot}
  }
  ArrowTop {
   widgetClassPath=&Overlay
   bIsVariable=true
   parent=GridPanel_0
   widget=ArrowTop
   slot {class=GridSlot, properties {layer=11, verticalAlignment=VAlign_Top, nudge=<0 -50>, padding=<ltrb 0 0 0 -150>}}
  }
  ArrowButton_Top {
   widgetClassPath=/FortUILibrary/UIKit/Buttons/ArrowButtons/WBP_UIKit_ViewMore_ArrowButton
   bIsVariable=true
   parent=ArrowTop
   widget=ArrowButton_Top
   properties {
    holdData=&CommonUI_HoldData_Base_C
    renderTransform {angle=180, scale=<1 1>, shear=<0 0>, translation=<0 0>}
   }
   slot {class=OverlaySlot, properties {horizontalAlignment=HAlign_Center}}
  }
  ArrowText_Top {
   widgetClassPath=&CommonTextBlock
   bIsVariable=true
   parent=ArrowTop
   widget=ArrowText_Top
   properties {
    mobileFontSizeMultiplier=1.1
    text=Landing Page
    textTransformPolicy=ToUpper
    font {
     bForceMonospaced=false
     bMaterialIsStencil=false
     fontMaterial=null
     fontObject=&HeadingNow
     letterSpacing=0
     monospacedWidth=1
     size=15.75
     skewAmount=0
     typefaceFontName=75 Medium
     outlineSettings {
      bApplyOutlineToDropShadows=false
      bMiteredCorners=false
      bSeparateFillAlpha=false
      outlineMaterial=null
      outlineSize=0
      outlineColor=<rgba 0 0 0 1>
     }
    }
    shadowOffset=<0 0>
   }
   slot {class=OverlaySlot, properties {horizontalAlignment=HAlign_Center, verticalAlignment=VAlign_Bottom, padding=<ltrb 0 0 0 30>}}
  }
  TournamentsContent {
   widgetClassPath=&GridPanel
   bIsVariable=true
   parent=GridPanel_0
   widget=TournamentsContent
   properties {navigation=&self:WidgetTree.TournamentsContent.WidgetNavigation_0, columnFill [1], rowFill [0,0,0,1,0,0]}
   slot {class=GridSlot, properties {row=1}}
  }
  CalendarWrapper {
   widgetClassPath=&/WBP_CalendarWrapper
   bIsVariable=true
   parent=TournamentsContent
   widget=CalendarWrapper
   slot {class=GridSlot, properties {horizontalAlignment=HAlign_Left, layer=2, verticalAlignment=VAlign_Center}}
  }
  FilterButton {
   widgetClassPath=&/WBP_FilterButton
   bIsVariable=true
   parent=TournamentsContent
   widget=FilterButton
   slot {class=GridSlot, properties {horizontalAlignment=HAlign_Right, verticalAlignment=VAlign_Top, nudge=<0 -8>, padding=<ltrb 0 0 24 0>}}
  }
  NavigationButtons {
   widgetClassPath=&GridPanel
   bIsVariable=true
   parent=TournamentsContent
   widget=NavigationButtons
   properties {columnFill [1], rowFill [1]}
   slot {class=GridSlot, properties {layer=10, row=1, padding=<ltrb 0 46 0 25>}}
  }
  PreviousButton {
   widgetClassPath=&WBP_UIKit_Button_Generic
   bIsVariable=true
   parent=NavigationButtons
   widget=PreviousButton
   properties {
    bHideInputActionWithKeyboard=true
    bIsFocusable=false
    backgroundMaterial=/FortUILibrary/UIKit/Buttons/MaterialInstances/CTAButtons/MI_UIKit_BgRegular
    enableFocusedSelectedState=false
    icon=/FortUILibrary/Textures/Icons/ToBeDeprecated/T_UI_Icon_Global_Chevron_Left
    iconMaterial=/FortUILibrary/Textures/Material/MI_UIKit_Icon_Base
    minHeight=64
    minWidth=64
    outlineMaterial=/FortUILibrary/UIKit/Buttons/MaterialInstances/CTAButtons/MI_UIKit_ButtonOutline
    priority=100
    promptHorizontalAlignment=HAlign_Center
    useTextBlock=false
    buttonPadding=<ltrb 8 0 8 0>
    iconSize=<32 32>
    promptNudge=<0 0>
   }
   slot {class=GridSlot, properties {horizontalAlignment=HAlign_Left, layer=10, verticalAlignment=VAlign_Center, nudge=<-24 0>}}
  }
  NextButton {
   widgetClassPath=&WBP_UIKit_Button_Generic
   bIsVariable=true
   parent=NavigationButtons
   widget=NextButton
   properties {
    bHideInputActionWithKeyboard=true
    bIsFocusable=false
    bShouldUseFallbackDefaultInputAction=false
    backgroundMaterial=&MI_UIKit_BgRegular
    enableFocusedSelectedState=false
    icon=/FortUILibrary/Textures/Icons/T_UI_Icon_UIKit_Chevron
    iconMaterial=&MI_UIKit_Icon_Base
    minHeight=64
    minWidth=64
    outlineMaterial=&MI_UIKit_ButtonOutline
    priority=100
    promptHorizontalAlignment=HAlign_Center
    useTextBlock=false
    buttonPadding=<ltrb 8 0 8 0>
    iconSize=<32 32>
    promptNudge=<0 0>
   }
   slot {class=GridSlot, properties {horizontalAlignment=HAlign_Right, layer=10, verticalAlignment=VAlign_Center, nudge=<24 0>}}
  }
  SizeBox {
   widgetClassPath=&SizeBox
   bIsVariable=true
   parent=TournamentsContent
   widget=SizeBox
   properties {
    bOverride_HeightOverride=true
    bOverride_WidthOverride=true
    heightOverride=590
    navigation=&self:WidgetTree.SizeBox.WidgetNavigation_0
    widthOverride=1024
   }
   slot {class=GridSlot, properties {layer=2, row=1, padding=<ltrb -132 39 -132 25>}}
  }
  TournamentsStateVisibilitySwitcher {
   widgetClassPath=&CommonVisibilitySwitcher
   bIsVariable=true
   parent=SizeBox
   widget=TournamentsStateVisibilitySwitcher
   slot {class=SizeBoxSlot}
  }
  WBP_TournamentsCarousel {
   widgetClassPath=&/WBP_TournamentsCarousel
   bIsVariable=true
   parent=TournamentsStateVisibilitySwitcher
   widget=WBP_TournamentsCarousel
   slot {class=CommonVisibilitySwitcherSlot}
  }
  Throbber {
   widgetClassPath=&Image
   parent=TournamentsStateVisibilitySwitcher
   widget=Throbber
   properties {brush {
     bIsDynamicallyLoaded=false
     drawAs=Image
     imageType=FullColor
     mirroring=NoMirror
     resourceName=None
     resourceObject=/FortUILibrary/UIKit/Throbber/MI_UIKit_Monochrome_Throbber
     tiling=NoTile
     imageSize=<64 64>
     margin=<ltrb 0 0 0 0>
     outlineSettings {
      bUseBrushTransparency=false
      roundingType=HalfHeightRadius
      width=0
      color {colorUseRule=UseColor_Specified, specifiedColor=<rgba 0 0 0 0>}
      cornerRadii=<0 0 0 0>
     }
     tintColor {colorUseRule=UseColor_Specified, specifiedColor=<rgba 1 1 1 1>}
     uVRegion {bIsValid=false, max=<0 0>, min=<0 0>}
    }}
   slot {class=CommonVisibilitySwitcherSlot, properties {horizontalAlignment=HAlign_Center, verticalAlignment=VAlign_Center}}
  }
  EmptyListText {
   widgetClassPath=&CommonTextBlock
   parent=TournamentsStateVisibilitySwitcher
   widget=EmptyListText
   properties {
    justification=Center
    text=No Tournaments Available
    font {
     bForceMonospaced=false
     bMaterialIsStencil=false
     fontMaterial=null
     fontObject=&HeadingNow
     letterSpacing=0
     monospacedWidth=1
     size=18
     skewAmount=0
     typefaceFontName=75 Medium
     outlineSettings {
      bApplyOutlineToDropShadows=false
      bMiteredCorners=false
      bSeparateFillAlpha=false
      outlineMaterial=null
      outlineSize=0
      outlineColor=<rgba 0 0 0 1>
     }
    }
    shadowOffset=<0 0>
   }
   slot {class=CommonVisibilitySwitcherSlot, properties {horizontalAlignment=HAlign_Center, verticalAlignment=VAlign_Center}}
  }
  UniformGridPanel {
   widgetClassPath=/Script/UMG.UniformGridPanel
   bIsVariable=true
   parent=TournamentsContent
   widget=UniformGridPanel
   slot {class=GridSlot, properties {layer=2, row=2, padding=<ltrb 0 2 0 0>}}
  }
  CommonTextBlock_58 {
   widgetClassPath=&CommonTextBlock
   parent=UniformGridPanel
   widget=CommonTextBlock_58
   properties {
    text=Mo
    textTransformPolicy=ToUpper
    font {
     bForceMonospaced=false
     bMaterialIsStencil=false
     fontMaterial=null
     fontObject=&HeadingNow
     letterSpacing=0
     monospacedWidth=1
     size=15.75
     skewAmount=0
     typefaceFontName=75 Medium
     outlineSettings {
      bApplyOutlineToDropShadows=false
      bMiteredCorners=false
      bSeparateFillAlpha=false
      outlineMaterial=null
      outlineSize=0
      outlineColor=<rgba 0 0 0 1>
     }
    }
    shadowOffset=<0 0>
   }
   slot {class=UniformGridSlot}
  }
  CommonTextBlock {
   widgetClassPath=&CommonTextBlock
   parent=UniformGridPanel
   widget=CommonTextBlock
   properties {
    text=Tu
    textTransformPolicy=ToUpper
    font {
     bForceMonospaced=false
     bMaterialIsStencil=false
     fontMaterial=null
     fontObject=&HeadingNow
     letterSpacing=0
     monospacedWidth=1
     size=15.75
     skewAmount=0
     typefaceFontName=75 Medium
     outlineSettings {
      bApplyOutlineToDropShadows=false
      bMiteredCorners=false
      bSeparateFillAlpha=false
      outlineMaterial=null
      outlineSize=0
      outlineColor=<rgba 0 0 0 1>
     }
    }
    shadowOffset=<0 0>
   }
   slot {class=UniformGridSlot, properties {column=1}}
  }
  CommonTextBlock_1 {
   widgetClassPath=&CommonTextBlock
   parent=UniformGridPanel
   widget=CommonTextBlock_1
   properties {
    text=We
    textTransformPolicy=ToUpper
    font {
     bForceMonospaced=false
     bMaterialIsStencil=false
     fontMaterial=null
     fontObject=&HeadingNow
     letterSpacing=0
     monospacedWidth=1
     size=15.75
     skewAmount=0
     typefaceFontName=75 Medium
     outlineSettings {
      bApplyOutlineToDropShadows=false
      bMiteredCorners=false
      bSeparateFillAlpha=false
      outlineMaterial=null
      outlineSize=0
      outlineColor=<rgba 0 0 0 1>
     }
    }
    shadowOffset=<0 0>
   }
   slot {class=UniformGridSlot, properties {column=2}}
  }
  CommonTextBlock_2 {
   widgetClassPath=&CommonTextBlock
   parent=UniformGridPanel
   widget=CommonTextBlock_2
   properties {
    text=Th
    textTransformPolicy=ToUpper
    font {
     bForceMonospaced=false
     bMaterialIsStencil=false
     fontMaterial=null
     fontObject=&HeadingNow
     letterSpacing=0
     monospacedWidth=1
     size=15.75
     skewAmount=0
     typefaceFontName=75 Medium
     outlineSettings {
      bApplyOutlineToDropShadows=false
      bMiteredCorners=false
      bSeparateFillAlpha=false
      outlineMaterial=null
      outlineSize=0
      outlineColor=<rgba 0 0 0 1>
     }
    }
    shadowOffset=<0 0>
   }
   slot {class=UniformGridSlot, properties {column=3}}
  }
  CommonTextBlock_3 {
   widgetClassPath=&CommonTextBlock
   parent=UniformGridPanel
   widget=CommonTextBlock_3
   properties {
    text=Fr
    textTransformPolicy=ToUpper
    font {
     bForceMonospaced=false
     bMaterialIsStencil=false
     fontMaterial=null
     fontObject=&HeadingNow
     letterSpacing=0
     monospacedWidth=1
     size=15.75
     skewAmount=0
     typefaceFontName=75 Medium
     outlineSettings {
      bApplyOutlineToDropShadows=false
      bMiteredCorners=false
      bSeparateFillAlpha=false
      outlineMaterial=null
      outlineSize=0
      outlineColor=<rgba 0 0 0 1>
     }
    }
    shadowOffset=<0 0>
   }
   slot {class=UniformGridSlot, properties {column=4}}
  }
  CommonTextBlock_4 {
   widgetClassPath=&CommonTextBlock
   parent=UniformGridPanel
   widget=CommonTextBlock_4
   properties {
    text=Sa
    textTransformPolicy=ToUpper
    font {
     bForceMonospaced=false
     bMaterialIsStencil=false
     fontMaterial=null
     fontObject=&HeadingNow
     letterSpacing=0
     monospacedWidth=1
     size=15.75
     skewAmount=0
     typefaceFontName=75 Medium
     outlineSettings {
      bApplyOutlineToDropShadows=false
      bMiteredCorners=false
      bSeparateFillAlpha=false
      outlineMaterial=null
      outlineSize=0
      outlineColor=<rgba 0 0 0 1>
     }
    }
    shadowOffset=<0 0>
   }
   slot {class=UniformGridSlot, properties {column=5}}
  }
  CommonTextBlock_5 {
   widgetClassPath=&CommonTextBlock
   parent=UniformGridPanel
   widget=CommonTextBlock_5
   properties {
    text=Su
    textTransformPolicy=ToUpper
    font {
     bForceMonospaced=false
     bMaterialIsStencil=false
     fontMaterial=null
     fontObject=&HeadingNow
     letterSpacing=0
     monospacedWidth=1
     size=15.75
     skewAmount=0
     typefaceFontName=75 Medium
     outlineSettings {
      bApplyOutlineToDropShadows=false
      bMiteredCorners=false
      bSeparateFillAlpha=false
      outlineMaterial=null
      outlineSize=0
      outlineColor=<rgba 0 0 0 1>
     }
    }
    shadowOffset=<0 0>
   }
   slot {class=UniformGridSlot, properties {column=6}}
  }
  SizeBox_Month {
   widgetClassPath=&SizeBox
   bIsVariable=true
   parent=TournamentsContent
   widget=SizeBox_Month
   properties {
    bOverride_MaxDesiredHeight=true
    clipping=ClipToBounds
    maxDesiredHeight=135
    navigation=&self:WidgetTree.SizeBox_Month.WidgetNavigation_0
   }
   slot {class=GridSlot, properties {row=3, verticalAlignment=VAlign_Top, padding=<ltrb -18 0 0 0>}}
  }
  ScrollBox_Month {
   widgetClassPath=&ScrollBox
   bIsVariable=true
   parent=SizeBox_Month
   widget=ScrollBox_Month
   properties {
    bAllowRightClickDragScrolling=false
    bConsumePointerInput=false
    bEnableTouchScrolling=false
    clipping=Inherit
    consumeMouseWheel=Never
    scrollBarVisibility=Collapsed
    scrollWhenFocusChanges=AnimatedScroll
    visibility=SelfHitTestInvisible
    wheelScrollMultiplier=0
    scrollbarPadding=<ltrb 0 0 0 0>
    scrollbarThickness=<0 0>
    widgetBarStyle {
     thickness=0
     draggedThumbImage {
      bIsDynamicallyLoaded=false
      drawAs=Box
      imageType=FullColor
      mirroring=NoMirror
      resourceName=../../../Engine/Content/Slate/Common/Scrollbar_Thumb.png
      resourceObject=null
      tiling=NoTile
      imageSize=<0 0>
      margin=<ltrb 0.25 0.25 0.25 0.25>
      outlineSettings {
       bUseBrushTransparency=false
       roundingType=HalfHeightRadius
       width=0
       color {colorUseRule=UseColor_Specified, specifiedColor=<rgba 0 0 0 0>}
       cornerRadii=<0 0 0 0>
      }
      tintColor {colorUseRule=UseColor_Specified, specifiedColor=<rgba 1 1 1 1>}
      uVRegion {bIsValid=false, max=<0 0>, min=<0 0>}
     }
     horizontalBackgroundImage {
      bIsDynamicallyLoaded=false
      drawAs=NoDrawType
      imageType=NoImage
      mirroring=NoMirror
      resourceName=None
      resourceObject=null
      tiling=NoTile
      imageSize=<0 0>
      margin=<ltrb 0 0 0 0>
      outlineSettings {
       bUseBrushTransparency=false
       roundingType=HalfHeightRadius
       width=0
       color {colorUseRule=UseColor_Specified, specifiedColor=<rgba 0 0 0 0>}
       cornerRadii=<0 0 0 0>
      }
      tintColor {colorUseRule=UseColor_Specified, specifiedColor=<rgba 1 1 1 1>}
      uVRegion {bIsValid=false, max=<0 0>, min=<0 0>}
     }
     horizontalBottomSlotImage {
      bIsDynamicallyLoaded=false
      drawAs=Image
      imageType=FullColor
      mirroring=NoMirror
      resourceName=../../../Engine/Content/Slate/Common/Scrollbar_Background_Horizontal.png
      resourceObject=null
      tiling=NoTile
      imageSize=<8 8>
      margin=<ltrb 0 0 0 0>
      outlineSettings {
       bUseBrushTransparency=false
       roundingType=HalfHeightRadius
       width=0
       color {colorUseRule=UseColor_Specified, specifiedColor=<rgba 0 0 0 0>}
       cornerRadii=<0 0 0 0>
      }
      tintColor {colorUseRule=UseColor_Specified, specifiedColor=<rgba 1 1 1 1>}
      uVRegion {bIsValid=false, max=<0 0>, min=<0 0>}
     }
     horizontalTopSlotImage {
      bIsDynamicallyLoaded=false
      drawAs=Image
      imageType=FullColor
      mirroring=NoMirror
      resourceName=../../../Engine/Content/Slate/Common/Scrollbar_Background_Horizontal.png
      resourceObject=null
      tiling=NoTile
      imageSize=<8 8>
      margin=<ltrb 0 0 0 0>
      outlineSettings {
       bUseBrushTransparency=false
       roundingType=HalfHeightRadius
       width=0
       color {colorUseRule=UseColor_Specified, specifiedColor=<rgba 0 0 0 0>}
       cornerRadii=<0 0 0 0>
      }
      tintColor {colorUseRule=UseColor_Specified, specifiedColor=<rgba 1 1 1 1>}
      uVRegion {bIsValid=false, max=<0 0>, min=<0 0>}
     }
     hoveredThumbImage {
      bIsDynamicallyLoaded=false
      drawAs=Box
      imageType=FullColor
      mirroring=NoMirror
      resourceName=../../../Engine/Content/Slate/Common/Scrollbar_Thumb.png
      resourceObject=null
      tiling=NoTile
      imageSize=<0 0>
      margin=<ltrb 0.25 0.25 0.25 0.25>
      outlineSettings {
       bUseBrushTransparency=false
       roundingType=HalfHeightRadius
       width=0
       color {colorUseRule=UseColor_Specified, specifiedColor=<rgba 0 0 0 0>}
       cornerRadii=<0 0 0 0>
      }
      tintColor {colorUseRule=UseColor_Specified, specifiedColor=<rgba 1 1 1 1>}
      uVRegion {bIsValid=false, max=<0 0>, min=<0 0>}
     }
     normalThumbImage {
      bIsDynamicallyLoaded=false
      drawAs=Box
      imageType=FullColor
      mirroring=NoMirror
      resourceName=../../../Engine/Content/Slate/Common/Scrollbar_Thumb.png
      resourceObject=null
      tiling=NoTile
      imageSize=<0 0>
      margin=<ltrb 0.25 0.25 0.25 0.25>
      outlineSettings {
       bUseBrushTransparency=false
       roundingType=HalfHeightRadius
       width=0
       color {colorUseRule=UseColor_Specified, specifiedColor=<rgba 0 0 0 0>}
       cornerRadii=<0 0 0 0>
      }
      tintColor {colorUseRule=UseColor_Specified, specifiedColor=<rgba 1 1 1 1>}
      uVRegion {bIsValid=false, max=<0 0>, min=<0 0>}
     }
     verticalBackgroundImage {
      bIsDynamicallyLoaded=false
      drawAs=NoDrawType
      imageType=NoImage
      mirroring=NoMirror
      resourceName=None
      resourceObject=null
      tiling=NoTile
      imageSize=<0 0>
      margin=<ltrb 0 0 0 0>
      outlineSettings {
       bUseBrushTransparency=false
       roundingType=HalfHeightRadius
       width=0
       color {colorUseRule=UseColor_Specified, specifiedColor=<rgba 0 0 0 0>}
       cornerRadii=<0 0 0 0>
      }
      tintColor {colorUseRule=UseColor_Specified, specifiedColor=<rgba 1 1 1 1>}
      uVRegion {bIsValid=false, max=<0 0>, min=<0 0>}
     }
     verticalBottomSlotImage {
      bIsDynamicallyLoaded=false
      drawAs=Image
      imageType=FullColor
      mirroring=NoMirror
      resourceName=../../../Engine/Content/Slate/Common/Scrollbar_Background_Vertical.png
      resourceObject=null
      tiling=NoTile
      imageSize=<8 8>
      margin=<ltrb 0 0 0 0>
      outlineSettings {
       bUseBrushTransparency=false
       roundingType=HalfHeightRadius
       width=0
       color {colorUseRule=UseColor_Specified, specifiedColor=<rgba 0 0 0 0>}
       cornerRadii=<0 0 0 0>
      }
      tintColor {colorUseRule=UseColor_Specified, specifiedColor=<rgba 1 1 1 1>}
      uVRegion {bIsValid=false, max=<0 0>, min=<0 0>}
     }
     verticalTopSlotImage {
      bIsDynamicallyLoaded=false
      drawAs=Image
      imageType=FullColor
      mirroring=NoMirror
      resourceName=../../../Engine/Content/Slate/Common/Scrollbar_Background_Vertical.png
      resourceObject=null
      tiling=NoTile
      imageSize=<8 8>
      margin=<ltrb 0 0 0 0>
      outlineSettings {
       bUseBrushTransparency=false
       roundingType=HalfHeightRadius
       width=0
       color {colorUseRule=UseColor_Specified, specifiedColor=<rgba 0 0 0 0>}
       cornerRadii=<0 0 0 0>
      }
      tintColor {colorUseRule=UseColor_Specified, specifiedColor=<rgba 1 1 1 1>}
      uVRegion {bIsValid=false, max=<0 0>, min=<0 0>}
     }
    }
    widgetStyle {
     barThickness=0
     bottomShadowBrush {
      bIsDynamicallyLoaded=false
      drawAs=NoDrawType
      imageType=FullColor
      mirroring=NoMirror
      resourceName=../../../Engine/Content/Slate/Common/ScrollBoxShadowBottom.png
      resourceObject=null
      tiling=NoTile
      imageSize=<16 8>
      margin=<ltrb 0.5 0 0.5 1>
      outlineSettings {
       bUseBrushTransparency=false
       roundingType=HalfHeightRadius
       width=0
       color {colorUseRule=UseColor_Specified, specifiedColor=<rgba 0 0 0 0>}
       cornerRadii=<0 0 0 0>
      }
      tintColor {colorUseRule=UseColor_Specified, specifiedColor=<rgba 1 1 1 1>}
      uVRegion {bIsValid=false, max=<0 0>, min=<0 0>}
     }
     horizontalScrolledContentPadding=<ltrb 0 0 1 0>
     leftShadowBrush {
      bIsDynamicallyLoaded=false
      drawAs=NoDrawType
      imageType=FullColor
      mirroring=NoMirror
      resourceName=../../../Engine/Content/Slate/Common/ScrollBoxShadowLeft.png
      resourceObject=null
      tiling=NoTile
      imageSize=<8 16>
      margin=<ltrb 1 0.5 0 0.5>
      outlineSettings {
       bUseBrushTransparency=false
       roundingType=HalfHeightRadius
       width=0
       color {colorUseRule=UseColor_Specified, specifiedColor=<rgba 0 0 0 0>}
       cornerRadii=<0 0 0 0>
      }
      tintColor {colorUseRule=UseColor_Specified, specifiedColor=<rgba 1 1 1 1>}
      uVRegion {bIsValid=false, max=<0 0>, min=<0 0>}
     }
     rightShadowBrush {
      bIsDynamicallyLoaded=false
      drawAs=NoDrawType
      imageType=FullColor
      mirroring=NoMirror
      resourceName=../../../Engine/Content/Slate/Common/ScrollBoxShadowRight.png
      resourceObject=null
      tiling=NoTile
      imageSize=<8 16>
      margin=<ltrb 0 0.5 1 0.5>
      outlineSettings {
       bUseBrushTransparency=false
       roundingType=HalfHeightRadius
       width=0
       color {colorUseRule=UseColor_Specified, specifiedColor=<rgba 0 0 0 0>}
       cornerRadii=<0 0 0 0>
      }
      tintColor {colorUseRule=UseColor_Specified, specifiedColor=<rgba 1 1 1 1>}
      uVRegion {bIsValid=false, max=<0 0>, min=<0 0>}
     }
     topShadowBrush {
      bIsDynamicallyLoaded=false
      drawAs=NoDrawType
      imageType=FullColor
      mirroring=NoMirror
      resourceName=../../../Engine/Content/Slate/Common/ScrollBoxShadowTop.png
      resourceObject=null
      tiling=NoTile
      imageSize=<16 8>
      margin=<ltrb 0.5 1 0.5 0>
      outlineSettings {
       bUseBrushTransparency=false
       roundingType=HalfHeightRadius
       width=0
       color {colorUseRule=UseColor_Specified, specifiedColor=<rgba 0 0 0 0>}
       cornerRadii=<0 0 0 0>
      }
      tintColor {colorUseRule=UseColor_Specified, specifiedColor=<rgba 1 1 1 1>}
      uVRegion {bIsValid=false, max=<0 0>, min=<0 0>}
     }
     verticalScrolledContentPadding=<ltrb 0 1 0 1>
    }
   }
   slot {class=SizeBoxSlot, properties {padding=<ltrb 18 10 0 10>}}
  }
  Days {
   widgetClassPath=/Script/FortniteUI.UIKitGridPanel
   bIsVariable=true
   parent=ScrollBox_Month
   widget=Days
   properties {navigation=&self:WidgetTree.Days.WidgetNavigation_1, columnFill [1,1,1,1,1,1,1]}
   slot {class=ScrollBoxSlot, properties {verticalAlignment=VAlign_Top}}
  }
  VertBox_MobileCalendarContainer {
   widgetClassPath=&VerticalBox
   bIsVariable=true
   parent=TournamentsContent
   widget=VertBox_MobileCalendarContainer
   slot {class=GridSlot, properties {horizontalAlignment=HAlign_Center, row=3}}
  }
  IconHintsPanel {
   widgetClassPath=/Compete/Calendar/IconHints/WBP_IconHintsPanel
   bIsVariable=true
   parent=TournamentsContent
   widget=IconHintsPanel
   properties {visibility=Collapsed}
   slot {class=GridSlot, properties {horizontalAlignment=HAlign_Left, row=4, verticalAlignment=VAlign_Bottom, padding=<ltrb 0 4 0 23>}}
  }
  DecoyButton_Back {
   widgetClassPath=/Game/UI/Frontend/BottomBar/WBP_BottomBarDecoyButton
   bIsVariable=true
   parent=Root
   widget=DecoyButton_Back
   properties {
    holdData=&CommonUI_HoldData_Base_C
    triggeringInputAction {dataTable=&InputActionDataTable, rowName=Back}
   }
   slot {class=GridSlot, properties {horizontalAlignment=HAlign_Left, verticalAlignment=VAlign_Top}}
  }
  DecoyButton_PowerRankingEligibility {
   widgetClassPath=&WBP_BottomBarDecoyButton
   bIsVariable=true
   parent=Root
   widget=DecoyButton_PowerRankingEligibility
   properties {
    holdData=&CommonUI_HoldData_Base_C
    triggeringInputAction {dataTable=&DT_CompeteUIInputActions, rowName=PowerRankingInfo}
   }
   slot {class=GridSlot, properties {horizontalAlignment=HAlign_Left, verticalAlignment=VAlign_Top}}
  }
  DecoyButton_Bookmark {
   widgetClassPath=&WBP_BottomBarDecoyButton
   bIsVariable=true
   parent=Root
   widget=DecoyButton_Bookmark
   properties {
    holdData=&CommonUI_HoldData_Base_C
    triggeringInputAction {dataTable=&DT_CompeteUIInputActions, rowName=Bookmark}
   }
   slot {class=GridSlot, properties {horizontalAlignment=HAlign_Left, verticalAlignment=VAlign_Top}}
  }
  DecoyButton_Today {
   widgetClassPath=&WBP_BottomBarDecoyButton
   bIsVariable=true
   parent=Root
   widget=DecoyButton_Today
   properties {
    holdData=&CommonUI_HoldData_Base_C
    triggeringInputAction {dataTable=&DT_CompeteUIInputActions, rowName=ActionToday}
   }
   slot {class=GridSlot, properties {horizontalAlignment=HAlign_Left, verticalAlignment=VAlign_Top}}
  }
  DecoyButton_FirstUpcoming {
   widgetClassPath=&WBP_BottomBarDecoyButton
   bIsVariable=true
   parent=Root
   widget=DecoyButton_FirstUpcoming
   properties {
    holdData=&CommonUI_HoldData_Base_C
    triggeringInputAction {dataTable=&DT_CompeteUIInputActions, rowName=ActionFirstUpcoming}
   }
   slot {class=GridSlot, properties {horizontalAlignment=HAlign_Left, verticalAlignment=VAlign_Top}}
  }
  ArrowBottom {
   widgetClassPath=&Overlay
   bIsVariable=true
   parent=Root
   widget=ArrowBottom
   slot {class=GridSlot, properties {layer=11, verticalAlignment=VAlign_Bottom, nudge=<0 15>}}
  }
  BottomGradient {
   widgetClassPath=&Image
   bIsVariable=true
   parent=ArrowBottom
   widget=BottomGradient
   properties {
    visibility=SelfHitTestInvisible
    brush {
     bIsDynamicallyLoaded=false
     drawAs=Image
     imageType=FullColor
     mirroring=NoMirror
     resourceName=None
     resourceObject=/Compete/LandingPage/Textures/M_UI_LandingPageGradient
     tiling=NoTile
     imageSize=<32 320>
     margin=<ltrb 0 0 0 0>
     outlineSettings {
      bUseBrushTransparency=false
      roundingType=HalfHeightRadius
      width=0
      color {colorUseRule=UseColor_Specified, specifiedColor=<rgba 0 0 0 0>}
      cornerRadii=<0 0 0 0>
     }
     tintColor {colorUseRule=UseColor_Specified, specifiedColor=<rgba 1 1 1 1>}
     uVRegion {bIsValid=false, max=<0 0>, min=<0 0>}
    }
   }
   slot {class=OverlaySlot, properties {horizontalAlignment=HAlign_Fill, verticalAlignment=VAlign_Fill}}
  }
  SafeZone_1 {
   widgetClassPath=&SafeZone
   bIsVariable=true
   parent=ArrowBottom
   widget=SafeZone_1
   properties {padLeft=false, padRight=false, padTop=false}
   slot {class=OverlaySlot, properties {horizontalAlignment=HAlign_Fill, verticalAlignment=VAlign_Bottom}}
  }
  Overlay_0 {widgetClassPath=&Overlay, parent=SafeZone_1, widget=Overlay_0, slot {class=SafeZoneSlot}}
  ArrowText_Bottom {
   widgetClassPath=&CommonTextBlock
   bIsVariable=true
   parent=Overlay_0
   widget=ArrowText_Bottom
   properties {
    mobileFontSizeMultiplier=1.1
    text=Tournaments
    textTransformPolicy=ToUpper
    font {
     bForceMonospaced=false
     bMaterialIsStencil=false
     fontMaterial=null
     fontObject=&HeadingNow
     letterSpacing=0
     monospacedWidth=1
     size=15.75
     skewAmount=0
     typefaceFontName=75 Medium
     outlineSettings {
      bApplyOutlineToDropShadows=false
      bMiteredCorners=false
      bSeparateFillAlpha=false
      outlineMaterial=null
      outlineSize=0
      outlineColor=<rgba 0 0 0 1>
     }
    }
    shadowOffset=<0 0>
   }
   slot {class=OverlaySlot, properties {horizontalAlignment=HAlign_Center, padding=<ltrb 0 30 0 0>}}
  }
  ArrowButton_Bottom {
   widgetClassPath=&WBP_UIKit_ViewMore_ArrowButton
   bIsVariable=true
   parent=Overlay_0
   widget=ArrowButton_Bottom
   properties {holdData=&CommonUI_HoldData_Base_C}
   slot {class=OverlaySlot, properties {horizontalAlignment=HAlign_Center, verticalAlignment=VAlign_Bottom}}
  }
  SafeZone_0 {
   widgetClassPath=&SafeZone
   bIsVariable=true
   parent=Root
   widget=SafeZone_0
   properties {padRight=false, padTop=false}
   slot {class=GridSlot, properties {horizontalAlignment=HAlign_Left, layer=99, verticalAlignment=VAlign_Bottom, padding=<ltrb 80 0 0 32>}}
  }
  Overlay_Eligibility {widgetClassPath=&Overlay, bIsVariable=true, parent=SafeZone_0, widget=Overlay_Eligibility, slot {class=SafeZoneSlot}}
  CommonHardwareVisibilityBorder_187 {
   widgetClassPath=&CommonHardwareVisibilityBorder
   parent=Overlay_Eligibility
   widget=CommonHardwareVisibilityBorder_187
   properties {visibilityQuery {
     autoDescription=" ALL( Input.Gamepad )"
     tokenStreamVersion=0
     userDescription=Gamepad Only
     queryTokenStream [0,1,2,1,0]
     tagDictionary [{tagName=Input.Gamepad}]
    }}
   slot {class=OverlaySlot, properties {horizontalAlignment=HAlign_Fill, verticalAlignment=VAlign_Fill}}
  }
  WBP_UIKit_Backplate {
   widgetClassPath=&WBP_UIKit_Backplate
   bIsVariable=true
   parent=CommonHardwareVisibilityBorder_187
   widget=WBP_UIKit_Backplate
   properties {brightness=Obscure, cornerRadius=Circle}
   slot {class=BorderSlot, properties {padding=<ltrb 0 0 0 0>}}
  }
  Button_Eligibility {
   widgetClassPath=/FortUILibrary/UIKit/Buttons/WBP_UIKit_Button_Quiet
   bIsVariable=true
   parent=Overlay_Eligibility
   widget=Button_Eligibility
   properties {text=Eligibility, triggeringInputAction {dataTable=&DT_CompeteUIInputActions, rowName=Eligibility}}
   slot {class=OverlaySlot}
  }
 ]
}
