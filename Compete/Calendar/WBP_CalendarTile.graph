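# Day tile button for the Compete tournament calendar, showing the date, tournament count, bookmark flag and six cup-type icons.
# Icons register as UIKit modular blocks, recolored per cup state through their MIDs; setters apply today styling and week-based dimming.
# Count text reads the day-tile view model, switching format when the calendar opens or closes as icons fade.
# Days without tournaments use a fake locked state that stays clickable, shaking after logging analytics.
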WidgetBlueprint /Compete/Calendar/WBP_CalendarTile
animations [
 Anim_ShowTournamentIcons {
  binding_count=1
  length_seconds=1
  object=&self:Anim_ShowTournamentIcons
  sequence {bindings [TournamentIcons {
     id=#1
     possessed_class=GridPanel
     tracks [{
       type=MovieSceneFloatTrack
       display_name=Render Opacity
       sections [{type=MovieSceneFloatSection, end_frame=20, start_frame=0, channels [None {type=MovieSceneScriptingFloatChannel, keys [[0,0],[20,1]]}]}]
      }]
    }]}
 }
 Anim_Locked {
  binding_count=1
  length_seconds=1
  object=&self:Anim_Locked
  sequence {bindings [Root {
     id=#2
     possessed_class=GridPanel
     tracks [{
       type=MovieScene2DTransformTrack
       sections [{
         type=MovieScene2DTransformSection
         end_frame=20
         start_frame=0
         channels [
          Translation.X {type=MovieSceneScriptingFloatChannel, keys [[0,0],[5,8],[15,-8],[20,0]]}
          Translation.Y {type=MovieSceneScriptingFloatChannel, const [0,1]}
          Angle {type=MovieSceneScriptingFloatChannel, const [0,1]}
          Scale.X {type=MovieSceneScriptingFloatChannel, const [1,1]}
          Scale.Y {type=MovieSceneScriptingFloatChannel, const [1,1]}
          Shear.X {type=MovieSceneScriptingFloatChannel, const [0,1]}
          Shear.Y {type=MovieSceneScriptingFloatChannel, const [0,1]}
         ]
        }]
      }]
    }]}
 }
 Anim_Selected {
  binding_count=1
  length_seconds=1
  object=&self:Anim_Selected
  sequence {bindings [SelectionOutline {
     id=#3
     possessed_class=CommonBorder
     tracks [{
       type=MovieSceneFloatTrack
       display_name=Render Opacity
       sections [{type=MovieSceneFloatSection, end_frame=20, start_frame=0, channels [None {type=MovieSceneScriptingFloatChannel, keys [[0,0],[20,1]]}]}]
      }]
    }]}
 }
]
graphs [
 DebugCalendarTileInstance {nodes [
   DebugCalendarTileInstance() {is=title K2Node_FunctionEntry}
   n1 = Get DebugTxt() -> (DebugTxt:Text Block Object Reference)
   > SetVisibility(self=@n1, InVisibility=SelfHitTestInvisible)
   n2 = GetDisplayName(Object=Self-Reference())
   > "SetText (Text)"(self=@n1, InText="To Text (String)"(InString=Replace(SourceString=@n2, From=WBP_Calendar)))
  ]}
 EventGraph {nodes [
   n1 = Event PreConstruct()
   > n2 = Sequence!(execute) -> (then_0, then_1, then_2:Exec)
   > Branch(Condition=@n1.IsDesignTime)
   > SetDate!(execute, self, Amount:int=5) -> (then)
   InitTile(execute=@n2.then_1)
   Event On Focused() {is=title K2Node_Event}
   > Parent: On Focused() {is=title K2Node_CallParentFunction}
   n3 = Get Block_Outline() -> (Block_Outline:WBP UIKit Block Outline Object Reference)
   > RequestTransition!(execute, self:UIKit Block Object Reference=@n3, State:EUIKitBlockVisualState Enum=Focused) -> (then)
   Event On Clicked()
   n4 = "GetText (Text)"!(self:Text Object Reference=Get Text_Date() -> (Text_Date:Common Text Object Reference)) -> (ReturnValue:Text)
   > Fire Event Element Interacted(
    Element=Self-Reference()
    AdditionalInfo="Add String (Non-Localized)"!(
     Info:"Flow Analytics Additional Info Structure (by ref)"=Make Flow Analytics Additional Info!() -> (FlowAnalyticsAdditionalInfo:Flow Analytics Additional Info Structure) {is=title K2Node_MakeStruct}
     Key:String=Date
     Value:String="To String (Text)"(InText=@n4)
    ) -> (ReturnValue:Flow Analytics Additional Info Structure)
   )
   > Branch(Condition=Get FakeLockedState() -> (FakeLockedState:bool))
   > PlayAnimation(InAnimation=Get Anim_Locked() -> (Anim_Locked:Widget Animation Object Reference), PlaybackSpeed=5)
   Event On Selected()
   > Parent: On Selected() {is=title K2Node_CallParentFunction}
   n5 = Get Anim_Selected() -> (Anim_Selected:Widget Animation Object Reference)
   > PlayAnimationForward(InAnimation=@n5, PlaybackSpeed=5)
   Event On Deselected()
   > Parent: On Deselected() {is=title K2Node_CallParentFunction}
   > PlayAnimationReverse(InAnimation=@n5, PlaybackSpeed=5)
   Event On Unfocused() {is=title K2Node_Event}
   > Parent: On Unfocused() {is=title K2Node_CallParentFunction}
   > RequestTransition(self=@n3, State=Unfocused)
  ]}
 InitTile {nodes [
   InitTile() {is=title K2Node_FunctionEntry}
   > n1 = Sequence!(execute) -> (then_0, then_1, then_2, then_3:Exec)
   > RegisterModularBlock!(execute, self:UIKit Modular Button Object Reference, Block:UIKit Block Object Reference=Get Block_Outline()) -> (then)
   InitTournamentBlockIcon!(
    execute=@n1.then_1
    self
    BlockImage:WBP Compete Block Image Object Reference=Get Block_CashIcon() -> (Block_CashIcon:WBP Compete Block Image Object Reference)
    Texture:"Texture 2D Object Reference (by ref)"=Get Cash_Texture() -> (Cash_Texture:Texture 2D Object Reference)
   ) -> (then)
   > InitTournamentBlockIcon(
    BlockImage=Get Block_ItemShopIcon() -> (Block_ItemShopIcon:WBP Compete Block Image Object Reference)
    Texture=Get ItemShop_Texture() -> (ItemShop_Texture:Texture 2D Object Reference)
   )
   > InitTournamentBlockIcon(
    BlockImage=Get Block_FNCSIcon() -> (Block_FNCSIcon:WBP Compete Block Image Object Reference)
    Texture=Get FNCS_Texture() -> (FNCS_Texture:Texture 2D Object Reference)
   )
   > InitTournamentBlockIcon(
    BlockImage=Get Block_VictoryIcon() -> (Block_VictoryIcon:WBP Compete Block Image Object Reference)
    Texture=Get Victory_Texture() -> (Victory_Texture:Texture 2D Object Reference)
   )
   > InitTournamentBlockIcon(
    BlockImage=Get Block_OthersIcon() -> (Block_OthersIcon:WBP Compete Block Image Object Reference)
    Texture=Get Others_Texture() -> (Others_Texture:Texture 2D Object Reference)
   )
   > InitTournamentBlockIcon(
    BlockImage=Get Block_RankedIcon() -> (Block_RankedIcon:WBP Compete Block Image Object Reference)
    Texture=Get Ranked_Texture() -> (Ranked_Texture:Texture 2D Object Reference)
   )
   SetBlockTiming!(
    execute=@n1.then_3
    self:UIKit Block Object Reference=Get ModularBlocks() -> (ModularBlocks:Array of UIKit Block Object References)
    Timing:"UIKit Block Timing Structure (by ref)"=Make UIKit Block Timing!(
     Hovering:float=0
     Unhovering:float=0
     Focusing:float=0.3
     Unfocusing:float=0.15
     Pressing:float=0.1
     Releasing:float=0.1
     Disabling:float=0.1
     Enabling:float=0.1
     Selecting:float=0.2
     DeselectingIdle:float=0.2
     DeselectingFocused:float=0.2
    ) -> (UIKitBlockTiming:UIKit Block Timing Structure) {is=title K2Node_MakeStruct}
   ) -> (then)
  ]}
 InitTournamentBlockIcon {nodes [
   n1 = InitTournamentBlockIcon!() -> (then, BlockImage:WBP Compete Block Image Object Reference, Texture:"Texture 2D Object Reference (by ref)") {is=title K2Node_FunctionEntry}
   > SetMaterial!(
    execute
    self:WBP Compete Block Image Object Reference=@n1
    InMaterial:"Material Instance Object Reference (by ref)"=Get Material_Cups() -> (Material_Cups:Material Instance Object Reference)
   ) -> (then)
   > SetTexture!(execute, self:WBP Compete Block Image Object Reference=@n1, InTexture:"Texture 2D Object Reference (by ref)"=@n1.Texture) -> (then)
   > RegisterModularBlock(Block=@n1)
  ]}
 PlayTransitionShowIcons {nodes [
   n1 = PlayTransitionShowIcons!() -> (then, PlayMode:EUMGSequencePlayMode Enum=Forward, Instant:bool=~) {is=title K2Node_FunctionEntry}
   > SetVisibility(
    self=Get Fade() -> (Fade:Image Object Reference)
    InVisibility=Select!(
     Forward:ESlateVisibility Enum=SelfHitTestInvisible
     Reverse:ESlateVisibility Enum=Hidden
     PingPong:ESlateVisibility Enum=Hidden
     Index:EUMGSequencePlayMode Enum=@n1
    ) -> (ReturnValue)
   )
   > Set IsCalendarOpened(IsCalendarOpened:bool=Select(Forward:bool=true, Reverse:bool=false, PingPong:bool=false, Index=@n1) -> (ReturnValue:bool)) -> (Output_Get:bool)
   > RefreshAmountTournaments()
   > n2 = Branch(Condition=Get FakeLockedState())
   > SetRenderOpacity(self=Get TournamentIcons() -> (TournamentIcons:Grid Panel Object Reference), InOpacity=0)
   n3 = Branch(execute=@n2.else)
   > SetRenderOpacity(self=Get TournamentIcons(), InOpacity=Select(Forward:float=1, Reverse:float=0, PingPong:float=0, Index=@n1) -> (ReturnValue:float))
   PlayAnimation(
    execute=@n3.else
    InAnimation=Get Anim_ShowTournamentIcons() -> (Anim_ShowTournamentIcons:Widget Animation Object Reference)
    PlayMode=@n1
    PlaybackSpeed=5
   )
   Reroute Node(InputPin:bool=@n1.Instant) -> (OutputPin:bool)
  ]}
 RefreshAmountTournaments {nodes [
   RefreshAmountTournaments() {is=title K2Node_FunctionEntry}
   > n1 = Branch(Condition=Get IsCalendarOpened() -> (IsCalendarOpened:bool))
   n2 = Get FortPoblanoTournamentsCalendarDayTileVM() -> (FortPoblanoTournamentsCalendarDayTileVM:Calendar Monthly Day Tile VM Object Reference)
   n3 = GetTotalTournaments!(self:Calendar Monthly Day Tile VM Object Reference=@n2) -> (ReturnValue:int=0)
   > SetText(
    self=Get RichText_Tournaments() -> (RichText_Tournaments:Common Rich Text Block Object Reference)
    InText=Format Text!(Format="INVTEXT(\"<text color=\\\"FFFFFF\\\">{Amount}</>\")", Amount:int=@n3) -> (Result)
   )
   n4 = GetCountOfBookmarkedTournaments!(self:Calendar Monthly Day Tile VM Object Reference=@n2) -> (ReturnValue:int=0)
   n5 = "To Text (Integer)"(Value=@n4)
   n6 = "integer > integer"(A=@n4)
   n7 = "integer > integer"(A=@n3, B=1)
   SetText(
    execute=@n1.else
    self=Get RichText_Tournaments()
    InText=Format Text!(
     Format="LOCTEXT(\"{Bookmarked}<text color=\\\"FFFFFF\\\">{Divider}{Amount} {Events}</>\")"
     Bookmarked:Text=SelectText(A=@n5, bPickA=@n6)
     Divider:Text=SelectText(A="INVTEXT(\"/\")", bPickA=@n6)
     Amount=@n3
     Events:Text=SelectText(A="LOCTEXT(\"events\")", B="LOCTEXT(\"event\")", bPickA=@n7)
    ) -> (Result)
   )
  ]}
 RefreshTotalBookmarked {nodes [
   RefreshTotalBookmarked!() -> (then, Amount:int) {is=title K2Node_FunctionEntry}
   > RefreshAmountTournaments() {is=title K2Node_CallFunction}
  ]}
 SetAsToday {nodes [
   n1 = SetAsToday!() -> (then, IsToday:bool=~) {is=title K2Node_FunctionEntry}
   > SetRenderOpacity(self=Get Background_Date() -> (Background_Date:Image Object Reference), InOpacity="To Float (Boolean)"(InBool=@n1))
   > SetFontMaterial!(
    execute
    self:Text Object Reference=Get Text_Date()
    InMaterial:Material Interface Object Reference=Select!(
     Option 0:Material Interface Object Reference
     Option 1:Material Interface Object Reference=Get Material_FontDateToday() -> (Material_FontDateToday:Material Instance Object Reference)
     Index:bool=@n1
    ) -> (ReturnValue:Material Interface Object Reference)
   ) -> (then)
  ]}
 SetDate {nodes [
   n1 = SetDate!() -> (then, Amount:int) {is=title K2Node_FunctionEntry}
   > "SetText (Text)"(self=Get Text_Date(), InText="To Text (Integer)"(Value=@n1))
  ]}
 SetHasAnyBookmarks {nodes [
   n1 = SetHasAnyBookmarks!() -> (then, IsBookmarked:bool=~) {is=title K2Node_FunctionEntry}
   > SetVisibility(
    self=Get AttachmentBookmarked() -> (AttachmentBookmarked:Common Visual Attachment Object Reference)
    InVisibility=Select(Option 0:ESlateVisibility Enum=Collapsed, Option 1:ESlateVisibility Enum=SelfHitTestInvisible, Index=@n1) -> (ReturnValue:ESlateVisibility Enum)
   )
   > SetColorAndOpacity(
    self:Image Object Reference=Get TotalTournamentsIcon() -> (TotalTournamentsIcon:Image Object Reference)
    InColorAndOpacity:Linear Color Structure=SelectColor(A=(R=0,G=0,B=0,A=1), B=(R=1,G=1,B=1,A=1), bPickA=@n1)
   )
  ]}
 SetRenderOpacityForNonModuarBlockElements {nodes [
   n1 = SetRenderOpacityForNonModuarBlockElements!() -> (then, InOpacity:float=~) {is=title K2Node_FunctionEntry}
   > SetRenderOpacity(self=[Get RichText_Tournaments(), Get TotalTournamentsIcon(), Get Text_Date(), Get TournamentIcons()], InOpacity=@n1)
  ]}
 SetTileTimeState {nodes [
   n1 = SetTileTimeState!() -> (then, TimeZone:ECupTimeZoneStyle Enum=PastWeek) {is=title K2Node_FunctionEntry}
   > n2 = Switch on ECupTimeZoneStyle!(execute, Selection:ECupTimeZoneStyle Enum=@n1) -> (PastWeek:Exec, CurrentWeek:Exec, UpcomingWeek:Exec) {is=title K2Node_SwitchEnum}
   n3 = Get ModularBlocks()
   SetRenderOpacity(execute=@n2.CurrentWeek, self=@n3, InOpacity=1)
   > SetRenderOpacityForNonModuarBlockElements!(execute, self, InOpacity:float=1) -> (then) {is=title K2Node_CallFunction}
   n4 = Get Background() -> (Background:Image Object Reference)
   n5 = Get Fade()
   n6 = Get Bookmarked() -> (Bookmarked:Image Object Reference)
   > SetRenderOpacity(self=[@n4, @n5, @n6], InOpacity=1)
   SetRenderOpacity(execute=@n2.PastWeek, self=@n3, InOpacity=0.08)
   > SetRenderOpacityForNonModuarBlockElements(InOpacity=0.4)
   > SetRenderOpacity(self=Get TotalTournamentsIcon(), InOpacity=0.08)
   > SetRenderOpacity(self=[@n4, @n5, @n6], InOpacity=0.3)
   SetRenderOpacity(execute=@n2.UpcomingWeek, self=@n3, InOpacity=1)
   > SetRenderOpacityForNonModuarBlockElements(InOpacity=1)
   > SetRenderOpacity(self=[@n4, @n5], InOpacity=0.5)
   > SetRenderOpacity(self=Get Bookmarked(), InOpacity=1)
  ]}
 ShouldFocusOnSelected {nodes [
   ShouldFocusOnSelected() {is=title K2Node_FunctionEntry}
   > Return Node!(execute, ShouldFocus:bool=false)
  ]}
 TotalTournaments {nodes [
   n1 = TotalTournaments!() -> (then, Amount:int) {is=title K2Node_FunctionEntry}
   > Set!(
    execute
    Padding_Left:float=Select(Option 0:float=0, Option 1:float=5, Index="integer < integer"(A=@n1, B=10)) -> (ReturnValue:float)
    Padding_Top:float=0
    Padding_Right:float=0
    Padding_Bottom:float=0
    self:Stack Box Slot Object Reference=SlotAsStackBoxSlot!(Widget:Widget Object Reference=Get RichText_Tournaments()) -> (ReturnValue:Stack Box Slot Object Reference)
   ) -> (then, Output_Get:Margin Structure)
   n2 = "Equal (Integer)"(A=Get Amount() -> (Amount:int=0))
   > Set FakeLockedState(FakeLockedState:bool=@n2) -> (Output_Get:bool)
   > SetVisibility(
    self=Get Stack_TotalTournaments() -> (Stack_TotalTournaments:Stack Box Object Reference)
    InVisibility=Select(Option 0:ESlateVisibility Enum=SelfHitTestInvisible, Option 1:ESlateVisibility Enum=Collapsed, Index=@n2) -> (ReturnValue:ESlateVisibility Enum)
   )
   > SetRenderOpacity(self=Get Date() -> (Date:Overlay Object Reference), InOpacity=SelectFloat(A=0.5, B=1, bPickA=@n2))
   > RequestTransitionOnBlocks!(
    execute
    self:UIKit Modular Button Object Reference
    State:EUIKitBlockVisualState Enum=Select(Option 0:EUIKitBlockVisualState Enum=Enabled, Option 1:EUIKitBlockVisualState Enum=Disabled, Index=Get FakeLockedState()) -> (ReturnValue:EUIKitBlockVisualState Enum)
   ) -> (then)
  ]}
 UpdateCashCupsState {nodes [
   n1 = UpdateCashCupsState!() -> (then, State:ECupState Enum=Available) {is=title K2Node_FunctionEntry}
   > UpdateCupImageColorByState!(execute, self, State:ECupState Enum=@n1, BlockImage:WBP Compete Block Image Object Reference=Get Block_CashIcon()) -> (then)
  ]}
 UpdateCupImageColorByState {nodes [
   n1 = UpdateCupImageColorByState!() -> (then, State:ECupState Enum=Available, BlockImage:WBP Compete Block Image Object Reference) {is=title K2Node_FunctionEntry}
   > Is Valid(InputObject=@n1.BlockImage)
   > n2 = Sequence!(execute) -> (then_0, then_1, then_2)
   n3 = Select!(
    Unavailable:Linear Color Structure=(R=1,G=1,B=1,A=0.02)
    Available:Linear Color Structure=(R=1,G=1,B=1,A=1)
    Bookmarked:Linear Color Structure=(R=0.0392,G=1,B=0.2314,A=1)
    Index:ECupState Enum=@n1
   ) -> (ReturnValue:Linear Color Structure)
   n4 = Get ImageMID!(self:WBP Compete Block Image Object Reference=@n1.BlockImage) -> (ImageMID:Material Instance Dynamic Object Reference)
   > SetVectorParameterValue(self=@n4, ParameterName=Color_Idle, Value=@n3)
   > SetVectorParameterValue(self=@n4, ParameterName=Color_Hovered, Value=@n3)
   > SetVectorParameterValue(self=@n4, ParameterName=Color_Focused, Value=@n3)
   > SetVectorParameterValue(self=@n4, ParameterName=Color_Pressed, Value=@n3)
   SetMarked!(
    execute=@n2.then_2
    self:WBP Compete Block Image Object Reference=@n1.BlockImage
    IsMarked:bool="Equal (Enum)"(A:ECupState Enum=@n1, B:ECupState Enum=Bookmarked)
   ) -> (then)
   SetVectorParameterValue(execute=@n2.then_1, self=@n4, ParameterName=Color_SelectedIdle, Value=@n3)
   > SetVectorParameterValue(self=@n4, ParameterName=Color_SelectedHovered, Value=@n3)
   > SetVectorParameterValue(self=@n4, ParameterName=Color_SelectedFocused, Value=@n3)
   > SetVectorParameterValue(self=@n4, ParameterName=Color_SelectedPressed, Value=@n3)
  ]}
 UpdateIconFNCSState {nodes [
   n1 = UpdateIconFNCSState!() -> (then, State:ECupState Enum=Available) {is=title K2Node_FunctionEntry}
   > UpdateCupImageColorByState(State=@n1, BlockImage=Get Block_FNCSIcon()) {is=title K2Node_CallFunction}
  ]}
 UpdateOtherCupsState {nodes [
   n1 = UpdateOtherCupsState!() -> (then, State:ECupState Enum=Available) {is=title K2Node_FunctionEntry}
   > UpdateCupImageColorByState(State=@n1, BlockImage=Get Block_OthersIcon())
  ]}
 UpdateRankedCupsState {nodes [
   n1 = UpdateRankedCupsState!() -> (then, State:ECupState Enum=Available) {is=title K2Node_FunctionEntry}
   > UpdateCupImageColorByState(State=@n1, BlockImage=Get Block_RankedIcon())
  ]}
 UpdateShopsCupsState {nodes [
   n1 = UpdateShopsCupsState!() -> (then, State:ECupState Enum=Available) {is=title K2Node_FunctionEntry}
   > UpdateCupImageColorByState(State=@n1, BlockImage=Get Block_ItemShopIcon())
  ]}
 UpdateVictoryCupsState {nodes [
   n1 = UpdateVictoryCupsState!() -> (then, State:ECupState Enum=Available) {is=title K2Node_FunctionEntry}
   > UpdateCupImageColorByState(State=@n1, BlockImage=Get Block_VictoryIcon())
  ]}
]
variables [
 BackgroundMaterial {type=MaterialInstance, default=/Compete/Calendar/MaterialInstances/MI_UI_CalendarTile_BackgroundToday}
 Cash_Texture {type=Texture2D, default=/FortUILibrary/Textures/Icons/T_UI_Icon_UIKit_CashCup_DF}
 FakeLockedState {type=boolean, default=true}
 FNCS_Texture {type=Texture2D, default=/FortUILibrary/Textures/Icons/T_UI_Icon_UIKit_FNCS_DF}
 Font_Date {
  type=SlateFontInfo
  default {
   bForceMonospaced=false
   bMaterialIsStencil=false
   fontMaterial=/Compete/Calendar/MaterialInstances/MI_UI_Day_Text
   fontObject=&HeadingNow
   letterSpacing=0
   monospacedWidth=1
   size=24
   skewAmount=0
   typefaceFontName=75 Medium
   outlineSettings {
    bApplyOutlineToDropShadows=false
    bMiteredCorners=false
    bSeparateFillAlpha=false
    outlineMaterial=null
    outlineSize=0
    outlineColor=<rgba 0 0 0 1>
   }
  }
 }
 Font_TotalTournaments {
  type=SlateFontInfo
  default {
   bForceMonospaced=false
   bMaterialIsStencil=false
   fontMaterial=/FortUILibrary/UIKit/Buttons/TabButtons/_MaterialInstances/MI_UI_TopNav_Text
   fontObject=&HeadingNow
   letterSpacing=0
   monospacedWidth=1
   size=18
   skewAmount=0
   typefaceFontName=74 Regular
   outlineSettings {
    bApplyOutlineToDropShadows=false
    bMiteredCorners=false
    bSeparateFillAlpha=false
    outlineMaterial=null
    outlineSize=0
    outlineColor=<rgba 0 0 0 1>
   }
  }
 }
 IsCalendarOpened {type=boolean, default=true}
 ItemShop_Texture {type=Texture2D, default=/FortUILibrary/Textures/Icons/T_UI_Icon_UIKit_ItemShopCup_DF}
 Material_Cups {type=MaterialInstance, default=/Compete/Calendar/MaterialInstances/MI_TournamentIconWithOutline}
 Material_FontDateToday {type=MaterialInstance, default=/Compete/Calendar/MaterialInstances/MI_UI_Today_Text}
 Material_TotalTournaments {type=MaterialInstance, default=/FortUILibrary/Textures/MaterialInstances/MI_UIKit_Icon_Compete}
 Others_Texture {type=Texture2D, default=/FortUILibrary/Textures/Icons/T_UI_Icon_UIKit_OtherTournaments_DF}
 Ranked_Texture {type=Texture2D, default=/FortUILibrary/Textures/Icons/T_UI_Icon_UIKit_RankedCup_DF}
 Victory_Texture {type=Texture2D, default=/FortUILibrary/Textures/Icons/T_UI_Icon_UIKit_VictoryCup_DF}
]
widgets {
 info {
  inheritedWidgetCount=1
  parentClass=/Compete/WBP_CompeteTileButton_Base.WBP_CompeteTileButton_Base_C
  rootWidgetClass=&GridPanel
  widgetCount=25
 }
 widgets [
  Root {widgetClassPath=&GridPanel, bIsVariable=true, widget=Root, properties {columnFill [0,1], rowFill [1]}}
  SelectionOutline {
   widgetClassPath=&CommonBorder
   bIsVariable=true
   parent=Root
   widget=SelectionOutline
   properties {
    horizontalAlignment=HAlign_Left
    renderOpacity=0
    verticalAlignment=VAlign_Bottom
    background {
     bIsDynamicallyLoaded=false
     drawAs=Image
     imageType=NoImage
     mirroring=NoMirror
     resourceName=None
     resourceObject=/FortUILibrary/Systems/Items/MaterialInstances/MI_UIKit_EquippedStateOutline
     tiling=NoTile
     imageSize=<32 32>
     margin=<ltrb 0 0 0 0>
     outlineSettings {
      bUseBrushTransparency=false
      roundingType=HalfHeightRadius
      width=0
      color {colorUseRule=UseColor_Specified, specifiedColor=<rgba 0 0 0 0>}
      cornerRadii=<0 0 0 0>
     }
     tintColor {colorUseRule=UseColor_Specified, specifiedColor=<rgba 1 1 1 1>}
     uVRegion {bIsValid=false, max=<0 0>, min=<0 0>}
    }
    contentColorAndOpacity=<rgba 1 1 1 0.995>
    padding=<ltrb 0 0 0 0>
   }
   slot {class=GridSlot, properties {columnSpan=2, layer=10}}
  }
  Background {
   widgetClassPath=&Image
   bIsVariable=true
   parent=Root
   widget=Background
   properties {brush {
     bIsDynamicallyLoaded=false
     drawAs=Image
     imageType=FullColor
     mirroring=NoMirror
     resourceName=None
     resourceObject=/Compete/Calendar/MaterialInstances/MI_UI_CalendarTile_Background
     tiling=NoTile
     imageSize=<32 32>
     margin=<ltrb 0 0 0 0>
     outlineSettings {
      bUseBrushTransparency=false
      roundingType=HalfHeightRadius
      width=0
      color {colorUseRule=UseColor_Specified, specifiedColor=<rgba 0 0 0 0>}
      cornerRadii=<0 0 0 0>
     }
     tintColor {colorUseRule=UseColor_Specified, specifiedColor=<rgba 1 1 1 1>}
     uVRegion {bIsValid=false, max=<0 0>, min=<0 0>}
    }}
   slot {class=GridSlot, properties {columnSpan=2}}
  }
  Fade {
   widgetClassPath=&Image
   bIsVariable=true
   parent=Root
   widget=Fade
   properties {brush {
     bIsDynamicallyLoaded=false
     drawAs=Image
     imageType=FullColor
     mirroring=NoMirror
     resourceName=None
     resourceObject=/Compete/Calendar/MaterialInstances/MI_UI_CalendarTile_Fade
     tiling=NoTile
     imageSize=<32 32>
     margin=<ltrb 0 0 0 0>
     outlineSettings {
      bUseBrushTransparency=false
      roundingType=HalfHeightRadius
      width=0
      color {colorUseRule=UseColor_Specified, specifiedColor=<rgba 0 0 0 0>}
      cornerRadii=<0 0 0 0>
     }
     tintColor {colorUseRule=UseColor_Specified, specifiedColor=<rgba 1 1 1 1>}
     uVRegion {bIsValid=false, max=<0 0>, min=<0 0>}
    }}
   slot {class=GridSlot}
  }
  Block_Outline {
   widgetClassPath=/FortUILibrary/UIKit/Buttons/Blocks/WBP_UIKit_Block_Outline
   bIsVariable=true
   parent=Root
   widget=Block_Outline
   properties {bottomLeftCornerRadius=30, bottomRightCornerRadius=30, radiusInPixels=true, topLeftCornerRadius=30, topRightCornerRadius=30}
   slot {class=GridSlot, properties {columnSpan=2, layer=10}}
  }
  Spacer_48 {widgetClassPath=&Spacer, parent=Root, widget=Spacer_48, properties {size=<84 1>}, slot {class=GridSlot}}
  Info {
   widgetClassPath=&StackBox
   parent=Root
   widget=Info
   properties {orientation=Orient_Vertical}
   slot {class=GridSlot, properties {layer=2, padding=<ltrb 8 10 0 12>}}
  }
  Date {
   widgetClassPath=&Overlay
   bIsVariable=true
   parent=Info
   widget=Date
   slot {class=StackBoxSlot, properties {verticalAlignment=VAlign_Top, padding=<ltrb 4 0 8 0>, size {sizeRule=Fill, value=1}}}
  }
  Background_Date {
   widgetClassPath=&Image
   bIsVariable=true
   parent=Date
   widget=Background_Date
   properties {
    renderOpacity=0
    visibility=SelfHitTestInvisible
    brush {
     bIsDynamicallyLoaded=false
     drawAs=Image
     imageType=FullColor
     mirroring=NoMirror
     resourceName=None
     resourceObject=&MI_UI_CalendarTile_BackgroundToday
     tiling=NoTile
     imageSize=<32 32>
     margin=<ltrb 0 0 0 0>
     outlineSettings {
      bUseBrushTransparency=false
      roundingType=HalfHeightRadius
      width=0
      color {colorUseRule=UseColor_Specified, specifiedColor=<rgba 0 0 0 0>}
      cornerRadii=<0 0 0 0>
     }
     tintColor {colorUseRule=UseColor_Specified, specifiedColor=<rgba 1 1 1 1>}
     uVRegion {bIsValid=false, max=<0 0>, min=<0 0>}
    }
   }
   slot {class=OverlaySlot, properties {horizontalAlignment=HAlign_Fill, verticalAlignment=VAlign_Fill}}
  }
  Text_Date {
   widgetClassPath=&CommonTextBlock
   bIsVariable=true
   parent=Date
   widget=Text_Date
   properties {
    text=5
    font {
     bForceMonospaced=false
     bMaterialIsStencil=false
     fontMaterial=null
     fontObject=&HeadingNow
     letterSpacing=0
     monospacedWidth=1
     size=24
     skewAmount=0
     typefaceFontName=75 Medium
     outlineSettings {
      bApplyOutlineToDropShadows=false
      bMiteredCorners=false
      bSeparateFillAlpha=false
      outlineMaterial=null
      outlineSize=0
      outlineColor=<rgba 0 0 0 1>
     }
    }
    shadowOffset=<0 0>
   }
   slot {class=OverlaySlot, properties {verticalAlignment=VAlign_Center, padding=<ltrb 11 0 0 0>}}
  }
  Stack_TotalTournaments {
   widgetClassPath=&StackBox
   bIsVariable=true
   parent=Info
   widget=Stack_TotalTournaments
   slot {class=StackBoxSlot, properties {horizontalAlignment=HAlign_Left, verticalAlignment=VAlign_Bottom, padding=<ltrb 9 0 -200 0>}}
  }
  Overlay_0 {
   widgetClassPath=&Overlay
   parent=Stack_TotalTournaments
   widget=Overlay_0
   slot {class=StackBoxSlot, properties {horizontalAlignment=HAlign_Left, verticalAlignment=VAlign_Center}}
  }
  AttachmentBookmarked {
   widgetClassPath=&CommonVisualAttachment
   bIsVariable=true
   parent=Overlay_0
   widget=AttachmentBookmarked
   properties {visibility=Collapsed, contentAnchor=<0.5 0>}
   slot {class=OverlaySlot, properties {horizontalAlignment=HAlign_Center, padding=<ltrb 0 -1 0 0>}}
  }
  Bookmarked {
   widgetClassPath=&Image
   bIsVariable=true
   parent=AttachmentBookmarked
   widget=Bookmarked
   properties {brush {
     bIsDynamicallyLoaded=false
     drawAs=Image
     imageType=FullColor
     mirroring=NoMirror
     resourceName=None
     resourceObject=/Compete/MaterialInstances/MI_UI_Compete_Flag
     tiling=NoTile
     imageSize=<30 38>
     margin=<ltrb 0 0 0 0>
     outlineSettings {
      bUseBrushTransparency=false
      roundingType=HalfHeightRadius
      width=0
      color {colorUseRule=UseColor_Specified, specifiedColor=<rgba 0 0 0 0>}
      cornerRadii=<0 0 0 0>
     }
     tintColor {colorUseRule=UseColor_Specified, specifiedColor=<rgba 1 1 1 1>}
     uVRegion {bIsValid=false, max=<0 0>, min=<0 0>}
    }}
   slot {class=SizeBoxSlot, properties {horizontalAlignment=HAlign_Center, verticalAlignment=VAlign_Center}}
  }
  TotalTournamentsIcon {
   widgetClassPath=&Image
   bIsVariable=true
   parent=Overlay_0
   widget=TotalTournamentsIcon
   properties {brush {
     bIsDynamicallyLoaded=false
     drawAs=Image
     imageType=FullColor
     mirroring=NoMirror
     resourceName=None
     resourceObject=&MI_UIKit_Icon_Compete
     tiling=NoTile
     imageSize=<32 32>
     margin=<ltrb 0 0 0 0>
     outlineSettings {
      bUseBrushTransparency=false
      roundingType=HalfHeightRadius
      width=0
      color {colorUseRule=UseColor_Specified, specifiedColor=<rgba 0 0 0 0>}
      cornerRadii=<0 0 0 0>
     }
     tintColor {colorUseRule=UseColor_Specified, specifiedColor=<rgba 1 1 1 1>}
     uVRegion {bIsValid=false, max=<0 0>, min=<0 0>}
    }}
   slot {class=OverlaySlot}
  }
  RichText_Tournaments {
   widgetClassPath=&CommonRichTextBlock
   bIsVariable=true
   parent=Stack_TotalTournaments
   widget=RichText_Tournaments
   properties {
    bOverrideDefaultStyle=true
    defaultTextStyleOverride {
     overflowPolicy=Clip
     transformPolicy=None
     colorAndOpacity {colorUseRule=UseColor_Specified, specifiedColor=<rgba 0.0395 1 0.2307 1>}
     font {
      bForceMonospaced=false
      bMaterialIsStencil=false
      fontMaterial=null
      fontObject=&HeadingNow
      letterSpacing=0
      monospacedWidth=1
      size=18
      skewAmount=0
      typefaceFontName=74 Regular
      outlineSettings {
       bApplyOutlineToDropShadows=false
       bMiteredCorners=false
       bSeparateFillAlpha=false
       outlineMaterial=null
       outlineSize=0
       outlineColor=<rgba 0 0 0 1>
      }
     }
     highlightColor {colorUseRule=UseColor_Specified, specifiedColor=<rgba 0 0 0 1>}
     highlightShape {
      bIsDynamicallyLoaded=false
      drawAs=Image
      imageType=NoImage
      mirroring=NoMirror
      resourceName=None
      resourceObject=null
      tiling=NoTile
      imageSize=<32 32>
      margin=<ltrb 0 0 0 0>
      outlineSettings {
       bUseBrushTransparency=false
       roundingType=HalfHeightRadius
       width=0
       color {colorUseRule=UseColor_Specified, specifiedColor=<rgba 0 0 0 0>}
       cornerRadii=<0 0 0 0>
      }
      tintColor {colorUseRule=UseColor_Specified, specifiedColor=<rgba 1 1 1 1>}
      uVRegion {bIsValid=false, max=<0 0>, min=<0 0>}
     }
     selectedBackgroundColor {colorUseRule=UseColor_Foreground, specifiedColor=<rgba 1 0 1 1>}
     shadowColorAndOpacity=<rgba 0 0 0 1>
     shadowOffset=<0 0>
     strikeBrush {
      bIsDynamicallyLoaded=false
      drawAs=Image
      imageType=NoImage
      mirroring=NoMirror
      resourceName=None
      resourceObject=null
      tiling=NoTile
      imageSize=<32 32>
      margin=<ltrb 0 0 0 0>
      outlineSettings {
       bUseBrushTransparency=false
       roundingType=HalfHeightRadius
       width=0
       color {colorUseRule=UseColor_Specified, specifiedColor=<rgba 0 0 0 0>}
       cornerRadii=<0 0 0 0>
      }
      tintColor {colorUseRule=UseColor_Specified, specifiedColor=<rgba 1 1 1 1>}
      uVRegion {bIsValid=false, max=<0 0>, min=<0 0>}
     }
     underlineBrush {
      bIsDynamicallyLoaded=false
      drawAs=Image
      imageType=NoImage
      mirroring=NoMirror
      resourceName=None
      resourceObject=null
      tiling=NoTile
      imageSize=<32 32>
      margin=<ltrb 0 0 0 0>
      outlineSettings {
       bUseBrushTransparency=false
       roundingType=HalfHeightRadius
       width=0
       color {colorUseRule=UseColor_Specified, specifiedColor=<rgba 0 0 0 0>}
       cornerRadii=<0 0 0 0>
      }
      tintColor {colorUseRule=UseColor_Specified, specifiedColor=<rgba 1 1 1 1>}
      uVRegion {bIsValid=false, max=<0 0>, min=<0 0>}
     }
    }
   }
   slot {class=StackBoxSlot}
  }
  TournamentIcons {
   widgetClassPath=&GridPanel
   bIsVariable=true
   parent=Root
   widget=TournamentIcons
   properties {visibility=HitTestInvisible, columnFill [1,1,1], rowFill [1,1]}
   slot {class=GridSlot, properties {column=1, layer=2, padding=<ltrb 0 2 8 2>}}
  }
  Block_CashIcon {
   widgetClassPath=&/WBP_Compete_Block_Image
   bIsVariable=true
   parent=TournamentIcons
   widget=Block_CashIcon
   slot {class=GridSlot, properties {horizontalAlignment=HAlign_Center, verticalAlignment=VAlign_Center}}
  }
  Block_ItemShopIcon {
   widgetClassPath=&WBP_Compete_Block_Image
   bIsVariable=true
   parent=TournamentIcons
   widget=Block_ItemShopIcon
   slot {class=GridSlot, properties {column=1, horizontalAlignment=HAlign_Center, verticalAlignment=VAlign_Center}}
  }
  Block_FNCSIcon {
   widgetClassPath=&WBP_Compete_Block_Image
   bIsVariable=true
   parent=TournamentIcons
   widget=Block_FNCSIcon
   slot {class=GridSlot, properties {column=2, horizontalAlignment=HAlign_Center, verticalAlignment=VAlign_Center}}
  }
  Block_VictoryIcon {
   widgetClassPath=&WBP_Compete_Block_Image
   bIsVariable=true
   parent=TournamentIcons
   widget=Block_VictoryIcon
   slot {class=GridSlot, properties {horizontalAlignment=HAlign_Center, row=1, verticalAlignment=VAlign_Center}}
  }
  Block_OthersIcon {
   widgetClassPath=&WBP_Compete_Block_Image
   bIsVariable=true
   parent=TournamentIcons
   widget=Block_OthersIcon
   slot {class=GridSlot, properties {column=1, horizontalAlignment=HAlign_Center, row=1, verticalAlignment=VAlign_Center}}
  }
  Block_RankedIcon {
   widgetClassPath=&WBP_Compete_Block_Image
   bIsVariable=true
   parent=TournamentIcons
   widget=Block_RankedIcon
   slot {class=GridSlot, properties {column=2, horizontalAlignment=HAlign_Center, row=1, verticalAlignment=VAlign_Center}}
  }
  DebugTxt {
   widgetClassPath=&UEFN_TextBlock
   bIsVariable=true
   parent=Root
   widget=DebugTxt
   properties {
    justification=Center
    text=Debug Text
    visibility=Collapsed
    font {
     bForceMonospaced=false
     bMaterialIsStencil=false
     fontMaterial=null
     fontObject=&BurbankSmall
     letterSpacing=0
     monospacedWidth=1
     size=24
     skewAmount=0
     typefaceFontName=Black
     outlineSettings {
      bApplyOutlineToDropShadows=false
      bMiteredCorners=false
      bSeparateFillAlpha=false
      outlineMaterial=null
      outlineSize=3
      outlineColor=<rgba 0 0 0 1>
     }
    }
   }
   slot {class=GridSlot, properties {horizontalAlignment=HAlign_Center, verticalAlignment=VAlign_Center}}
  }
  InputActionWidget {widgetClassPath=&CommonActionWidget, bInherited=true, widget=null}
 ]
}
